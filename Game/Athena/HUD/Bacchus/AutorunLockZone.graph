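# Touch HUD widget for the autorun lock zone, built on FortAutorunLockZone with animated button and ripple material images.
# When shown, it plays the Ripple animation if touch pulse is enabled in client settings, otherwise Enter; hiding stops Loop.
# Zone changes switch between Highlight and Enter, and releasing in-zone plays Release, then broadcasts completion.
# Animations mainly drive material parameters; the graph never plays Drag or Loop.
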WidgetBlueprint /Game/Athena/HUD/Bacchus/AutorunLockZone
animations [
 Anim_Autorun_Drag {
  binding_count=1
  length_seconds=1.0005
  object=&self:Anim_Autorun_Drag
  sequence {bindings [Image_ImprovedAutorunLockZone {
     id=#1
     possessed_class=Image
     tracks [{
       type=MovieSceneWidgetMaterialTrack
       sections [{
         type=MovieSceneParameterSection
         end_frame=20
         start_frame=0
         channels [
          GlowFX_Opacity {type=MovieSceneScriptingFloatChannel, keys [[0,0],[20,1],[20.0047,0]]}
          Animation_Icon_Color {type=MovieSceneScriptingFloatChannel, keys [[0,0],[20.0047,1]]}
          All_Opacity {type=MovieSceneScriptingFloatChannel, const [1,1]}
          OuterOutline_Opacity {type=MovieSceneScriptingFloatChannel, const [0.8,1]}
          UseArc {type=MovieSceneScriptingFloatChannel, const [1,1]}
          GlowFX_ColorR {type=MovieSceneScriptingFloatChannel, keys [[0,1],[20,0.9131]]}
          GlowFX_ColorG {type=MovieSceneScriptingFloatChannel, keys [[0,1],[20,0.6308]]}
          GlowFX_ColorB {type=MovieSceneScriptingFloatChannel, keys [[0,1],[20,0.0802]]}
          GlowFX_ColorA {type=MovieSceneScriptingFloatChannel, keys [[0,0],[20,1]]}
         ]
        }]
      }]
    }]}
 }
 Anim_Autorun_Release {
  binding_count=3
  length_seconds=0.35
  object=&self:Anim_Autorun_Release
  sequence {bindings [
    Image_ImprovedAutorunLockZone {
     id=#2
     possessed_class=Image
     tracks [{
       type=MovieSceneWidgetMaterialTrack
       sections [{
         type=MovieSceneParameterSection
         end_frame=6
         start_frame=-1
         channels [
          GlowFX_Scale {type=MovieSceneScriptingFloatChannel, keys [[-5,0.614],[-1,1],[0,1],[5,1.614],[6,0.614]]}
          All_Opacity {type=MovieSceneScriptingFloatChannel, keys [[0,1],[5,0],[6,0]]}
          Animation_Icon_Color {type=MovieSceneScriptingFloatChannel, const [1,1]}
          UseArc {type=MovieSceneScriptingFloatChannel, const [1,1]}
         ]
        }]
      }]
    }
    Image_Ripple {
     id=#3
     possessed_class=Image
     tracks [
      {
       type=MovieSceneWidgetMaterialTrack
       sections [{
         type=MovieSceneParameterSection
         end_frame=6
         start_frame=-1
         channels [
          RippleCount {type=MovieSceneScriptingFloatChannel, const [1,1]}
          Phase {type=MovieSceneScriptingFloatChannel, keys [[0,0.44],[6,0.72]]}
          ColorR {type=MovieSceneScriptingFloatChannel, const [0.9131,1]}
          ColorG {type=MovieSceneScriptingFloatChannel, const [0.6308,1]}
          ColorB {type=MovieSceneScriptingFloatChannel, const [0.0802,1]}
          ColorA {type=MovieSceneScriptingFloatChannel, const [1,1]}
         ]
        }]
      }
      {
       type=MovieSceneFloatTrack
       display_name=Render Opacity
       sections [{type=MovieSceneFloatSection, end_frame=0, start_frame=-1, channels [None {type=MovieSceneScriptingFloatChannel, keys [[-1,0],[0,1]]}]}]
      }
     ]
    }
   ]}
 }
 Anim_Autorun_Loop {
  binding_count=1
  length_seconds=3
  object=&self:Anim_Autorun_Loop
  sequence {bindings [Image_ImprovedAutorunLockZone {
     id=#1
     possessed_class=Image
     tracks [{
       type=MovieSceneWidgetMaterialTrack
       sections [{
         type=MovieSceneParameterSection
         end_frame=60
         start_frame=0
         channels [
          GlowFX_Opacity {type=MovieSceneScriptingFloatChannel, keys [[0,0],[10,1],[20.0047,0]]}
          All_Opacity {type=MovieSceneScriptingFloatChannel, const [1,1]}
          Animation_Icon_Color {type=MovieSceneScriptingFloatChannel, const [1,1]}
          UseArc {type=MovieSceneScriptingFloatChannel, const [1,1]}
          GlowFX_ColorR {type=MovieSceneScriptingFloatChannel, keys [[0,1],[10,0.9131],[20.0047,1]]}
          GlowFX_ColorG {type=MovieSceneScriptingFloatChannel, keys [[0,1],[10,0.6308],[20.0047,1]]}
          GlowFX_ColorB {type=MovieSceneScriptingFloatChannel, keys [[0,1],[10,0.0802],[20.0047,1]]}
          GlowFX_ColorA {type=MovieSceneScriptingFloatChannel, keys [[0,0],[10,0.5],[20.0047,0]]}
         ]
        }]
      }]
    }]}
 }
 Anim_Autorun_Highlight {
  binding_count=1
  length_seconds=1
  object=&self:Anim_Autorun_Highlight
  sequence {bindings [Image_ImprovedAutorunLockZone {
     id=#4
     possessed_class=Image
     tracks [{
       type=MovieSceneWidgetMaterialTrack
       sections [{
         type=MovieSceneParameterSection
         end_frame=20
         start_frame=0
         channels [
          GlowFX_Opacity {type=MovieSceneScriptingFloatChannel, keys [[0,1],[20,0.75]]}
          All_Opacity {type=MovieSceneScriptingFloatChannel, const [1,1]}
          OuterOutline_Opacity {type=MovieSceneScriptingFloatChannel, const [0,1]}
          Animation_Icon_Color {type=MovieSceneScriptingFloatChannel, const [1,1]}
          UseArc {type=MovieSceneScriptingFloatChannel, const [1,1]}
          GlowFX_ColorR {type=MovieSceneScriptingFloatChannel, const [0.9131,1]}
          GlowFX_ColorG {type=MovieSceneScriptingFloatChannel, const [0.6308,1]}
          GlowFX_ColorB {type=MovieSceneScriptingFloatChannel, const [0.0802,1]}
          GlowFX_ColorA {type=MovieSceneScriptingFloatChannel, const [0.5,1]}
         ]
        }]
      }]
    }]}
 }
 Anim_Autorun_Enter {
  binding_count=1
  length_seconds=3.5
  object=&self:Anim_Autorun_Enter
  sequence {bindings [Image_ImprovedAutorunLockZone {
     id=#1
     possessed_class=Image
     tracks [{
       type=MovieSceneWidgetMaterialTrack
       sections [{
         type=MovieSceneParameterSection
         end_frame=70
         start_frame=0
         channels [
          GlowFX_Opacity {type=MovieSceneScriptingFloatChannel, keys [[0,0],[10,0.75],[20.0047,0]]}
          All_Opacity {type=MovieSceneScriptingFloatChannel, const [1,1]}
          OuterOutline_Opacity {type=MovieSceneScriptingFloatChannel, const [0.8,1]}
          UseArc {type=MovieSceneScriptingFloatChannel, const [1,1]}
          GlowFX_ColorR {type=MovieSceneScriptingFloatChannel, keys [[0,1],[10,0.9131],[20.0047,1]]}
          GlowFX_ColorG {type=MovieSceneScriptingFloatChannel, keys [[0,1],[10,0.9131],[20.0047,1]]}
          GlowFX_ColorB {type=MovieSceneScriptingFloatChannel, keys [[0,1],[10,0.9131],[20.0047,1]]}
          GlowFX_ColorA {type=MovieSceneScriptingFloatChannel, keys [[0,0],[10,0.5],[20.0047,0]]}
         ]
        }]
      }]
    }]}
 }
 Anim_Autorun_Ripple {
  binding_count=2
  length_seconds=1
  object=&self:Anim_Autorun_Ripple
  sequence {bindings [
    Image_Ripple {
     id=#5
     possessed_class=Image
     tracks [
      {
       type=MovieSceneWidgetMaterialTrack
       sections [{
         type=MovieSceneParameterSection
         end_frame=20
         start_frame=0
         channels [
          Phase {type=MovieSceneScriptingFloatChannel, keys [[0,0.45],[15,0.8]]}
          ColorR {type=MovieSceneScriptingFloatChannel, const [1,1]}
          ColorG {type=MovieSceneScriptingFloatChannel, const [1,1]}
          ColorB {type=MovieSceneScriptingFloatChannel, const [1,1]}
          ColorA {type=MovieSceneScriptingFloatChannel, const [1,1]}
         ]
        }]
      }
      {
       type=MovieSceneFloatTrack
       display_name=Render Opacity
       sections [{type=MovieSceneFloatSection, end_frame=1, start_frame=0, channels [None {type=MovieSceneScriptingFloatChannel, keys [[0,0],[1,1]]}]}]
      }
     ]
    }
    Image_ImprovedAutorunLockZone {
     id=#6
     possessed_class=Image
     tracks [{
       type=MovieSceneWidgetMaterialTrack
       sections [{
         type=MovieSceneParameterSection
         end_frame=20
         start_frame=0
         channels [
          GlowFX_Opacity {type=MovieSceneScriptingFloatChannel, const [0,1]}
          All_Opacity {type=MovieSceneScriptingFloatChannel, const [1,1]}
          OuterOutline_Opacity {type=MovieSceneScriptingFloatChannel, const [0.8,1]}
          UseArc {type=MovieSceneScriptingFloatChannel, const [1,1]}
          GlowFX_ColorR {type=MovieSceneScriptingFloatChannel, const [1,1]}
          GlowFX_ColorG {type=MovieSceneScriptingFloatChannel, const [1,1]}
          GlowFX_ColorB {type=MovieSceneScriptingFloatChannel, const [1,1]}
          GlowFX_ColorA {type=MovieSceneScriptingFloatChannel, const [1,1]}
         ]
        }]
      }]
    }
   ]}
 }
]
graphs [
 EventGraph {nodes [
   Event BP_OnReleasedInZone() {is=title K2Node_Event}
   > StopAllAnimations(self:User Widget Object Reference)
   > n1 = "Play Animation with Finished event (legacy)"!(
    execute:Exec
    Widget:User Widget Object Reference=Self-Reference()
    InAnimation:Widget Animation Object Reference=Get Anim_Autorun_Release() -> (Anim_Autorun_Release:Widget Animation Object Reference)
    StartAtTime:float=0
    NumLoopsToPlay:int=1
    PlayMode:EUMGSequencePlayMode Enum
    PlaybackSpeed:float=1
   ) -> (then:Exec, Result:UMGSequence Player Object Reference, Finished:Exec) {is=title K2Node_PlayAnimation}
   BroadcastReleaseAnimationFinished(execute=@n1.Finished, self:Fort Autorun Lock Zone Object Reference)
   n2 = Event OnIsInZoneChanged!() -> (OutputDelegate, then, bNewIsInZone:bool) {is=title K2Node_Event}
   > HighlightAutorunLockZone!(execute, self, IsInZone:bool=@n2.bNewIsInZone) -> (then)
   n3 = Event OnDisplayAutorunLockZone!() -> (OutputDelegate, then, bIsShown:bool) {is=title K2Node_Event}
   > n4 = Branch(Condition=@n3.bIsShown)
   > n5 = Cast To FortPlayerController(Object=GetOwningPlayer())
   n6 = GetClientSettings!(self:Fort Player Controller Object Reference=@n5) -> (ReturnValue:Fort Client Settings Record Object Reference)
   > n7 = Is Valid(InputObject=@n6)
   > n8 = Branch(Condition=GetTouchPulseAnimationEnabled!(self:Fort Client Settings Record Object Reference=@n6) -> (ReturnValue:bool))
   n9 = Get AutoRunAnimSpeed() -> (AutoRunAnimSpeed:float)
   > "Play Animation with Finished event (legacy)"(
    Widget=Self-Reference()
    InAnimation=Get Anim_Autorun_Ripple() -> (Anim_Autorun_Ripple:Widget Animation Object Reference)
    StartAtTime=0
    NumLoopsToPlay=1
    PlaybackSpeed=@n9
   )
   "Play Animation with Finished event (legacy)"(
    execute=[@n8.else, @n7.Is Not Valid, @n5.CastFailed]
    Widget=Self-Reference()
    InAnimation=Get Anim_Autorun_Enter() -> (Anim_Autorun_Enter:Widget Animation Object Reference)
    StartAtTime=0
    NumLoopsToPlay=1
    PlaybackSpeed=@n9
   )
   StopAnimation(execute=@n4.else, InAnimation=Get Anim_Autorun_Loop() -> (Anim_Autorun_Loop:Widget Animation Object Reference))
   SetBrush(self=Get Icon() -> (Icon:Image Object Reference), InBrush=Get RunLockNormal() -> (RunLockNormal:Slate Brush Structure))
   SetBrush(self=Get Icon(), InBrush=Get RunLockHighlight() -> (RunLockHighlight:Slate Brush Structure))
  ]}
 HighlightAutorunLockZone {nodes [
   n1 = HighlightAutorunLockZone!() -> (then, IsInZone:bool=~) {is=title K2Node_FunctionEntry}
   > n2 = Branch(Condition=@n1)
   > StopAllAnimations()
   > PlayAnimation(InAnimation=Get Anim_Autorun_Highlight() -> (Anim_Autorun_Highlight:Widget Animation Object Reference))
   StopAllAnimations(execute=@n2.else)
   > PlayAnimation(InAnimation=Get Anim_Autorun_Enter())
  ]}
]
variables [
 AutoRunAnimSpeed {type=number, default=1}
 RunLockHighlight {
  type=SlateBrush
  default {
   bIsDynamicallyLoaded=false
   drawAs=Image
   imageType=NoImage
   mirroring=NoMirror
   resourceName=None
   resourceObject=null
   tiling=NoTile
   imageSize=<182 182>
   margin=<ltrb 0 0 0 0>
   outlineSettings {
    bUseBrushTransparency=false
    roundingType=HalfHeightRadius
    width=0
    color {colorUseRule=UseColor_Specified, specifiedColor=<rgba 0 0 0 0>}
    cornerRadii=<0 0 0 1>
   }
   tintColor {colorUseRule=UseColor_Specified, specifiedColor=<rgba 1 1 1 1>}
   uVRegion {bIsValid=false, max=<0 0>, min=<0 0>}
  }
 }
 RunLockNormal {
  type=SlateBrush
  default {
   bIsDynamicallyLoaded=false
   drawAs=Image
   imageType=NoImage
   mirroring=NoMirror
   resourceName=None
   resourceObject=/Game/Athena/HUD/Atlas/Bacchus/Autorun_Normal_Sprite
   tiling=NoTile
   imageSize=<182 182>
   margin=<ltrb 0 0 0 0>
   outlineSettings {
    bUseBrushTransparency=false
    roundingType=HalfHeightRadius
    width=0
    color {colorUseRule=UseColor_Specified, specifiedColor=<rgba 0 0 0 0>}
    cornerRadii=<0 0 0 1>
   }
   tintColor {colorUseRule=UseColor_Specified, specifiedColor=<rgba 1 1 1 1>}
   uVRegion {bIsValid=false, max=<0 0>, min=<0 0>}
  }
 }
]
widgets {
 info {inheritedWidgetCount=0, parentClass=/Script/FortniteUI.FortAutorunLockZone, rootWidgetClass=&ScaleBox, widgetCount=4}
 widgets [
  ScaleBox {widgetClassPath=&ScaleBox, widget=ScaleBox}
  Overlay_0 {widgetClassPath=&Overlay, parent=ScaleBox, widget=Overlay_0, slot {class=ScaleBoxSlot}}
  Image_Ripple {
   widgetClassPath=&Image
   bIsVariable=true
   parent=Overlay_0
   widget=Image_Ripple
   properties {brush {
     bIsDynamicallyLoaded=false
     drawAs=Image
     imageType=FullColor
     mirroring=NoMirror
     resourceName=None
     resourceObject=/Game/Athena/HUD/Bacchus/Material/Autorun/MI_Autorun_ArcRipple
     tiling=NoTile
     imageSize=<32 32>
     margin=<ltrb 0 0 0 0>
     outlineSettings {
      bUseBrushTransparency=false
      roundingType=HalfHeightRadius
      width=0
      color {colorUseRule=UseColor_Specified, specifiedColor=<rgba 0 0 0 0>}
      cornerRadii=<0 0 0 0>
     }
     tintColor {colorUseRule=UseColor_Specified, specifiedColor=<rgba 1 1 1 1>}
     uVRegion {bIsValid=false, max=<0 0>, min=<0 0>}
    }}
   slot {class=OverlaySlot, properties {horizontalAlignment=HAlign_Fill, verticalAlignment=VAlign_Fill, padding=<ltrb -56 -56 -56 -56>}}
  }
  Image_ImprovedAutorunLockZone {
   widgetClassPath=&Image
   bIsVariable=true
   parent=Overlay_0
   widget=Image_ImprovedAutorunLockZone
   properties {
    visibility=SelfHitTestInvisible
    brush {
     bIsDynamicallyLoaded=false
     drawAs=Image
     imageType=FullColor
     mirroring=NoMirror
     resourceName=None
     resourceObject=/Game/Athena/HUD/Bacchus/Material/Autorun/M_Autorun_Button_Inst_Dark
     tiling=NoTile
     imageSize=<128 128>
     margin=<ltrb 0 0 0 0>
     outlineSettings {
      bUseBrushTransparency=false
      roundingType=HalfHeightRadius
      width=0
      color {colorUseRule=UseColor_Specified, specifiedColor=<rgba 0 0 0 0>}
      cornerRadii=<0 0 0 0>
     }
     tintColor {colorUseRule=UseColor_Specified, specifiedColor=<rgba 1 1 1 1>}
     uVRegion {bIsValid=false, max=<0 0>, min=<0 0>}
    }
    renderTransform {angle=0, scale=<1.95 1.95>, shear=<0 0>, translation=<0 0>}
   }
   slot {class=OverlaySlot}
  }
 ]
}
