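# Widget Blueprint for the Bacchus HUD's autorun lock zone button, built on the native FortAutorunLockZoneButton class.
# A root ScaleBox with user-specified stretch wraps the AutorunLockZone widget, which starts collapsed with autorun lock enabled.
# Background image and tutorial highlight entries are also listed.
# The event graph holds only unconnected PreConstruct, Construct and Tick events, so the Blueprint adds no logic of its own.
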WidgetBlueprint /Game/Athena/HUD/Bacchus/BacchusAutorunLockZone
graphs [EventGraph {nodes [
   Event PreConstruct()
   Event Construct()
   Event Tick()
  ]}]
variables []
widgets {
 info {inheritedWidgetCount=4, parentClass=/Script/FortniteUI.FortAutorunLockZoneButton, rootWidgetClass=&ScaleBox, widgetCount=4}
 widgets [
  ScaleBox_HUDWidgetScale {widgetClassPath=&ScaleBox, bInherited=true, widget=ScaleBox_HUDWidgetScale, properties {stretch=UserSpecified}}
  AutorunLockZone {
   widgetClassPath=&/AutorunLockZone
   bInherited=true
   bIsVariable=true
   parent=ScaleBox_HUDWidgetScale
   widget=AutorunLockZone
   properties {bIsAutorunLockEnabled=true, visibility=Collapsed}
   slot {class=ScaleBoxSlot}
  }
  Image_Background {widgetClassPath=&Image, bInherited=true, widget=null}
  TutorialHighlight {widgetClassPath=&TutorialHighlight}
 ]
}
