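# BacchusFPSTicker is a Widget Blueprint in the Bacchus HUD that provides the layout for an FPS readout with min and max figures.
# It derives from the native AthenaFPSBase and has an empty event graph, so it adds no logic of its own.
# A fixed-height SizeBox centers the inherited FPS text block, styled in bold Noto Sans with placeholder text.
# The inherited ThermalWarning text binding has no widget in this layout.
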WidgetBlueprint /Game/Athena/HUD/Bacchus/BacchusFPSTicker
graphs [EventGraph {links [], nodes []}]
variables []
widgets {
 info {inheritedWidgetCount=2, parentClass=/Script/FortniteUI.AthenaFPSBase, rootWidgetClass=&SizeBox, widgetCount=3}
 widgets [
  SizeBox_24 {widgetClassPath=&SizeBox, widget=SizeBox_24, properties {bOverride_HeightOverride=true, heightOverride=55}}
  FPS {
   widgetClassPath=&CommonTextBlock
   bInherited=true
   bIsVariable=true
   parent=SizeBox_24
   widget=FPS
   properties {
    justification=Center
    minDesiredWidth=200
    style=/FortUILibrary/Styles/Typography/Foundation/TextStyle-Base-S-B.TextStyle-Base-S-B_C
    text="120 FPS [ MIN,  MAX ]"
    font {
     bForceMonospaced=false
     bMaterialIsStencil=false
     fontMaterial=null
     fontObject=&NotoSans
     letterSpacing=0
     monospacedWidth=1
     size=18
     skewAmount=0
     typefaceFontName=NotoSans-Bold
     outlineSettings {
      bApplyOutlineToDropShadows=false
      bMiteredCorners=false
      bSeparateFillAlpha=false
      outlineMaterial=null
      outlineSize=0
      outlineColor=<rgba 0 0 0 1>
     }
    }
    shadowOffset=<0 0>
   }
   slot {class=SizeBoxSlot, properties {horizontalAlignment=HAlign_Center, verticalAlignment=VAlign_Center}}
  }
  ThermalWarning {widgetClassPath=&CommonTextBlock, bInherited=true, widget=null}
 ]
}
