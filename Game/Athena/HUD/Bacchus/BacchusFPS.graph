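# BacchusFPS is a Bacchus HUD element widget that hosts an FPS ticker display.
# The inherited HUD scale box holds an overlay that layers a translucent black, rounded border behind a BacchusFPSTicker child widget.
# The graph contains only a Construct event and no variables, so any behavior comes from the BacchusHUDElement parent class and the ticker widget.
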WidgetBlueprint /Game/Athena/HUD/Bacchus/BacchusFPS
graphs [EventGraph {nodes [Event Construct()]}]
variables []
widgets {
 info {inheritedWidgetCount=3, parentClass=/Script/FortniteUI.BacchusHUDElement, rootWidgetClass=&ScaleBox, widgetCount=6}
 widgets [
  ScaleBox_HUDWidgetScale {widgetClassPath=&ScaleBox, bInherited=true, widget=ScaleBox_HUDWidgetScale, properties {stretch=UserSpecified}}
  Overlay_30 {widgetClassPath=&Overlay, parent=ScaleBox_HUDWidgetScale, widget=Overlay_30, slot {class=ScaleBoxSlot}}
  Border_background {
   widgetClassPath=&Border
   parent=Overlay_30
   widget=Border_background
   properties {
    visibility=SelfHitTestInvisible
    background {
     bIsDynamicallyLoaded=false
     drawAs=RoundedBox
     imageType=NoImage
     mirroring=NoMirror
     resourceName=None
     resourceObject=null
     tiling=NoTile
     imageSize=<32 32>
     margin=<ltrb 0 0 0 0>
     outlineSettings {
      bUseBrushTransparency=false
      roundingType=FixedRadius
      width=0
      color {colorUseRule=UseColor_Specified, specifiedColor=<rgba 0 0 0 0>}
      cornerRadii=<5 5 5 5>
     }
     tintColor {colorUseRule=UseColor_Specified, specifiedColor=<rgba 1 1 1 1>}
     uVRegion {bIsValid=false, max=<0 0>, min=<0 0>}
    }
    brushColor=<rgba 0 0 0 0.2>
    padding=<ltrb 0 0 0 0>
   }
   slot {class=OverlaySlot, properties {horizontalAlignment=HAlign_Fill, verticalAlignment=VAlign_Fill}}
  }
  BacchusFPSTicker {
   widgetClassPath=&/BacchusFPSTicker
   bIsVariable=true
   parent=Overlay_30
   widget=BacchusFPSTicker
   slot {class=OverlaySlot, properties {horizontalAlignment=HAlign_Fill, verticalAlignment=VAlign_Fill, padding=<ltrb 0 -5 5 -5>}}
  }
  Image_Background {widgetClassPath=&Image, bInherited=true, widget=null}
  TutorialHighlight {widgetClassPath=&TutorialHighlight}
 ]
}
