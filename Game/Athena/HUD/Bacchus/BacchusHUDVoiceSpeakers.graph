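# Mobile (Bacchus) HUD widget, derived from FortBacchusSpeakerWrapper, that hosts the voice channel speaker container in a scale box.
# On Construct, a minimum-size SizeBox, named as a mobile canvas bug workaround, shows or collapses per the bEnableVoiceSpeakerWidget runtime option.
# A bottom-padded spacer becomes visible when the HUD context starts the post-game menu.
# Construct binds that event and Destruct unbinds it, both collapsing the spacer.
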WidgetBlueprint /Game/Athena/HUD/Bacchus/BacchusHUDVoiceSpeakers
graphs [EventGraph {nodes [
   n1 = OnPostMatchScreen() {is=title K2Node_CustomEvent}
   n2 = Get Spacer_PostMatchPadding() -> (Spacer_PostMatchPadding:Spacer Object Reference)
   > SetVisibility(self=@n2, InVisibility=SelfHitTestInvisible)
   Event Construct()
   n3 = GetRuntimeOptions()
   > SetVisibility(
    self=Get SizeBox_ReserveSizeMobileCanvasBug_DoNotDelete() -> (SizeBox_ReserveSizeMobileCanvasBug_DoNotDelete:Size Box Object Reference)
    InVisibility=Select(
     Option 0=Collapsed
     Option 1=SelfHitTestInvisible
     Index=Get bEnableVoiceSpeakerWidget!(self:Fort Runtime Options Object Reference=@n3) -> (bEnableVoiceSpeakerWidget:bool)
    )
   )
   n4 = Get AthenaHUDContext!() -> (ReturnValue:Athena HUDContext Object Reference) {is=title K2Node_GetBlueprintContextEx}
   > n5 = Bind Event to On Start Post Game Menu(self:Athena HUDContext Object Reference=@n4, Delegate=@n1) {is=title K2Node_AddDelegate}
   SetVisibility(execute=[@n5.then, @n6.then], self=@n2, InVisibility=Collapsed)
   Event Destruct()
   > n6 = Unbind Event from On Start Post Game Menu(self:Athena HUDContext Object Reference=@n4, Delegate=@n1) {is=title K2Node_RemoveDelegate}
  ]}]
variables []
widgets {
 info {inheritedWidgetCount=4, parentClass=/Script/FortniteUI.FortBacchusSpeakerWrapper, rootWidgetClass=&ScaleBox, widgetCount=8}
 widgets [
  ScaleBox_HUDWidgetScale {widgetClassPath=&ScaleBox, bInherited=true, widget=ScaleBox_HUDWidgetScale, properties {stretch=UserSpecified}}
  VerticalBox_0 {widgetClassPath=&VerticalBox, parent=ScaleBox_HUDWidgetScale, widget=VerticalBox_0, slot {class=ScaleBoxSlot}}
  Overlay_0 {
   widgetClassPath=&Overlay
   parent=VerticalBox_0
   widget=Overlay_0
   slot {class=VerticalBoxSlot, properties {horizontalAlignment=HAlign_Left, verticalAlignment=VAlign_Top}}
  }
  WrappedSpeakers {
   widgetClassPath=/Game/UI/Social/FortVoiceChannelSpeakerContainer
   bInherited=true
   bIsVariable=true
   parent=Overlay_0
   widget=WrappedSpeakers
   slot {class=OverlaySlot}
  }
  Spacer_PostMatchPadding {
   widgetClassPath=&Spacer
   bIsVariable=true
   parent=Overlay_0
   widget=Spacer_PostMatchPadding
   properties {visibility=Collapsed}
   slot {class=OverlaySlot, properties {horizontalAlignment=HAlign_Fill, verticalAlignment=VAlign_Bottom, padding=<ltrb 0 0 0 500>}}
  }
  SizeBox_ReserveSizeMobileCanvasBug_DoNotDelete {
   widgetClassPath=&SizeBox
   bIsVariable=true
   parent=Overlay_0
   widget=SizeBox_ReserveSizeMobileCanvasBug_DoNotDelete
   properties {bOverride_MinDesiredHeight=true, bOverride_MinDesiredWidth=true, minDesiredHeight=32, minDesiredWidth=512}
   slot {class=OverlaySlot, properties {horizontalAlignment=HAlign_Fill, verticalAlignment=VAlign_Fill}}
  }
  Image_Background {widgetClassPath=&Image, bInherited=true, widget=null}
  TutorialHighlight {widgetClassPath=&TutorialHighlight}
 ]
}
