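# Touch interaction button widget for the Bacchus HUD, built on FortTouchInteractionButton, that animates an interaction indicator as focus changes.
# A fixed-size box holds a scaled overlay with a collapsed glow and the inherited icon image, both drawn by indicator materials.
# OnInteractionFocusChanged switches between focusOn and focusOff animations, which drive ring radii and icon scale through material parameters.
# A radial progress animation keys Progress and TimerOpacity, but this graph never plays it.
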WidgetBlueprint /Game/Athena/HUD/Bacchus/BacchusInteractionButton
animations [
 focusOn {
  binding_count=1
  length_seconds=0.15
  object=&self:focusOn
  sequence {bindings [Image_Icon {
     id=#1
     possessed_class=Image
     tracks [{
       type=MovieSceneWidgetMaterialTrack
       sections [{
         type=MovieSceneParameterSection
         end_frame=3
         start_frame=0
         channels [
          OuterRingRadius {type=MovieSceneScriptingFloatChannel, keys [[0,0.29],[2.4933,0.43]]}
          InnerRingRadius {type=MovieSceneScriptingFloatChannel, keys [[0,0.03],[2.4933,0.07]]}
          AnimatedIconScale {type=MovieSceneScriptingFloatChannel, keys [[0,0.65],[2.1173,1]]}
         ]
        }]
      }]
    }]}
 }
 focusOff {
  binding_count=1
  length_seconds=0.05
  object=&self:focusOff
  sequence {bindings [Image_Icon {
     id=#2
     possessed_class=Image
     tracks [{
       type=MovieSceneWidgetMaterialTrack
       sections [{
         type=MovieSceneParameterSection
         end_frame=1
         start_frame=0
         channels [
          OuterRingRadius {type=MovieSceneScriptingFloatChannel, const [0.29,1]}
          InnerRingRadius {type=MovieSceneScriptingFloatChannel, const [0.03,1]}
          AnimatedIconScale {type=MovieSceneScriptingFloatChannel, const [0.65,1]}
         ]
        }]
      }]
    }]}
 }
 Anim_RadialProgress {
  binding_count=1
  length_seconds=1.0045
  object=&self:Anim_RadialProgress
  sequence {bindings [Image_Icon {
     id=#3
     possessed_class=Image
     tracks [{
       type=MovieSceneWidgetMaterialTrack
       sections [{
         type=MovieSceneParameterSection
         end_frame=20
         start_frame=0
         channels [
          TimerOpacity {type=MovieSceneScriptingFloatChannel, keys [[0,1],[20,0]]}
          Progress {type=MovieSceneScriptingFloatChannel, keys [[0,0],[20,1]]}
         ]
        }]
      }]
    }]}
 }
]
graphs [EventGraph {nodes [
   n1 = Event OnInteractionFocusChanged!() -> (OutputDelegate, then, bNewFocused:bool) {is=title K2Node_Event}
   > n2 = Branch(Condition=@n1.bNewFocused)
   > StopAnimation(InAnimation=Get focusOff() -> (focusOff:Widget Animation Object Reference))
   > PlayAnimation(InAnimation=Get focusOn() -> (focusOn:Widget Animation Object Reference))
   StopAnimation(execute=@n2.else, InAnimation=Get focusOn())
   > PlayAnimation(InAnimation=Get focusOff())
   Event PreConstruct()
   Event Construct()
   Event Tick()
  ]}]
variables []
widgets {
 info {inheritedWidgetCount=1, parentClass=/Script/FortniteUI.FortTouchInteractionButton, rootWidgetClass=&SizeBox, widgetCount=5}
 widgets [
  SizeBox_0 {
   widgetClassPath=&SizeBox
   widget=SizeBox_0
   properties {bOverride_HeightOverride=true, bOverride_WidthOverride=true, heightOverride=100, maxDesiredHeight=900, maxDesiredWidth=75, widthOverride=100}
  }
  ScaleBox_0 {widgetClassPath=&ScaleBox, parent=SizeBox_0, widget=ScaleBox_0, slot {class=SizeBoxSlot}}
  Overlay_0 {widgetClassPath=&Overlay, parent=ScaleBox_0, widget=Overlay_0, slot {class=ScaleBoxSlot}}
  Image_IconGlow {
   widgetClassPath=&Image
   bIsVariable=true
   parent=Overlay_0
   widget=Image_IconGlow
   properties {
    visibility=Collapsed
    brush {
     bIsDynamicallyLoaded=false
     drawAs=Image
     imageType=NoImage
     mirroring=NoMirror
     resourceName=None
     resourceObject=/Game/Athena/HUD/Materials/M_UI_TouchInteractionIndicator_Inst_Glow
     tiling=NoTile
     imageSize=<128 128>
     margin=<ltrb 0 0 0 0>
     outlineSettings {
      bUseBrushTransparency=false
      roundingType=HalfHeightRadius
      width=0
      color {colorUseRule=UseColor_Specified, specifiedColor=<rgba 0 0 0 0>}
      cornerRadii=<0 0 0 0>
     }
     tintColor {colorUseRule=UseColor_Specified, specifiedColor=<rgba 1 1 1 0.0762>}
     uVRegion {bIsValid=false, max=<0 0>, min=<0 0>}
    }
    renderTransform {angle=0, scale=<1.3 1.3>, shear=<0 0>, translation=<0 0>}
   }
   slot {class=OverlaySlot, properties {horizontalAlignment=HAlign_Center, verticalAlignment=VAlign_Center}}
  }
  Image_Icon {
   widgetClassPath=&Image
   bInherited=true
   bIsVariable=true
   parent=Overlay_0
   widget=Image_Icon
   properties {
    brush {
     bIsDynamicallyLoaded=false
     drawAs=Image
     imageType=NoImage
     mirroring=NoMirror
     resourceName=None
     resourceObject=/Game/Athena/HUD/Materials/M_UI_TouchInteractionIndicator_Inst
     tiling=NoTile
     imageSize=<128 128>
     margin=<ltrb 0 0 0 0>
     outlineSettings {
      bUseBrushTransparency=false
      roundingType=HalfHeightRadius
      width=0
      color {colorUseRule=UseColor_Specified, specifiedColor=<rgba 0 0 0 0>}
      cornerRadii=<0 0 0 0>
     }
     tintColor {colorUseRule=UseColor_Specified, specifiedColor=<rgba 1 1 1 1>}
     uVRegion {bIsValid=false, max=<0 0>, min=<0 0>}
    }
    renderTransform {angle=0, scale=<1.6 1.6>, shear=<0 0>, translation=<0 0>}
   }
   slot {class=OverlaySlot, properties {horizontalAlignment=HAlign_Center, verticalAlignment=VAlign_Center}}
  }
 ]
}
