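# BacchusKillfeedItem is a HUD element widget that displays a single kill feed entry as rich text.
# Its SetText function passes the incoming text straight to the KillText rich text block.
# A kill feed style set supplies tags for killer, victim, team and local-player names, and image and texture decorators allow inline images, as the placeholder text demonstrates.
# An inherited TutorialHighlight widget is also present.
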WidgetBlueprint /Game/Athena/HUD/Bacchus/BacchusKillfeedItem
graphs [
 EventGraph {links [], nodes []}
 SetText {nodes [
   n1 = SetText!() -> (then, Text:Text) {is=title K2Node_FunctionEntry}
   > SetText(self=Get KIllText() -> (KIllText:Rich Text Block Object Reference), InText=@n1) {is=title K2Node_CallFunction}
  ]}
]
variables []
widgets {
 info {inheritedWidgetCount=1, parentClass=&FortHUDElementWidget, rootWidgetClass=/Script/UMG.RichTextBlock, widgetCount=2}
 widgets [
  KillText {
   widgetClassPath=&RichTextBlock
   bIsVariable=true
   widget=KillText
   properties {
    text="(<Killer>Killer</>|<TeamKiller>Team Killer</>|<LocalKiller>Local Killer</>) killed (<Victim>Victim</>|<TeamVictim>Team Victim</>) with a {weapon} <img id=\"Cube\" width=\"100\" stretch=\"Fill\"></>"
    textStyleSet=&/BacchusKillFeedStyles
    decoratorClasses [/Game/Athena/HUD/TextStyles/KillFeedImageDecorator.KillFeedImageDecorator_C, /Script/FortniteUI.FortRichTextTextureDecorator]
   }
  }
  TutorialHighlight {widgetClassPath=&TutorialHighlight}
 ]
}
