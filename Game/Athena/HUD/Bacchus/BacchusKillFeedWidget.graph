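# Bacchus HUD kill feed widget, derived from BacchusKillFeed, showing Athena game state kill feed text in five stacked item widgets.
# Construct caches FortGameStateAthena, binds its Kill Feed Updated event and runs UpdateKillFeed; Destruct unbinds.
# UpdateKillFeed fills and shows items within both KillFeedEntry bounds and NumberOfActiveLines, collapsing the rest.
# It then passes bAnyActiveContent, which this graph only ever sets true, to SetHasActiveContent.
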WidgetBlueprint /Game/Athena/HUD/Bacchus/BacchusKillFeedWidget
graphs [
 EventGraph {nodes [
   Event Destruct()
   n1 = Create Event()
   > Unbind Event from Kill Feed Updated(self:Fort Game State Athena Object Reference=Get GameState() -> (GameState:Fort Game State Athena Object Reference), Delegate=@n1) {is=title K2Node_RemoveDelegate}
   Event Construct()
   > n2 = Cast To FortGameStateAthena!(execute, Object=GetGameState!() -> (ReturnValue:Game State Base Object Reference)) -> (then, CastFailed, AsFort Game State Athena:Fort Game State Athena Object Reference) {is=title K2Node_DynamicCast}
   > n3 = Set GameState(GameState:Fort Game State Athena Object Reference=@n2) -> (Output_Get:Fort Game State Athena Object Reference)
   > Assign Kill Feed Updated!(execute:Exec, self:Fort Game State Athena Object Reference=@n3, Delegate:"Delegate (by ref)"=@n1) -> (then:Exec) {is=title K2Node_AssignDelegate}
   > UpdateKillFeed()
  ]}
 UpdateKillFeed {nodes [
   UpdateKillFeed() {is=title K2Node_FunctionEntry}
   > n1 = For Each Loop(Array:Array of Bacchus Killfeed Item Object References=Make Array!(
     "[0]":Bacchus Killfeed Item Object Reference=Get BacchusKillfeedItem() -> (BacchusKillfeedItem:Bacchus Killfeed Item Object Reference)
     "[1]":Bacchus Killfeed Item Object Reference=Get BacchusKillfeedItem2() -> (BacchusKillfeedItem2:Bacchus Killfeed Item Object Reference)
     "[2]":Bacchus Killfeed Item Object Reference=Get BacchusKillfeedItem3() -> (BacchusKillfeedItem3:Bacchus Killfeed Item Object Reference)
     "[3]":Bacchus Killfeed Item Object Reference=Get BacchusKillfeedItem4() -> (BacchusKillfeedItem4:Bacchus Killfeed Item Object Reference)
     "[4]":Bacchus Killfeed Item Object Reference=Get BacchusKillfeedItem5() -> (BacchusKillfeedItem5:Bacchus Killfeed Item Object Reference)
    ) -> (Array:Array of Bacchus Killfeed Item Object References)) -> (Array Element:Bacchus Killfeed Item Object Reference)
   SetHasActiveContent!(execute=@n1.Completed, self:Fort Prioritized Widget Interface, bNewActive:bool=Get bAnyActiveContent() -> (bAnyActiveContent:bool)) -> (then)
   n2 = Get KillFeedEntry!(self:Fort Game State Athena Object Reference=Get GameState()) -> (KillFeedEntry:Array of Texts)
   n3 = Get NumberOfActiveLines() -> (NumberOfActiveLines:int=0)
   n4 = Branch(
    execute=@n1.LoopBody
    Condition=AND Boolean(
     A=Is Valid Index!(TargetArray:Array of Texts=@n2, IndexToTest:int=@n1.Array Index) -> (ReturnValue:bool)
     B="integer < integer"!(A:int=@n1.Array Index, B:int=@n3) -> (ReturnValue:bool) {is=title K2Node_CallFunction}
    )
   )
   > SetText!(
    execute
    self:Bacchus Killfeed Item Object Reference=@n1
    Text:Text="Get (a copy)"(Array:Array of Texts=@n2, Dimension 1=@n1.Array Index) -> (Output:Text)
   ) -> (then)
   > SetVisibility(self=@n1, InVisibility=Visible)
   > Set bAnyActiveContent(bAnyActiveContent:bool=true) -> (Output_Get:bool)
   SetVisibility(execute=@n4.else, self=@n1, InVisibility=Collapsed)
  ]}
]
variables [GameState {type=FortGameStateAthena, default=null}]
widgets {
 info {inheritedWidgetCount=4, parentClass=/Script/FortniteUI.BacchusKillFeed, rootWidgetClass=&ScaleBox, widgetCount=12}
 widgets [
  ScaleBox_HUDWidgetScale {widgetClassPath=&ScaleBox, bInherited=true, widget=ScaleBox_HUDWidgetScale, properties {stretch=UserSpecified}}
  SizeBox_0 {
   widgetClassPath=&SizeBox
   parent=ScaleBox_HUDWidgetScale
   widget=SizeBox_0
   properties {bOverride_WidthOverride=true, widthOverride=484}
   slot {class=ScaleBoxSlot}
  }
  MainSizeBox {
   widgetClassPath=&SizeBox
   bInherited=true
   parent=SizeBox_0
   widget=MainSizeBox
   properties {bOverride_MinDesiredWidth=true, heightOverride=150, maxDesiredWidth=600, minDesiredHeight=150, minDesiredWidth=700, widthOverride=600}
   slot {class=SizeBoxSlot}
  }
  VerticalBox_0 {widgetClassPath=&VerticalBox, parent=MainSizeBox, widget=VerticalBox_0, slot {class=SizeBoxSlot}}
  BacchusKillfeedItem {
   widgetClassPath=&/BacchusKillfeedItem
   bIsVariable=true
   parent=VerticalBox_0
   widget=BacchusKillfeedItem
   slot {class=VerticalBoxSlot}
  }
  BacchusKillfeedItem2 {
   widgetClassPath=&BacchusKillfeedItem
   bIsVariable=true
   parent=VerticalBox_0
   widget=BacchusKillfeedItem2
   slot {class=VerticalBoxSlot}
  }
  BacchusKillfeedItem3 {
   widgetClassPath=&BacchusKillfeedItem
   bIsVariable=true
   parent=VerticalBox_0
   widget=BacchusKillfeedItem3
   slot {class=VerticalBoxSlot}
  }
  BacchusKillfeedItem4 {
   widgetClassPath=&BacchusKillfeedItem
   bIsVariable=true
   parent=VerticalBox_0
   widget=BacchusKillfeedItem4
   slot {class=VerticalBoxSlot}
  }
  BacchusKillfeedItem5 {
   widgetClassPath=&BacchusKillfeedItem
   bIsVariable=true
   parent=VerticalBox_0
   widget=BacchusKillfeedItem5
   slot {class=VerticalBoxSlot}
  }
  Spacer_87 {
   widgetClassPath=&Spacer
   parent=VerticalBox_0
   widget=Spacer_87
   properties {size=<1 20>}
   slot {class=VerticalBoxSlot, properties {size {sizeRule=Fill, value=1}}}
  }
  Image_Background {widgetClassPath=&Image, bInherited=true, widget=null}
  TutorialHighlight {widgetClassPath=&TutorialHighlight}
 ]
}
