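# Athena HUD quickbar slot widget on BacchusQuickbarSlotBase, showing an item in a native cell with cooldown, durability, selection animation and rarity-tinted multi-slot indicators.
# Inventory-change events refresh the item, which updates cell size and indicators.
# Taps select or use the slot, with a rollback console variable choosing between StartSelectSlot/StopSelectSlot and SlotSelected/StopUseSlot.
# Drag-and-drop moves items between slots, and cancelled drags drop droppable items from inventory.
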WidgetBlueprint /Game/Athena/HUD/Bacchus/BacchusQuickBarSlot.BacchusQuickbarSlot
animations [Selected {
  binding_count=1
  display_label=""
  length_seconds=0.1
  object=&self:Selected
  sequence {bindings [Overlay_1 {
     id=#1
     possessed_class=Overlay
     tracks [{
       type=MovieScene2DTransformTrack
       sections [{
         type=MovieScene2DTransformSection
         end_frame=2
         start_frame=0
         channels [
          Translation.X {type=MovieSceneScriptingFloatChannel, const [0,2]}
          Translation.Y {type=MovieSceneScriptingFloatChannel, keys [[0,0],[2,-7]]}
          Angle {type=MovieSceneScriptingFloatChannel}
          Scale.X {type=MovieSceneScriptingFloatChannel}
          Scale.Y {type=MovieSceneScriptingFloatChannel}
          Shear.X {type=MovieSceneScriptingFloatChannel}
          Shear.Y {type=MovieSceneScriptingFloatChannel}
         ]
        }]
      }]
    }]}
 }]
graphs [
 Cancel Interaction {nodes [
   Cancel Interaction() {is=title K2Node_FunctionEntry}
   > Branch(Condition=Get bInteractionStarted() -> (bInteractionStarted:bool))
   > Set bInteractionStarted(bInteractionStarted:bool=false) -> (Output_Get:bool)
   > SetBrushTintColor!(
    execute
    self:Image Object Reference=Get SelectionHighlight() -> (SelectionHighlight:Image Object Reference)
    TintColor:Slate Color Structure=Get UntouchedColor() -> (UntouchedColor:Slate Color Structure)
   ) -> (then)
   > Branch(Condition=Get bHasOngoingUseAction() -> (bHasOngoingUseAction:bool))
   > n1 = GetConsoleVariableBoolValue!(execute, VariableName:String=Fort.Rollback.IndependentTapAnywherePrimaryAction) -> (then, ReturnValue:bool)
   > n2 = Branch(Condition=@n1)
   > StopSelectSlot(self:Athena Quick Bar Slot Object Reference)
   StopUseSlot(execute=@n2.else, self:Bacchus Quickbar Slot Base Object Reference)
  ]}
 End Interaction_BP {nodes [
   End Interaction_BP() {is=title K2Node_FunctionEntry}
   n1 = Get bDragDetected() -> (bDragDetected:bool)
   n2 = Get bIsLocked() -> (bIsLocked:bool)
   > Set DragAttempted(DragAttempted:bool=OR Boolean(A=IsDragDropping!() -> (ReturnValue:bool), B=AND Boolean(A=@n1, B=@n2))) -> (Output_Get:bool)
   > Set bDragDetected(bDragDetected:bool=false) -> (Output_Get:bool)
   > Set bDragOpStarted(bDragOpStarted:bool=false) -> (Output_Get:bool)
   > Branch(Condition=Get bInteractionStarted())
   > Cancel Interaction() {is=title K2Node_CallFunction}
   n3 = Get bIsLocked()
   n4 = Get DragAttempted() -> (DragAttempted:bool)
   > Branch(Condition=AND Boolean!(A=NOT Boolean(A=@n4), B=Get bSupportsDragAndDrop() -> (bSupportsDragAndDrop:bool), C:bool=NOT Boolean(A=@n3)) -> (ReturnValue))
   > n5 = GetConsoleVariableBoolValue(VariableName=Fort.Rollback.IndependentTapAnywherePrimaryAction)
   > n6 = Branch(Condition=@n5)
   > StartSelectSlot!(execute, self:Athena Quick Bar Slot Object Reference, bReloadOrUseIfAlreadySelected:bool=true, bStopUseImmediately:bool=true) -> (then)
   SlotSelected!(execute=@n6.else, self:Bacchus Quickbar Slot Base Object Reference, bStopUseImmediately:bool=true) -> (then)
  ]}
 EventGraph {nodes [
   n1 = Handle OnQuickbarContentsChanged!() -> (OutputDelegate, then, QuickbarIndex:EFortQuickBars Enum=Primary, ChangedSlots:Array of Integers) {is=title K2Node_CustomEvent}
   > RefreshItemIfSlotChanged!(
    execute
    self:Fort Quick Bar Slot Base Object Reference
    InQuickBar:EFortQuickBars Enum=@n1.QuickbarIndex
    InSlot:Array of Integers=@n1.ChangedSlots
   ) -> (then)
   n2 = SetAnimatedBoxScaleFactor!() -> (OutputDelegate, then, AnimatedBoxScaleFactor:double=~) {is=title K2Node_CustomEvent}
   > Set AnimatedBoxScaleFactor(AnimatedBoxScaleFactor:double=@n2.AnimatedBoxScaleFactor) -> (Output_Get:double)
   > UpdateItemBoxSize()
   Event OnRefreshItem() {is=title K2Node_Event}
   > UpdateItemBoxSize()
   > Setup Multi Slot Indicators()
   n3 = Event OnDragLeave!() -> (OutputDelegate, then, PointerEvent:Pointer Event Structure, Operation:Drag Drop Operation Object Reference) {is=title K2Node_Event}
   > SetHovered!(execute, self, NewHovered:bool=false) -> (then)
   > n4 = Cast To AthenaInventoryDragDrop!(execute, Object=@n3.Operation) -> (then, CastFailed, AsAthena Inventory Drag Drop:Athena Inventory Drag Drop Object Reference) {is=title K2Node_DynamicCast}
   > ShowDropIcon!(execute, self:Athena Inventory Drag Drop Object Reference=@n4, Drop Action:AthenaDragDropAction Enum=NewEnumerator1) -> (then)
   Event PreConstruct()
   > UpdateItemBoxSize()
   > n5 = Switch on EFortQuickBars!(execute, Selection:EFortQuickBars Enum=Get QuickBarType() -> (QuickBarType:EFortQuickBars Enum=Primary)) -> (Primary:Exec, Secondary:Exec, Creative:Exec, Max_None:Exec) {is=title K2Node_SwitchEnum}
   > SetVisibility(
    self=Get KeybindWidget!(self:Fort Quickbar Keybind Widget Object Reference=Get QuickbarKeybindWidgetTop() -> (QuickbarKeybindWidgetTop:Fort Quickbar Keybind Widget Object Reference)) -> (KeybindWidget:Fort Keybind Widget Object Reference)
    InVisibility=Collapsed
   )
   SetVisibility(
    execute=@n5.Secondary
    self=Get KeybindWidget(self=Get QuickbarKeybindWidgetBottom() -> (QuickbarKeybindWidgetBottom:Fort Quickbar Keybind Widget Object Reference))
    InVisibility=Collapsed
   )
   Event Construct()
   > Set bSupportsWideItems!(
    execute
    bSupportsWideItems:bool=Get Can Display Wide Items() -> (Can Display Wide Items:bool)
    self:Athena Quick Bar Native Cell Object Reference=Get NativeCell() -> (NativeCell:Athena Quick Bar Native Cell Object Reference)
   ) -> (then, Output_Get:bool)
   > InitializeCooldowns()
   > RefreshItem(self:Fort Quick Bar Slot Base Object Reference)
   n6 = Get SelectionHighlight()
   > SetBrushFromTexture!(
    execute
    self:Image Object Reference=[Get EmptyImage() -> (EmptyImage:Image Object Reference), @n6]
    Texture:Texture 2D Object Reference=Select!(
     NewEnumerator2:Texture 2D Object Reference=/Game/UI/InGame/Quickbar/Quickbar-Frame-Default
     NewEnumerator3:Texture 2D Object Reference=&Quickbar-Frame-Default
     NewEnumerator4:Texture 2D Object Reference
     NewEnumerator5:Texture 2D Object Reference=/Game/UI/InGame/Quickbar/Quickbar-Frame-Ability
     NewEnumerator6:Texture 2D Object Reference
     NewEnumerator7:Texture 2D Object Reference
     NewEnumerator8:Texture 2D Object Reference
     Index:ItemDisplayStyle Enum=Get ItemDisplayStyle() -> (ItemDisplayStyle:ItemDisplayStyle Enum=NewEnumerator2)
    ) -> (ReturnValue:Texture 2D Object Reference)
    bMatchSize:bool=false
   ) -> (then)
   > SetVisibility(self=@n6, InVisibility=Hidden)
   > Bind Event to On Quickbar Contents Changed(
    self:Fort Inventory Context Object Reference=Get FortInventoryContext!() -> (ReturnValue:Fort Inventory Context Object Reference) {is=title K2Node_GetBlueprintContextEx}
    Delegate=@n1
   ) {is=title K2Node_AddDelegate}
   n7 = Event OnDragCancelled!() -> (OutputDelegate, then, PointerEvent:"Pointer Event Structure (by ref)", Operation:Drag Drop Operation Object Reference) {is=title K2Node_Event}
   > Branch(Condition=IsDragItemDroppable!(Drag Drop Object:Drag Drop Operation Object Reference=@n7.Operation) -> (Is Droppable:bool))
   > n8 = Cast To AthenaInventoryDragDrop(Object=@n7.Operation)
   > n9 = Cast To FortWorldItem!(execute, Object=Get Payload!(self:Drag Drop Operation Object Reference=@n8) -> (Payload:Object Reference)) -> (then, CastFailed, AsFort World Item:Fort World Item Object Reference) {is=title K2Node_DynamicCast}
   > DropItem!(
    execute
    self:Fort Inventory Context Object Reference=Get FortInventoryContext()
    ItemBeingDropped:Fort World Item Object Reference=@n9
    Quantity:int=GetNumInStack!(self:Fort Item Object Reference=@n9) -> (ReturnValue:int=0)
   ) -> (then, ReturnValue:bool)
   Event OnMouseCaptureLost() {is=title K2Node_Event}
   n10 = Get bDragOpStarted() -> (bDragOpStarted:bool)
   > Branch(Condition=AND Boolean!(A=NOT Boolean(A=@n10), B=Get bInteractionStarted()) -> (ReturnValue))
   > End Interaction_BP() {is=title K2Node_CallFunction}
   n11 = Event OnExtensionWidgetAdded!() -> (OutputDelegate, then, ExtensionSlot:EUIExtensionSlot Enum=Invalid, ExtensionWidget:User Widget Object Reference) {is=title K2Node_Event}
   > n12 = AddChildToOverlay!(
    execute
    self:Overlay Object Reference=Get OverlayExtensionWidgets() -> (OverlayExtensionWidgets:Overlay Object Reference)
    Content:Widget Object Reference=@n11.ExtensionWidget
   ) -> (then, ReturnValue:Overlay Slot Object Reference)
   > SetHorizontalAlignment(self=@n12, InHorizontalAlignment=HAlign_Fill)
   > SetVerticalAlignment(self=@n12, InVerticalAlignment=VAlign_Fill)
   n13 = Event SetSlotSelected!() -> (OutputDelegate, then, bSelected:bool) {is=title K2Node_Event}
   > SetIsSelected!(execute, self:Athena Quick Bar Native Cell Object Reference=Get NativeCell(), bInSelected:bool=@n13.bSelected) -> (then)
   > Set bIsSelected(bIsSelected:bool=@n13.bSelected) -> (Output_Get:bool)
   > SetVisibility(
    self=Get SelectionHighlight()
    InVisibility=Select!(Option 0:ESlateVisibility Enum=Collapsed, Option 1:ESlateVisibility Enum=Visible, Index:bool=@n13.bSelected) -> (ReturnValue:ESlateVisibility Enum)
   )
   > n14 = Branch(Condition=@n13.bSelected)
   n15 = Get DoNotPlaySelectionAnimation() -> (DoNotPlaySelectionAnimation:bool)
   > n16 = Branch(Condition=@n15)
   PlayAnimation(execute=@n16.else, InAnimation=Get Selected() -> (Selected:Widget Animation Object Reference))
   n17 = Branch(execute=@n14.else, Condition=@n15)
   n18 = Get Selected()
   PlayAnimation(execute=@n17.else, InAnimation=@n18, StartAtTime=GetEndTime(self=@n18), PlayMode=Reverse)
   Event EndInteraction() {is=title K2Node_Event}
   > End Interaction_BP()
   Event StartOngoingUse() {is=title K2Node_Event}
   > StartOngoingUse_BP()
  ]}
 Handle Drop Item Confirmation Callback {nodes [
   n1 = Handle Drop Item Confirmation Callback!() -> (then, Chosen Quanitity:int, Item:Fort Item Object Reference) {is=title K2Node_FunctionEntry}
   > n2 = Cast To FortWorldItem(Object=@n1.Item)
   > DropItem(self=Get FortInventoryContext(), ItemBeingDropped=@n2, Quantity=@n1)
  ]}
 InitializeCooldowns {nodes [
   InitializeCooldowns() {is=title K2Node_FunctionEntry}
   > n1 = Cast To QuickbarSlotCooldown!(execute, Object=Get QuickbarSlotCooldown() -> (QuickbarSlotCooldown:Fort Item Cooldown Widget Object Reference)) -> (then, CastFailed, AsQuickbar Slot Cooldown:Quickbar Slot Cooldown Object Reference) {is=title K2Node_DynamicCast}
   > Initialize!(
    execute
    self:Quickbar Slot Cooldown Object Reference=@n1
    QuickbarItemDisplayStyle:ItemDisplayStyle Enum=Get ItemDisplayStyle() -> (ItemDisplayStyle=NewEnumerator2)
    DesiredCooldownTypesSupported:Array of EFortItemCooldownType Enums=Get CooldownTypesSupported() -> (CooldownTypesSupported:Array of EFortItemCooldownType Enums=None)
   ) -> (then)
  ]}
 IsDragItemDroppable {nodes [
   n1 = IsDragItemDroppable!() -> (then, Drag Drop Object:Drag Drop Operation Object Reference) {is=title K2Node_FunctionEntry}
   > n2 = Cast To AthenaInventoryDragDrop(Object=@n1)
   > n3 = Cast To FortWorldItem(Object=Get Payload(self=@n2))
   > Return Node!(execute, Is Droppable:bool=CanBeDropped!(self:Fort Item Object Reference=@n3) -> (ReturnValue:bool))
   Return Node(execute=@n3.CastFailed, Is Droppable=false)
   Return Node(execute=@n2.CastFailed, Is Droppable=false)
  ]}
 OnDragDetected {nodes [
   OnDragDetected!() -> (then, MyGeometry:Geometry Structure, PointerEvent:"Pointer Event Structure (by ref)") {is=title K2Node_FunctionEntry}
   > Set bDragDetected(bDragDetected=true)
   > n1 = Branch(Condition=Get bDragAndDropEnabled() -> (bDragAndDropEnabled:bool))
   n2 = Get Item() -> (Item:Fort Item Object Reference)
   > n3 = Branch(Condition=IsValid(Object=@n2))
   > n4 = Create Widget!(
    execute:Exec
    Class:User Widget Class Reference=/Game/Athena/HUD/Inventory/AthenaInventoryDragVisual.AthenaInventoryDragVisual_C
    OwningPlayer:Player Controller Object Reference=GetOwningPlayer()
    Item:Fort Item Object Reference=@n2
   ) -> (then:Exec, ReturnValue:Athena Inventory Drag Visual Object Reference)
   > n5 = "Create Drag && Drop Operation"!(
    execute:Exec
    Class:Drag Drop Operation Class Reference=/Game/Athena/HUD/Inventory/AthenaInventoryDragDrop.AthenaInventoryDragDrop_C
    Tag:String=FromQuickbar
    Payload:Object Reference=Get Item()
    DefaultDragVisual:Widget Object Reference=@n4
    Pivot:EDragPivot Enum=CenterCenter
    Offset:Vector 2D Structure=(X=0,Y=0)
    SplitTheStack:bool=False
    bIsFromCreativeInventory:bool=False
    InstigatorSlotIndex:int=Get SlotIndex() -> (SlotIndex:int=0)
   ) -> (then:Exec, ReturnValue:Athena Inventory Drag Drop Object Reference) {is=title K2Node_CreateDragDropOperation}
   > Set bDragOpStarted(bDragOpStarted=true)
   > Return Node!(execute, Operation:Drag Drop Operation Object Reference=@n5)
   Return Node(execute=[@n3.else, @n1.else])
  ]}
 OnDragOver {nodes [
   n1 = OnDragOver!() -> (then, MyGeometry:Geometry Structure, PointerEvent:Pointer Event Structure, Operation:Drag Drop Operation Object Reference) {is=title K2Node_FunctionEntry}
   > Branch(Condition=NOT Boolean(A=Get bIsLocked()))
   > n2 = Cast To AthenaInventoryDragDrop(Object=@n1.Operation)
   > n3 = Cast To FortItem!(execute, Object=Get Payload(self=@n1.Operation)) -> (then, CastFailed, AsFort Item:Fort Item Object Reference) {is=title K2Node_DynamicCast}
   > n4 = CanSwapItem!(
    execute
    self:Fort Inventory Context Object Reference=Get FortInventoryContext()
    Item:Fort Item Object Reference=@n3
    TargetQuickbar:EFortQuickBars Enum=Get QuickBarType() -> (QuickBarType=Primary)
    TargetSlot:int=Get SlotIndex() -> (SlotIndex=0)
   ) -> (then, ReturnValue:bool)
   > n5 = Branch(Condition=@n4)
   > ShowDropIcon(self=@n2, Drop Action=NewEnumerator2)
   > n6 = SetHovered(NewHovered=true)
   Return Node!(execute=[@n6.then, @n7.then], ReturnValue:bool=true)
   n7 = ShowDropIcon(execute=@n5.else, self=@n2, Drop Action=NewEnumerator3)
  ]}
 OnDrop {nodes [
   n1 = TakeItemSelectionInputFromHUD!(
    execute
    self:Athena HUDContext Object Reference=Get AthenaHUDContext!() -> (ReturnValue:Athena HUDContext Object Reference) {is=title K2Node_GetBlueprintContextEx}
    TouchIndex:int=0
    SlotIdx:int=Get SlotIndex() -> (SlotIndex=0)
    InQuickBar:EFortQuickBars Enum=Get QuickBarType() -> (QuickBarType=Primary)
    bReloadOrUseIfAlreadySelected:bool=true
    bStopUseImmediately:bool=true
   ) -> (then, bIsUsing:bool)
   Return Node(execute=[@n1.then, @n6.CastFailed, @n7.else, @n4.CastFailed], ReturnValue=true)
   n2 = OnDrop!() -> (then, MyGeometry:Geometry Structure, PointerEvent:Pointer Event Structure, Operation:Drag Drop Operation Object Reference) {is=title K2Node_FunctionEntry}
   > n3 = Branch(Condition=NOT Boolean(A=Get bIsLocked()))
   > SetHovered(NewHovered=false)
   > n4 = Cast To FortItem(Object=Get Payload(self=@n2.Operation))
   n5 = Get SlotIndex() -> (SlotIndex=0)
   > DropItemOnQuickBar!(
    execute
    self:Fort Inventory Context Object Reference=Get FortInventoryContext()
    Item:Fort Item Object Reference=@n4
    TargetQuickbar:EFortQuickBars Enum=Get QuickBarType() -> (QuickBarType=Primary)
    TargetSlot:int=@n5
   ) -> (then, ReturnValue:bool)
   > n6 = Cast To AthenaInventoryDragDrop(Object=@n2.Operation)
   > n7 = Branch(Condition="Equal (Integer)"(A=Get InstigatorSlotIndex!(self:Athena Inventory Drag Drop Object Reference=@n6) -> (InstigatorSlotIndex:int=0), B=@n5) {is=title K2Node_CallFunction})
   > n8 = GetConsoleVariableBoolValue(VariableName=Fort.Rollback.IndependentTapAnywherePrimaryAction)
   > n9 = Branch(Condition=@n8)
   > StartSelectSlot(bReloadOrUseIfAlreadySelected=true, bStopUseImmediately=true)
   > Return Node(ReturnValue=true)
   SlotSelected(execute=@n9.else, bStopUseImmediately=true)
   > Return Node(ReturnValue=true)
   Return Node(execute=@n3.else, ReturnValue=true)
  ]}
 SetHovered {nodes [
   n1 = SetHovered!() -> (then, NewHovered:bool=~) {is=title K2Node_FunctionEntry}
   > Branch(Condition="Not Equal (Boolean)"!(A:bool=@n1, B:bool=Get IsDragHovered() -> (IsDragHovered:bool)) -> (ReturnValue:bool) {is=title K2Node_CallFunction})
   > n2 = Set IsDragHovered(IsDragHovered:bool=@n1) -> (Output_Get:bool)
   > n3 = Branch(Condition=@n2)
   > n4 = SetVisibility(self=Get DropTarget() -> (DropTarget:Image Object Reference), InVisibility=HitTestInvisible)
   InvalidateLayoutAndVolatility(execute=[@n4.then, @n5.then], self:Widget Object Reference="Get [SizeBox] Item"() -> ("[SizeBox] Item":Size Box Object Reference))
   n5 = SetVisibility(execute=@n3.else, self=Get DropTarget(), InVisibility=Collapsed)
  ]}
 Setup Multi Slot Indicators {nodes [
   Setup Multi Slot Indicators() {is=title K2Node_FunctionEntry}
   > n1 = Branch(Condition=Get bUseSlotSizeBySlotUsage() -> (bUseSlotSizeBySlotUsage:bool))
   SetVisibility(
    execute=[@n9.else, @n3.Is Not Valid, @n1.then]
    self=Get Multi-SlotsIndicators() -> (Multi-SlotsIndicators:Horizontal Box Object Reference)
    InVisibility=Collapsed
   )
   n2 = Get Item()
   n3 = Is Valid(exec=@n1.else, InputObject=@n2)
   n4 = Make Array!(
    "[0]":Border Object Reference=Get IndicatorSlot1() -> (IndicatorSlot1:Border Object Reference)
    "[1]":Border Object Reference=Get IndicatorSlot2() -> (IndicatorSlot2:Border Object Reference)
    "[2]":Border Object Reference=Get IndicatorSlot3() -> (IndicatorSlot3:Border Object Reference)
    "[3]":Border Object Reference=Get IndicatorSlot4() -> (IndicatorSlot4:Border Object Reference)
    "[4]":Border Object Reference=Get IndicatorSlot5() -> (IndicatorSlot5:Border Object Reference)
   ) -> (Array:Array of Border Object References)
   > Set Indicator Slots(Indicator Slots:Array of Border Object References=@n4) -> (Output_Get:Array of Border Object References)
   n5 = FindItem!(
    self:Fort Player Controller Object Reference=GetFirstLocalFortPlayerController!(bIncludeNonSpectators:bool=true, bIncludeSpectators:bool=false) -> (ReturnValue:Fort Player Controller Object Reference)
    Item:Fort Item Object Reference=Get Item()
   ) -> (OutQuickBar:EFortQuickBars Enum=Primary, OutSlot:int=0, ReturnValue:bool)
   n6 = Length(TargetArray:Array of Border Object References=@n4)
   n7 = GetNumberOfSlotsTaken!(self:Fort Item Object Reference=@n2) -> (ReturnValue:Byte=0)
   n8 = "To Integer (Byte)"!(InByte:Byte=@n7) -> (ReturnValue:int=0)
   > n9 = Branch(Condition=AND Boolean!(
     A=@n5.ReturnValue
     B="integer > integer"(A=@n6, B=@n8) {is=title K2Node_CallFunction}
     C:bool="Byte > Byte"!(A:Byte=@n7, B:Byte=1) -> (ReturnValue:bool)
    ) -> (ReturnValue))
   > n10 = For Loop(LastIndex=int - int(A=@n8, B=1) {is=title K2Node_CallFunction})
   n11 = Get Indicator Slots() -> (Indicator Slots:Array of Border Object References)
   n12 = For Loop(execute=@n10.Completed, FirstIndex=@n8, LastIndex=int - int(A=Length(TargetArray=@n11), B=1))
   SetVisibility(execute=[@n13.then, @n12.Completed], self=Get Multi-SlotsIndicators(), InVisibility=HitTestInvisible)
   n13 = SetVisibility(
    execute=@n12.LoopBody
    self="Get (a copy)"(Array:Array of Border Object References=Get Indicator Slots(), Dimension 1=@n12) -> (Output:Border Object Reference)
    InVisibility=Hidden
   )
   n14 = Get SlotIndex() -> (SlotIndex=0)
   n15 = Branch(execute=@n10.LoopBody, Condition="Equal (Integer)"(A=int - int(A=@n14, B=@n5.OutSlot), B=@n10))
   n16 = "Get (a copy)"(Array=Get Indicator Slots(), Dimension 1=@n10)
   > n17 = SetBrushColor(self=@n16, InBrushColor=GetRarityColor!(self:Fort Item Object Reference=Get Item()) -> (ReturnValue:Linear Color Structure))
   SetVisibility(execute=[@n19.then, @n17.then], self=@n16, InVisibility=Visible)
   n18 = Break Linear Color!(LinearColor:"Linear Color Structure (by ref)"=GetRarityColor(self=Get Item())) -> (R:float=~, G:float=~, B:float=~, A:float=~) {is=title K2Node_BreakStruct}
   n19 = SetBrushColor(
    execute=@n15.else
    self=@n16
    InBrushColor=Make Linear Color!(R:float=@n18, G:float=@n18.G, B:float=@n18.B, A:float=0.3) -> (LinearColor:Linear Color Structure) {is=title K2Node_MakeStruct}
   )
  ]}
 StartOngoingUse_BP {nodes [
   StartOngoingUse_BP() {is=title K2Node_FunctionEntry}
   > n1 = Branch(Condition=Get bHasOngoingUseAction())
   > n2 = End Interaction_BP()
   n3 = GetConsoleVariableBoolValue(execute=[@n2.then, @n1.else], VariableName=Fort.Rollback.IndependentTapAnywherePrimaryAction)
   > n4 = Branch(Condition=@n3)
   > StartSelectSlot(bReloadOrUseIfAlreadySelected=true, bStopUseImmediately=false)
   SlotSelected(execute=@n4.else, bStopUseImmediately=false)
  ]}
 Update Item Box Size Native {nodes [
   Update Item Box Size Native() {is=title K2Node_FunctionEntry}
   n1 = Get Item()
   > n2 = Is Valid(InputObject=@n1)
   DesiredSizeOverride!(
    execute=[@n3.else, @n2.Is Not Valid]
    self:Athena Quick Bar Native Cell Object Reference=Get NativeCell()
    WidthOverride:float=88
    HeightOverride:float=88
   ) -> (then)
   n3 = Branch(execute=@n2.Is Valid, Condition=Get bUseSlotSizeBySlotUsage())
   n4 = GetSlotSizeForItem!(self:Fort Quick Bar Slot Base Object Reference, InItem:Fort Item Object Reference=@n1) -> (ReturnValue:Vector 2D Structure)
   n5 = BreakVector2D(InVec=@n4)
   n6 = BreakVector2D(InVec=vector2d * float!(A:Vector 2D Structure=@n4, B:double=0.7) -> (ReturnValue:Vector 2D Structure))
   n7 = Get AnimatedBoxScaleFactor() -> (AnimatedBoxScaleFactor:double)
   > DesiredSizeOverride(self=Get NativeCell(), WidthOverride=Lerp(A=@n6, B=@n5, Alpha=@n7), HeightOverride=Lerp(A=@n6.Y, B=@n5.Y, Alpha=@n7))
  ]}
 UpdateItemBoxSize {nodes [
   UpdateItemBoxSize() {is=title K2Node_FunctionEntry}
   > Update Item Box Size Native() {is=title K2Node_CallFunction}
  ]}
]
variables [
 AnimatedBoxScaleFactor {type=number, default=1}
 Can Display Wide Items {type=boolean}
 CooldownTypesSupported {type="array<EFortItemCooldownType>", default []}
 IsDragHovered {type=boolean}
 ItemDisplayStyle {type=ItemDisplayStyle, default=World}
 ShouldCollapseItemWidgetBorder {type=boolean}
 Use Native {type=boolean}
]
widgets {
 info {inheritedWidgetCount=15, parentClass=/Script/FortniteUI.BacchusQuickbarSlotBase, rootWidgetClass=&Overlay, widgetCount=29}
 widgets [
  Overlay_0 {widgetClassPath=&Overlay, widget=Overlay_0, properties {visibility=Visible}}
  Overlay_1 {widgetClassPath=&Overlay, parent=Overlay_0, widget=Overlay_1, slot {class=OverlaySlot}}
  "[SizeBox] Item" {
   widgetClassPath=&SizeBox
   bIsVariable=true
   parent=Overlay_1
   widget="[SizeBox] Item"
   slot {class=OverlaySlot, properties {horizontalAlignment=HAlign_Fill, verticalAlignment=VAlign_Fill}}
  }
  NativeCell {
   widgetClassPath=/Script/FortniteUI.AthenaQuickBarNativeCell
   bInherited=true
   bIsVariable=true
   parent="[SizeBox] Item"
   widget=NativeCell
   properties {
    selectionAnimLength_Seconds=0.1
    ammoTypeIcon_Padding=<4 8>
    ammoTypeIcon_Size=<18 18>
    emptyCell_Brush {
     bIsDynamicallyLoaded=false
     drawAs=Image
     imageType=NoImage
     mirroring=NoMirror
     resourceName=None
     resourceObject=&Quickbar-Frame-Default
     tiling=NoTile
     imageSize=<80 80>
     margin=<ltrb 0.5 0.5 0.5 0.5>
     outlineSettings {
      bUseBrushTransparency=false
      roundingType=HalfHeightRadius
      width=0
      color {colorUseRule=UseColor_Specified, specifiedColor=<rgba 0 0 0 0>}
      cornerRadii=<0 0 0 0>
     }
     tintColor {colorUseRule=UseColor_Specified, specifiedColor=<rgba 1 1 1 0.2>}
     uVRegion {bIsValid=false, max=<0 0>, min=<0 0>}
    }
    emptyCell_Padding=<0 0>
    gadgetFuel_Brush {
     bIsDynamicallyLoaded=false
     drawAs=Image
     imageType=NoImage
     mirroring=NoMirror
     resourceName=None
     resourceObject=/Game/Athena/UI/InGame/Jetpack/MI_Player_FuelBar
     tiling=NoTile
     imageSize=<6 32>
     margin=<ltrb 0.49 0.49 0.49 0.49>
     outlineSettings {
      bUseBrushTransparency=false
      roundingType=HalfHeightRadius
      width=0
      color {colorUseRule=UseColor_Specified, specifiedColor=<rgba 0 0 0 0>}
      cornerRadii=<0 0 0 0>
     }
     tintColor {colorUseRule=UseColor_Specified, specifiedColor=<rgba 1 1 1 1>}
     uVRegion {bIsValid=false, max=<0 0>, min=<0 0>}
    }
    gadgetFuel_Padding=<4 4>
    secondaryAmmoTypeIconBuildMode_Padding=<20 60>
    secondaryAmmoTypeIcon_Padding=<20 60>
    secondaryAmmoTypeText_Padding=<8 5>
    selectedCell_Brush {
     bIsDynamicallyLoaded=false
     drawAs=Image
     imageType=NoImage
     mirroring=NoMirror
     resourceName=None
     resourceObject=&Quickbar-Frame-Default
     tiling=NoTile
     imageSize=<24 24>
     margin=<ltrb 0.5 0.5 0.5 0.5>
     outlineSettings {
      bUseBrushTransparency=false
      roundingType=HalfHeightRadius
      width=0
      color {colorUseRule=UseColor_Specified, specifiedColor=<rgba 0 0 0 0>}
      cornerRadii=<0 0 0 0>
     }
     tintColor {colorUseRule=UseColor_Specified, specifiedColor=<rgba 1 1 1 1>}
     uVRegion {bIsValid=false, max=<0 0>, min=<0 0>}
    }
    selectionOffset=<0 0>
   }
   slot {class=SizeBoxSlot, properties {horizontalAlignment=HAlign_Center, verticalAlignment=VAlign_Center, padding=<ltrb 2 2 2 0>}}
  }
  Overlay_2 {widgetClassPath=&Overlay, parent=NativeCell, widget=Overlay_2, slot {class=PanelSlot}}
  SizeBox_2 {
   widgetClassPath=&SizeBox
   parent=Overlay_2
   widget=SizeBox_2
   properties {bOverride_HeightOverride=true, bOverride_WidthOverride=true, heightOverride=96, widthOverride=96}
   slot {class=OverlaySlot, properties {horizontalAlignment=HAlign_Fill, verticalAlignment=VAlign_Fill}}
  }
  QuickbarSlotCooldown {
   widgetClassPath=/Game/UI/InGame/HUD/QuickbarSlotCooldown
   bInherited=true
   bIsVariable=true
   parent=SizeBox_2
   widget=QuickbarSlotCooldown
   properties {visibility=HitTestInvisible}
   slot {class=SizeBoxSlot}
  }
  SizeBox_69 {
   widgetClassPath=&SizeBox
   parent=Overlay_2
   widget=SizeBox_69
   properties {bOverride_HeightOverride=true, bOverride_WidthOverride=true, heightOverride=96, widthOverride=96}
   slot {class=OverlaySlot, properties {horizontalAlignment=HAlign_Fill, verticalAlignment=VAlign_Fill}}
  }
  QuickbarSlotDurability {
   widgetClassPath=/Game/UI/InGame/HUD/QuickbarSlotDurability
   bInherited=true
   bIsVariable=true
   parent=SizeBox_69
   widget=QuickbarSlotDurability
   properties {visibility=SelfHitTestInvisible}
   slot {class=SizeBoxSlot}
  }
  SlotInteraction {
   widgetClassPath=&Image
   bInherited=true
   bIsVariable=true
   parent=Overlay_1
   widget=SlotInteraction
   properties {
    brush {
     bIsDynamicallyLoaded=false
     drawAs=Box
     imageType=NoImage
     mirroring=NoMirror
     resourceName=None
     resourceObject=/Game/UI/InGame/Quickbar/Icon-QuickbarSlot-InvalidDrop
     tiling=NoTile
     imageSize=<64 64>
     margin=<ltrb 0.25 0.25 0.25 0.25>
     outlineSettings {
      bUseBrushTransparency=false
      roundingType=HalfHeightRadius
      width=0
      color {colorUseRule=UseColor_Specified, specifiedColor=<rgba 0 0 0 0>}
      cornerRadii=<0 0 0 0>
     }
     tintColor {colorUseRule=UseColor_Specified, specifiedColor=<rgba 1 1 1 1>}
     uVRegion {bIsValid=false, max=<0 0>, min=<0 0>}
    }
    colorAndOpacity=<rgba 1 1 1 0.1>
   }
   slot {class=OverlaySlot, properties {horizontalAlignment=HAlign_Fill, verticalAlignment=VAlign_Fill, padding=<ltrb 8 8 8 8>}}
  }
  DropTarget {
   widgetClassPath=&Image
   bIsVariable=true
   parent=Overlay_1
   widget=DropTarget
   properties {
    visibility=Collapsed
    brush {
     bIsDynamicallyLoaded=false
     drawAs=Image
     imageType=NoImage
     mirroring=NoMirror
     resourceName=None
     resourceObject=/Game/Athena/HUD/Inventory/Styles/T-Icon-SwapSlot
     tiling=NoTile
     imageSize=<80 80>
     margin=<ltrb 0 0 0 0>
     outlineSettings {
      bUseBrushTransparency=false
      roundingType=HalfHeightRadius
      width=0
      color {colorUseRule=UseColor_Specified, specifiedColor=<rgba 0 0 0 0>}
      cornerRadii=<0 0 0 0>
     }
     tintColor {colorUseRule=UseColor_Specified, specifiedColor=<rgba 1 1 0.995 1>}
     uVRegion {bIsValid=false, max=<0 0>, min=<0 0>}
    }
   }
   slot {class=OverlaySlot, properties {horizontalAlignment=HAlign_Fill, verticalAlignment=VAlign_Fill}}
  }
  LockIcon {
   widgetClassPath=&Image
   bInherited=true
   bIsVariable=true
   parent=Overlay_1
   widget=LockIcon
   properties {
    visibility=Collapsed
    brush {
     bIsDynamicallyLoaded=false
     drawAs=Image
     imageType=NoImage
     mirroring=NoMirror
     resourceName=None
     resourceObject=/Game/UI/Foundation/Textures/Icons/Locks/T-Icon-Lock-64
     tiling=NoTile
     imageSize=<32 32>
     margin=<ltrb 0 0 0 0>
     outlineSettings {
      bUseBrushTransparency=false
      roundingType=HalfHeightRadius
      width=0
      color {colorUseRule=UseColor_Specified, specifiedColor=<rgba 0 0 0 0>}
      cornerRadii=<0 0 0 0>
     }
     tintColor {colorUseRule=UseColor_Specified, specifiedColor=<rgba 1 1 1 1>}
     uVRegion {bIsValid=false, max=<0 0>, min=<0 0>}
    }
   }
   slot {class=OverlaySlot}
  }
  OverlayExtensionWidgets {
   widgetClassPath=&Overlay
   bIsVariable=true
   parent=Overlay_1
   widget=OverlayExtensionWidgets
   slot {class=OverlaySlot, properties {horizontalAlignment=HAlign_Fill, verticalAlignment=VAlign_Fill}}
  }
  TutorialHighlight {
   widgetClassPath=&TutorialHighlight
   bIsVariable=true
   parent=Overlay_1
   widget=TutorialHighlight
   properties {customHighlightWidget=/Game/Athena/HUD/Tutorial/Tutorial_Outline_Square.Tutorial_Outline_Square_C}
   slot {class=OverlaySlot, properties {horizontalAlignment=HAlign_Fill, verticalAlignment=VAlign_Fill, padding=<ltrb -6 -6 -6 -6>}}
  }
  Multi-SlotsIndicators {
   widgetClassPath=&HorizontalBox
   bIsVariable=true
   parent=Overlay_1
   widget=Multi-SlotsIndicators
   properties {visibility=Collapsed}
   slot {class=OverlaySlot, properties {horizontalAlignment=HAlign_Center, verticalAlignment=VAlign_Bottom, padding=<ltrb 0 0 0 6>}}
  }
  IndicatorSlot1 {
   widgetClassPath=&Border
   bIsVariable=true
   parent=Multi-SlotsIndicators
   widget=IndicatorSlot1
   properties {
    background {
     bIsDynamicallyLoaded=false
     drawAs=Box
     imageType=NoImage
     mirroring=NoMirror
     resourceName=None
     resourceObject=null
     tiling=NoTile
     imageSize=<32 32>
     margin=<ltrb 0 0 0 0>
     outlineSettings {
      bUseBrushTransparency=false
      roundingType=HalfHeightRadius
      width=0
      color {colorUseRule=UseColor_Specified, specifiedColor=<rgba 0 0 0 0>}
      cornerRadii=<0 0 0 0>
     }
     tintColor {colorUseRule=UseColor_Specified, specifiedColor=<rgba 1 1 1 1>}
     uVRegion {bIsValid=false, max=<0 0>, min=<0 0>}
    }
    padding=<ltrb 8 8 0 0>
   }
   slot {class=HorizontalBoxSlot}
  }
  IndicatorSlot2 {
   widgetClassPath=&Border
   bIsVariable=true
   parent=Multi-SlotsIndicators
   widget=IndicatorSlot2
   properties {padding=<ltrb 8 8 0 0>}
   slot {class=HorizontalBoxSlot, properties {padding=<ltrb 8 0 0 0>}}
  }
  IndicatorSlot3 {
   widgetClassPath=&Border
   bIsVariable=true
   parent=Multi-SlotsIndicators
   widget=IndicatorSlot3
   properties {padding=<ltrb 8 8 0 0>}
   slot {class=HorizontalBoxSlot, properties {padding=<ltrb 8 0 0 0>}}
  }
  IndicatorSlot4 {
   widgetClassPath=&Border
   bIsVariable=true
   parent=Multi-SlotsIndicators
   widget=IndicatorSlot4
   properties {padding=<ltrb 8 8 0 0>}
   slot {class=HorizontalBoxSlot, properties {padding=<ltrb 8 0 0 0>}}
  }
  IndicatorSlot5 {
   widgetClassPath=&Border
   bIsVariable=true
   parent=Multi-SlotsIndicators
   widget=IndicatorSlot5
   properties {padding=<ltrb 8 8 0 0>}
   slot {class=HorizontalBoxSlot, properties {padding=<ltrb 8 0 0 0>}}
  }
  EmptyImage {
   widgetClassPath=&Image
   bInherited=true
   bIsVariable=true
   parent=Overlay_1
   widget=EmptyImage
   properties {
    visibility=Collapsed
    brush {
     bIsDynamicallyLoaded=false
     drawAs=NoDrawType
     imageType=NoImage
     mirroring=NoMirror
     resourceName=None
     resourceObject=null
     tiling=NoTile
     imageSize=<80 80>
     margin=<ltrb 0.25 0.25 0.25 0.25>
     outlineSettings {
      bUseBrushTransparency=false
      roundingType=HalfHeightRadius
      width=0
      color {colorUseRule=UseColor_Specified, specifiedColor=<rgba 0 0 0 0>}
      cornerRadii=<0 0 0 0>
     }
     tintColor {colorUseRule=UseColor_Specified, specifiedColor=<rgba 1 1 1 1>}
     uVRegion {bIsValid=false, max=<0 0>, min=<0 0>}
    }
    colorAndOpacity=<rgba 1 1 1 0.2>
   }
   slot {class=OverlaySlot, properties {horizontalAlignment=HAlign_Fill, verticalAlignment=VAlign_Fill, padding=<ltrb 6 6 6 6>}}
  }
  SelectionHighlight {
   widgetClassPath=&Image
   bInherited=true
   bIsVariable=true
   parent=Overlay_1
   widget=SelectionHighlight
   properties {
    visibility=Collapsed
    brush {
     bIsDynamicallyLoaded=false
     drawAs=NoDrawType
     imageType=NoImage
     mirroring=NoMirror
     resourceName=None
     resourceObject=null
     tiling=NoTile
     imageSize=<24 24>
     margin=<ltrb 0.5 0.5 0.5 0.5>
     outlineSettings {
      bUseBrushTransparency=false
      roundingType=HalfHeightRadius
      width=0
      color {colorUseRule=UseColor_Specified, specifiedColor=<rgba 0 0 0 0>}
      cornerRadii=<0 0 0 0>
     }
     tintColor {colorUseRule=UseColor_Specified, specifiedColor=<rgba 1 1 1 1>}
     uVRegion {bIsValid=false, max=<0 0>, min=<0 0>}
    }
   }
   slot {class=OverlaySlot, properties {horizontalAlignment=HAlign_Fill, verticalAlignment=VAlign_Fill}}
  }
  Image_Disabled {
   widgetClassPath=&Image
   bInherited=true
   bIsVariable=true
   parent=Overlay_1
   widget=Image_Disabled
   properties {
    visibility=Hidden
    brush {
     bIsDynamicallyLoaded=false
     drawAs=Image
     imageType=NoImage
     mirroring=NoMirror
     resourceName=None
     resourceObject=null
     tiling=NoTile
     imageSize=<32 32>
     margin=<ltrb 0 0 0 0>
     outlineSettings {
      bUseBrushTransparency=false
      roundingType=HalfHeightRadius
      width=0
      color {colorUseRule=UseColor_Specified, specifiedColor=<rgba 0 0 0 0>}
      cornerRadii=<0 0 0 0>
     }
     tintColor {colorUseRule=UseColor_Specified, specifiedColor=<rgba 0 0 0 0.4>}
     uVRegion {bIsValid=false, max=<0 0>, min=<0 0>}
    }
    renderTransform {angle=0, scale=<0.95 0.95>, shear=<0 0>, translation=<0 2>}
   }
   slot {class=OverlaySlot, properties {horizontalAlignment=HAlign_Fill, verticalAlignment=VAlign_Fill}}
  }
  SizeBox_BottomKeybind {widgetClassPath=&SizeBox, bInherited=true, widget=null}
  SizeBox_TopKeybind {widgetClassPath=&SizeBox, bInherited=true, widget=null}
  QuickbarKeybindWidgetBottom {widgetClassPath=/Script/FortniteUI.FortQuickbarKeybindWidget, bInherited=true, widget=null}
  QuickbarKeybindWidgetTop {widgetClassPath=&FortQuickbarKeybindWidget, bInherited=true, widget=null}
  ItemCardMaximized {widgetClassPath=/Script/FortniteUI.FortMultiSizeItemCard, bInherited=true, widget=null}
  ItemCardMinimized {widgetClassPath=&FortMultiSizeItemCard, bInherited=true, widget=null}
 ]
}
