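# BacchusDynamicQuickbar is a HUD quickbar widget, derived from QuickbarBase, that arranges its slots in a single grid row in either direction.
# Refresh loops over the quickbar slots, gets each one's grid slot and, if valid, sets it to row 0.
# The IsForward flag picks the column: the loop index for forward order, or a value counted back from the slot count for reverse order.
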WidgetBlueprint /Game/Athena/HUD/Bacchus/BacchusDynamicQuickbar
graphs [
 EventGraph {nodes [
   Event Construct()
   > Parent: Construct()
   Event PreConstruct()
   Event Tick()
  ]}
 Refresh {nodes [
   Refresh() {is=title K2Node_FunctionEntry}
   > n1 = For Each Loop(Array:Array of Fort Quick Bar Slot Base Object References=Get QuickbarSlots() -> (QuickbarSlots:Array of Fort Quick Bar Slot Base Object References)) -> (Array Element:Fort Quick Bar Slot Base Object Reference)
   > Set CurrentIndex(CurrentIndex:int=@n1.Array Index) -> (Output_Get:int=0)
   > n2 = Set CurrentGridSlot(CurrentGridSlot:Grid Slot Object Reference=SlotAsGridSlot(Widget=@n1)) -> (Output_Get:Grid Slot Object Reference)
   > Is Valid(InputObject=@n2)
   > SetRow!(execute, self:Grid Slot Object Reference=Get CurrentGridSlot() -> (CurrentGridSlot:Grid Slot Object Reference), InRow:int=0) -> (then)
   > n3 = Branch(Condition=Get IsForward() -> (IsForward:bool))
   > SetColumn!(execute, self:Grid Slot Object Reference=Get CurrentGridSlot(), InColumn:int=Get CurrentIndex() -> (CurrentIndex:int=0)) -> (then)
   n4 = Get QuickbarSlots()
   n5 = Get CurrentIndex() -> (CurrentIndex=0)
   SetColumn(
    execute=@n3.else
    self=Get CurrentGridSlot()
    InColumn=int - int(A=Length(TargetArray:Array of Fort Quick Bar Slot Base Object References=@n4), B=int + int(A=@n5))
   )
  ]}
]
variables [IsForward {type=boolean}]
widgets {info {inheritedWidgetCount=0, parentClass=/Game/UI/InGame/HUD/QuickbarBase.QuickbarBase_C, rootWidgetClass=null, widgetCount=0}, widgets []}
