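# Bacchus HUD widget for the primary quickbar: six slots plus a frame previewing where an interacted pickup would land.
# HUD interaction events store the pickup's dummy item; quickbar focus and content changes also refresh the preview.
# PreviewItemPickUp picks the slot, then the frame shifts to the lower grid column and widens for multi-slot items.
# It also records the last non-zero focused slot and disables invalidation-box caching on simplified HUD.
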WidgetBlueprint /Game/Athena/HUD/Bacchus/BacchusQuickbarPrimary
graphs [
 EventGraph {nodes [
   n1 = OnQuickBarSlotFocusChanged_Bind!() -> (OutputDelegate, then, QuickbarIndex:EFortQuickBars Enum=Primary, Slot:int) {is=title K2Node_CustomEvent}
   > UpdateItemPreview()
   > Branch(Condition=AND Boolean(
     A="Equal (Enum)"(A:EFortQuickBars Enum=@n1.QuickbarIndex, B:EFortQuickBars Enum=Primary)
     B="Not Equal (Integer)"!(A:int=@n1.Slot, B:int=0) -> (ReturnValue:bool) {is=title K2Node_CallFunction}
    ))
   > Set CurrentWeaponSlot(CurrentWeaponSlot:int=@n1.Slot) -> (Output_Get:int=0)
   n2 = Event OnQuickbarContentsChanged!() -> (OutputDelegate, then, QuickbarIndex:EFortQuickBars Enum=Primary, ChangedSlots:Array of Integers) {is=title K2Node_Event}
   > Parent: OnQuickbarContentsChanged!(execute, QuickbarIndex:EFortQuickBars Enum=@n2.QuickbarIndex, ChangedSlots:Array of Integers=@n2.ChangedSlots) -> (then) {is=title K2Node_CallParentFunction}
   > UpdateItemPreview()
   Event Construct()
   > n3 = Sequence!(execute) -> (then_0, then_1, then_2:Exec, then_3:Exec, then_4:Exec)
   Parent: Construct(execute=@n3.then_1)
   n4 = Create Event()
   n5 = Get FortHUDContext!() -> (ReturnValue:Fort HUDContext Object Reference) {is=title K2Node_GetBlueprintContextEx}
   Bind Event to On Player Can Interact(execute=@n3.then_4, self:Fort HUDContext Object Reference=@n5, Delegate=@n4) {is=title K2Node_AddDelegate}
   > Bind Event to On Interact Updated(self:Fort HUDContext Object Reference=@n5, Delegate=@n4) {is=title K2Node_AddDelegate}
   Set QuickbarSlots(
    execute=@n3.then_0
    QuickbarSlots:Array of Fort Quick Bar Slot Base Object References=Make Array!(
     "[0]":Fort Quick Bar Slot Base Object Reference=Get QuickbarSlot1() -> (QuickbarSlot1:Bacchus Quickbar Slot Object Reference)
     "[1]":Fort Quick Bar Slot Base Object Reference=Get QuickbarSlot2() -> (QuickbarSlot2:Bacchus Quickbar Slot Object Reference)
     "[2]":Fort Quick Bar Slot Base Object Reference=Get QuickbarSlot3() -> (QuickbarSlot3:Bacchus Quickbar Slot Object Reference)
     "[3]":Fort Quick Bar Slot Base Object Reference=Get QuickbarSlot4() -> (QuickbarSlot4:Bacchus Quickbar Slot Object Reference)
     "[4]":Fort Quick Bar Slot Base Object Reference=Get QuickbarSlot5() -> (QuickbarSlot5:Bacchus Quickbar Slot Object Reference)
     "[5]":Fort Quick Bar Slot Base Object Reference=Get QuickbarSlot6() -> (QuickbarSlot6:Bacchus Quickbar Slot Object Reference)
    ) -> (Array:Array of Fort Quick Bar Slot Base Object References)
   ) -> (Output_Get:Array of Fort Quick Bar Slot Base Object References)
   n6 = GetQuickbarFocus!(self:Fort Inventory Context Object Reference=Get FortInventoryContext!() -> (ReturnValue:Fort Inventory Context Object Reference) {is=title K2Node_GetBlueprintContextEx}) -> (OutQuickBar:EFortQuickBars Enum=Primary, OutSlot:int=0, OutSecondarySlot:int=0, OutPreviousFocusedSlot:int=0)
   OnQuickbarSlotFocusChanged!(execute=@n3.then_2, self:Fort Quickbar Base Object Reference, QuickbarIndex:EFortQuickBars Enum=@n6, SlotIndex:int=@n6.OutSlot) -> (then)
   Assign On Quickbar Slot Focus Changed!(
    execute:Exec=@n3.then_3
    self:Fort Inventory Context Object Reference=Get FortInventoryContext()
    Delegate:"Delegate (by ref)"=Create Event()
   ) -> (then:Exec) {is=title K2Node_AssignDelegate}
  ]}
 Get Slot Display Index {nodes [
   n1 = Get Slot Display Index!() -> (then, SlotIndex:int) {is=title K2Node_FunctionEntry}
   n2 = GetSpecificSlot!(self:Fort Quickbar Base Object Reference, SlotNumber:int=@n1) -> (ReturnValue:Fort Quick Bar Slot Base Object Reference)
   > Return Node!(execute, ColumnIndex:int=Get Column!(self:Grid Slot Object Reference=SlotAsGridSlot(Widget=@n2)) -> (Column:int=0))
  ]}
 GetSpecificSlot {nodes [
   n1 = GetSpecificSlot!() -> (then, SlotNumber:int=0) {is=title K2Node_FunctionEntry}
   n2 = Make Array(
    "[0]":Bacchus Quickbar Slot Object Reference=Get QuickbarSlot1()
    "[1]":Bacchus Quickbar Slot Object Reference=Get QuickbarSlot2()
    "[2]":Bacchus Quickbar Slot Object Reference=Get QuickbarSlot3()
    "[3]":Bacchus Quickbar Slot Object Reference=Get QuickbarSlot4()
    "[4]":Bacchus Quickbar Slot Object Reference=Get QuickbarSlot5()
    "[5]":Bacchus Quickbar Slot Object Reference=Get QuickbarSlot6()
   ) -> (Array:Array of Bacchus Quickbar Slot Object References)
   > Branch(Condition=Is Valid Index!(TargetArray:Array of Bacchus Quickbar Slot Object References=@n2, IndexToTest:int=@n1) -> (ReturnValue:bool))
   > Return Node!(
    execute
    ReturnValue:Fort Quick Bar Slot Base Object Reference="Get (a copy)"(Array:Array of Bacchus Quickbar Slot Object References=@n2, Dimension 1=@n1) -> (Output:Bacchus Quickbar Slot Object Reference)
   )
  ]}
 Handle Interaction Updated {nodes [
   n1 = Handle Interaction Updated!() -> (then, ContextInfo:Fort Interact Context Info Object Reference) {is=title K2Node_FunctionEntry}
   > n2 = Is Valid(InputObject=@n1)
   > n3 = Cast To FortPickup!(execute, Object=Get ReceivingActor!(self:Fort Interact Context Info Object Reference=@n1) -> (ReceivingActor:Actor Weak object)) -> (then, CastFailed, AsFort Pickup:Fort Pickup Object Reference) {is=title K2Node_DynamicCast}
   > n4 = Set Current Pickup Preview Item(Current Pickup Preview Item:Fort Item Object Reference=Get PrimaryPickupDummyItem!(self:Fort Pickup Object Reference=@n3) -> (PrimaryPickupDummyItem:Fort Item Object Reference)) -> (Output_Get:Fort Item Object Reference)
   UpdateItemPreview(execute=[@n4.then, @n5.then])
   n5 = Set Current Pickup Preview Item(execute=[@n3.CastFailed, @n2.Is Not Valid])
  ]}
 UpdateItemPreview {nodes [
   UpdateItemPreview() {is=title K2Node_FunctionEntry}
   n1 = Get Current Pickup Preview Item() -> (Current Pickup Preview Item:Fort Item Object Reference)
   > n2 = Is Valid(InputObject=@n1)
   > n3 = PreviewItemPickUp!(
    execute
    self:Fort HUDContext Object Reference=Get FortHUDContext()
    InItem:Fort Item Object Reference=@n1
    InQuickBarType:EFortQuickBars Enum=Primary
   ) -> (then, OutSlotIndex:int=0, OutItemsToPush:Map of Integers to Integers, OutItemsToDrop:Array of Integers, ReturnValue:bool)
   > n4 = Branch(Condition="integer > integer"(A=@n3))
   > n5 = Branch(Condition=@n3.ReturnValue)
   n6 = Break Widget Transform!(WidgetTransform:"Widget Transform Structure (by ref)"=Get RenderTransform!(self:Widget Object Reference=Get PickupPreview() -> (PickupPreview:Size Box Object Reference)) -> (RenderTransform:Widget Transform Structure)) -> (Translation:Vector 2D Structure, Scale:Vector 2D Structure, Shear:Vector 2D Structure, Angle:float=~) {is=title K2Node_BreakStruct}
   n7 = BreakVector2D(InVec=@n6)
   n8 = GetNumberOfSlotsTaken!(self:Fort Item Object Reference=Get Current Pickup Preview Item()) -> (ReturnValue:Byte=0)
   n9 = int + int(A=@n3, B="To Integer (Byte)"!(InByte:Byte=Byte - Byte!(A:Byte=@n8, B:Byte=1) -> (ReturnValue:Byte=0)) -> (ReturnValue:int=0)) -> (ReturnValue=0) {is=title K2Node_CommutativeAssociativeBinaryOperator}
   n10 = int * float!(
    A:int="Min (Integer)"(
     A=Get Slot Display Index!(SlotIndex:int=@n3) -> (ColumnIndex:int=0) {is=title K2Node_CallFunction}
     B=Get Slot Display Index(SlotIndex=@n9) -> (ColumnIndex=0)
    ) -> (ReturnValue=0)
    B:double=Get Item Preview Width() -> (Item Preview Width:double)
   ) -> (ReturnValue:double)
   > SetRenderTransform!(
    execute
    self:Widget Object Reference=Get PickupPreview()
    InTransform:Widget Transform Structure=Make Widget Transform!(
     Translation:Vector 2D Structure=MakeVector2D(X=@n10, Y=@n7.Y)
     Scale:Vector 2D Structure=@n6.Scale
     Shear:Vector 2D Structure=@n6.Shear
     Angle:float=@n6.Angle
    ) -> (WidgetTransform:Widget Transform Structure) {is=title K2Node_MakeStruct}
   ) -> (then)
   n11 = GetNumberOfSlotsTaken(self=Get Current Pickup Preview Item()) -> (ReturnValue=0)
   > SetWidthOverride(self=Get PickupPreview(), InWidthOverride=int * float(A="To Integer (Byte)"(InByte=@n11) -> (ReturnValue=0), B=Get Item Preview Width()))
   > SetVisibility(self=Get PickupPreview(), InVisibility=HitTestInvisible)
   SetVisibility(execute=[@n5.else, @n4.else, @n2.Is Not Valid], self=Get PickupPreview(), InVisibility=Hidden)
  ]}
 UpdateQuickBarCaching {nodes [
   n1 = Is Simplified HUD!(Exec:Exec) -> (Simplified:Exec, Default:Exec) {is=title K2Node_MacroInstance}
   n2 = Get QuickbarPrimaryInvalidationBox() -> (QuickbarPrimaryInvalidationBox:Invalidation Box Object Reference)
   SetCanCache!(
    execute=@n1.Default
    self:Invalidation Box Object Reference=@n2
    CanCache:bool=UpdateQuickBarCaching!() -> (then, DesireCaching:bool=~) {is=title K2Node_FunctionEntry}
   ) -> (then)
   SetCanCache(execute=@n1.Simplified, self=@n2, CanCache=false)
  ]}
]
variables [
 Current Pickup Preview Item {type=FortItem, default=null}
 CurrentWeaponSlot {type=integer, default=1}
 Item Preview Width {type=number, default=92}
]
widgets {
 info {
  inheritedWidgetCount=2
  parentClass=&/BacchusDynamicQuickbar.BacchusDynamicQuickbar_C
  rootWidgetClass=&Overlay
  widgetCount=15
 }
 widgets [
  Overlay_1 {widgetClassPath=&Overlay, widget=Overlay_1}
  CommonVisibilitySwitcher_0 {widgetClassPath=&CommonVisibilitySwitcher, parent=Overlay_1, widget=CommonVisibilitySwitcher_0, slot {class=OverlaySlot}}
  GridPanel_1 {widgetClassPath=&GridPanel, parent=CommonVisibilitySwitcher_0, widget=GridPanel_1, slot {class=CommonVisibilitySwitcherSlot}}
  QuickbarSlot1 {
   widgetClassPath=&/BacchusQuickBarSlot.BacchusQuickbarSlot
   bIsVariable=true
   parent=GridPanel_1
   widget=QuickbarSlot1
   properties {bConsumePointerInput=false, shouldCollapseItemWidgetBorder=true, use Native=true, cooldownTypesSupported [Death]}
   slot {class=GridSlot}
  }
  QuickbarSlot2 {
   widgetClassPath=&BacchusQuickbarSlot
   bIsVariable=true
   parent=GridPanel_1
   widget=QuickbarSlot2
   properties {
    bConsumePointerInput=false
    bSupportsDragAndDrop=true
    can Display Wide Items=true
    shouldCollapseItemWidgetBorder=true
    slotIndex=1
    use Native=true
    cooldownTypesSupported [Death, WeaponReloading, ItemActivation, AthenaWeaponFireCooldown, AmmoRegeneration]
   }
   slot {class=GridSlot, properties {column=1}}
  }
  QuickbarSlot3 {
   widgetClassPath=&BacchusQuickbarSlot
   bIsVariable=true
   parent=GridPanel_1
   widget=QuickbarSlot3
   properties {
    bConsumePointerInput=false
    bSupportsDragAndDrop=true
    can Display Wide Items=true
    shouldCollapseItemWidgetBorder=true
    slotIndex=2
    use Native=true
    cooldownTypesSupported [Death, WeaponReloading, ItemActivation, AthenaWeaponFireCooldown, AmmoRegeneration]
   }
   slot {class=GridSlot, properties {column=2}}
  }
  QuickbarSlot4 {
   widgetClassPath=&BacchusQuickbarSlot
   bIsVariable=true
   parent=GridPanel_1
   widget=QuickbarSlot4
   properties {
    bConsumePointerInput=false
    bSupportsDragAndDrop=true
    can Display Wide Items=true
    shouldCollapseItemWidgetBorder=true
    slotIndex=3
    use Native=true
    cooldownTypesSupported [Death, WeaponReloading, ItemActivation, AthenaWeaponFireCooldown, AmmoRegeneration]
   }
   slot {class=GridSlot, properties {column=3}}
  }
  QuickbarSlot5 {
   widgetClassPath=&BacchusQuickbarSlot
   bIsVariable=true
   parent=GridPanel_1
   widget=QuickbarSlot5
   properties {
    bConsumePointerInput=false
    bSupportsDragAndDrop=true
    can Display Wide Items=true
    shouldCollapseItemWidgetBorder=true
    slotIndex=4
    use Native=true
    cooldownTypesSupported [Death, WeaponReloading, ItemActivation, AthenaWeaponFireCooldown, AmmoRegeneration]
   }
   slot {class=GridSlot, properties {column=4}}
  }
  QuickbarSlot6 {
   widgetClassPath=&BacchusQuickbarSlot
   bIsVariable=true
   parent=GridPanel_1
   widget=QuickbarSlot6
   properties {
    bConsumePointerInput=false
    bSupportsDragAndDrop=true
    can Display Wide Items=true
    shouldCollapseItemWidgetBorder=true
    slotIndex=5
    use Native=true
    cooldownTypesSupported [Death, WeaponReloading, ItemActivation, AthenaWeaponFireCooldown, AmmoRegeneration]
   }
   slot {class=GridSlot, properties {column=5}}
  }
  HorizontalBox_Slots {
   widgetClassPath=&HorizontalBox
   bIsVariable=true
   parent=CommonVisibilitySwitcher_0
   widget=HorizontalBox_Slots
   slot {class=CommonVisibilitySwitcherSlot}
  }
  DynamicQuickbarBox {
   widgetClassPath=&DynamicEntryBox
   bIsVariable=true
   parent=CommonVisibilitySwitcher_0
   widget=DynamicQuickbarBox
   properties {entryWidgetClass=&/BacchusQuickBarSlot.BacchusQuickbarSlot_C, numDesignerPreviewEntries=6}
   slot {class=CommonVisibilitySwitcherSlot}
  }
  PickupPreview {
   widgetClassPath=&SizeBox
   bIsVariable=true
   parent=Overlay_1
   widget=PickupPreview
   properties {
    bOverride_HeightOverride=true
    bOverride_WidthOverride=true
    heightOverride=94
    visibility=Hidden
    widthOverride=94
    renderTransformPivot=<0 0.5>
   }
   slot {class=OverlaySlot}
  }
  Border {
   widgetClassPath=&Border
   parent=PickupPreview
   widget=Border
   properties {
    visibility=SelfHitTestInvisible
    background {
     bIsDynamicallyLoaded=false
     drawAs=Box
     imageType=NoImage
     mirroring=NoMirror
     resourceName=None
     resourceObject=/Game/UI/InGame/Quickbar/M_UI_FramePreview_Inst
     tiling=NoTile
     imageSize=<80 80>
     margin=<ltrb 0 0 0 0>
     outlineSettings {
      bUseBrushTransparency=false
      roundingType=HalfHeightRadius
      width=0
      color {colorUseRule=UseColor_Specified, specifiedColor=<rgba 0 0 0 0>}
      cornerRadii=<0 0 0 0>
     }
     tintColor {colorUseRule=UseColor_Specified, specifiedColor=<rgba 1 1 1 1>}
     uVRegion {bIsValid=false, max=<0 0>, min=<0 0>}
    }
    padding=<ltrb -200 -200 -200 -200>
    renderTransformPivot=<0 0>
   }
   slot {class=SizeBoxSlot, properties {padding=<ltrb -10 -10 -10 -10>}}
  }
  KeybindWidget_SwitchModeKeybind {widgetClassPath=/Script/FortniteUI.FortKeybindWidget, bInherited=true, widget=null}
  TutorialHighlight {widgetClassPath=&TutorialHighlight}
 ]
}
