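# Mobile (Bacchus) HUD quickbar combo slot that shows one slot index as a primary, secondary or Creative slot.
# Nested visibility switchers follow slot type, build mode, quickbar focus and Creative equip, cancelling interaction on slot switches.
# Inventory and HUD context events drive refresh, selection and a preview frame marking where an interacted pickup would land.
# In secondary mode, ability-system tags collapse the widget unless its slot holds a trap.
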WidgetBlueprint /Game/Athena/HUD/Bacchus/BacchusQuickbarSlot_Combo
graphs [
 EventGraph {nodes [
   n1 = On Building Mode Changed!() -> (OutputDelegate, then, bEntering:bool=~) {is=title K2Node_CustomEvent}
   > TrySetQuickbarType!(
    execute
    self
    Mode:EFortQuickBars Enum=Select(Option 0:EFortQuickBars Enum=Primary, Option 1:EFortQuickBars Enum=Secondary, Index=@n1.bEntering) -> (ReturnValue:EFortQuickBars Enum)
   ) -> (then)
   n2 = On Contents Changed!() -> (OutputDelegate, then, QuickbarIndex:EFortQuickBars Enum=Primary, ChangedSlots:Array of Integers) {is=title K2Node_CustomEvent}
   > Branch(Condition=Contains Item(TargetArray:Array of Integers=@n2.ChangedSlots, ItemToFind:"Integer (by ref)"=Get SlotIndex() -> (SlotIndex:int=0)))
   > Refresh(self:Bacchus Quickbar Combo Slot Object Reference)
   n3 = OnLocalPlayerDied!() -> (OutputDelegate, then, DeathReport:"Fort Player Death Report Structure (by ref)") {is=title K2Node_CustomEvent}
   Refresh(execute=[@n3.then, @n9.then, @n10.then])
   n4 = Focus Slot Changed!() -> (OutputDelegate, then, QuickbarIndex:EFortQuickBars Enum=Primary, Slot:int) {is=title K2Node_CustomEvent}
   n5 = Get SlotIndex() -> (SlotIndex=0)
   n6 = GetQuickbarType!(self:Bacchus Quickbar Combo Slot Object Reference) -> (ReturnValue:EFortQuickBars Enum=Primary)
   n7 = "Equal (Enum)"(A:EFortQuickBars Enum=@n6, B:EFortQuickBars Enum=Max_None)
   n8 = "Equal (Enum)"(A=@n4.QuickbarIndex, B=@n6)
   > SetSlotSelected!(
    execute
    self:Fort Quick Bar Slot Base Object Reference=[
     Get PrimarySlot() -> (PrimarySlot:Bacchus Quickbar Slot Base Object Reference)
     Get SecondarySlot() -> (SecondarySlot:Bacchus Quickbar Slot Base Object Reference)
    ]
    bSelected:bool=AND Boolean(A=OR Boolean(A=@n8, B=@n7), B="Equal (Integer)"(A=@n4.Slot, B=@n5) {is=title K2Node_CallFunction})
   ) -> (then)
   n9 = OnLocalPlayerRevived() {is=title K2Node_CustomEvent}
   n10 = OnLocalPlayerSpawned() {is=title K2Node_CustomEvent}
   n11 = Event PreConstruct()
   > Branch(Condition=@n11.IsDesignTime)
   > Refresh()
   Event Construct()
   > n12 = Sequence!(execute) -> (then_0, then_1, then_2:Exec, then_3:Exec)
   Assign On Quickbar Contents Changed!(
    execute:Exec=@n12.then_1
    self:Fort Inventory Context Object Reference=Get FortInventoryContext!() -> (ReturnValue:Fort Inventory Context Object Reference) {is=title K2Node_GetBlueprintContextEx}
    Delegate:"Delegate (by ref)"=@n2
   ) -> (then:Exec) {is=title K2Node_AssignDelegate}
   > Bind Event to On Quickbar Slot Focus Changed(self:Fort Inventory Context Object Reference=Get FortInventoryContext(), Delegate=@n4) {is=title K2Node_AddDelegate}
   > Assign On Creative Quickbar Equipped Changed!(
    execute:Exec
    self:Creative HUDLContext Object Reference=Get CreativeHUDLContext!() -> (ReturnValue:Creative HUDLContext Object Reference) {is=title K2Node_GetBlueprintContextEx}
    Delegate:"Delegate (by ref)"=Create Event()
   ) -> (then:Exec) {is=title K2Node_AssignDelegate}
   Bind Event to On Build Mode Changed(
    execute=@n12.then_0
    self:Fort HUDContext Object Reference=Get FortHUDContext!() -> (ReturnValue:Fort HUDContext Object Reference) {is=title K2Node_GetBlueprintContextEx}
    Delegate=@n1
   ) {is=title K2Node_AddDelegate}
   n13 = Get FortHUDContext()
   Bind Event to On Local Player Died(execute=@n12.then_2, self:Fort HUDContext Object Reference=@n13, Delegate=@n3) {is=title K2Node_AddDelegate}
   > Assign On Local Player Revived!(execute:Exec, self:Fort HUDContext Object Reference=@n13, Delegate:"Delegate (by ref)"=@n9) -> (then:Exec) {is=title K2Node_AssignDelegate}
   > Assign On Local Player Spawned!(execute:Exec, self:Fort HUDContext Object Reference=@n13, Delegate:"Delegate (by ref)"=@n10) -> (then:Exec) {is=title K2Node_AssignDelegate}
   n14 = Create Event()
   > Bind Event to On Player Can Interact(self:Fort HUDContext Object Reference=@n13, Delegate=@n14) {is=title K2Node_AddDelegate}
   > Bind Event to On Interact Updated(self:Fort HUDContext Object Reference=@n13, Delegate=@n14) {is=title K2Node_AddDelegate}
   Event Refresh() {is=title K2Node_Event}
   > Parent: Refresh() {is=title K2Node_CallParentFunction}
   > n15 = Sequence!(execute) -> (then_0, then_1)
   > SetQuickbarIndexRuntime!(
    execute
    self:Fort Quick Bar Slot Base Object Reference=[Get PrimarySlot(), Get CreativeSlot() -> (CreativeSlot:Bacchus Quickbar Slot Base Object Reference)]
    Index:int=Get SlotIndex() -> (SlotIndex=0)
    bRefresh:bool=true
   ) -> (then)
   > SetQuickbarIndexRuntime(self=Get SecondarySlot(), Index="Min (Integer)"(A=Get SlotIndex() -> (SlotIndex=0), B=4) -> (ReturnValue=0), bRefresh=true)
   > TrySetQuickbarType(Mode=GetQuickbarFocus!(self:Fort Inventory Context Object Reference=Get FortInventoryContext()) -> (OutQuickBar:EFortQuickBars Enum=Primary, OutSlot:int=0, OutSecondarySlot:int=0, OutPreviousFocusedSlot:int=0))
   > SetUserSpecifiedScale!(
    execute
    self:Scale Box Object Reference=Get ResizeBox() -> (ResizeBox:Scale Box Object Reference)
    InUserSpecifiedScale:float=float / float(A=Get SlotSize() -> (SlotSize:float), B=92) {is=title K2Node_CallFunction}
   ) -> (then)
   > UpdateSecondarySlotsVisibility()
   > Branch(Condition="integer >= integer"!(A:int=Get SlotIndex() -> (SlotIndex=0), B:int=5) -> (ReturnValue:bool) {is=title K2Node_CallFunction})
   > n16 = Switch on Int!(
    execute:Exec
    Selection:int=GetActiveWidgetIndex!(self:Common Visibility Switcher Object Reference=Get Switcher() -> (Switcher:Common Visibility Switcher Object Reference)) -> (ReturnValue:int=0)
   ) -> (Default:Exec, 0:Exec, 1:Exec)
   SetVisibility(execute=@n16.0, InVisibility=Visible)
   SetVisibility(execute=@n16.1, InVisibility=Collapsed)
   UpdateItemPreview(execute=@n15.then_1)
   Event OnHUDStateUpdate!() -> (OutputDelegate, then, NewState:"Fort HUD State Structure (by ref)") {is=title K2Node_Event}
   > UpdateSecondarySlotsVisibility()
  ]}
 HasTrapInSlot {nodes [
   HasTrapInSlot() {is=title K2Node_FunctionEntry}
   > n1 = Branch(Condition="integer >= integer"(A=Get SlotIndex() -> (SlotIndex=0), B=4) {is=title K2Node_PromotableOperator})
   > n2 = Cast To GameplayTagAssetInterface!(
    execute
    Object=GetQuickBarSlottedItem!(
     self:Fort Inventory Context Object Reference=Get FortInventoryContext()
     InQuickBar:EFortQuickBars Enum=Secondary
     Slot:int=Get SlotIndex() -> (SlotIndex=0)
    ) -> (ReturnValue:Fort Item Object Reference)
   ) -> (then, CastFailed, AsGameplay Tag Asset Interface:Gameplay Tag Asset Interface Interface) {is=title K2Node_DynamicCast}
   > Return Node!(
    execute
    ReturnValue:bool=HasMatchingGameplayTag!(self:Gameplay Tag Asset Interface Interface=@n2, TagToCheck:Gameplay Tag Structure=(TagName="Trap")) -> (ReturnValue:bool)
   )
   Return Node(execute=[@n2.CastFailed, @n1.else], ReturnValue=false)
  ]}
 Interaction Updated {nodes [
   n1 = Interaction Updated!() -> (then, ContextInfo:Fort Interact Context Info Object Reference) {is=title K2Node_FunctionEntry}
   > n2 = Is Valid(InputObject=@n1)
   > n3 = Cast To FortPickup!(execute, Object=Get ReceivingActor!(self:Fort Interact Context Info Object Reference=@n1) -> (ReceivingActor:Actor Weak object)) -> (then, CastFailed, AsFort Pickup:Fort Pickup Object Reference) {is=title K2Node_DynamicCast}
   > n4 = Set Current Preview Item(Current Preview Item:Fort Item Object Reference=Get PrimaryPickupDummyItem!(self:Fort Pickup Object Reference=@n3) -> (PrimaryPickupDummyItem:Fort Item Object Reference)) -> (Output_Get:Fort Item Object Reference)
   UpdateItemPreview(execute=[@n5.then, @n4.then])
   n5 = Set Current Preview Item(execute=[@n3.CastFailed, @n2.Is Not Valid])
  ]}
 OnCreativeQuickbarChanged {nodes [
   n1 = OnCreativeQuickbarChanged!() -> (then, IsEquipped:bool=~) {is=title K2Node_FunctionEntry}
   > SetActiveWidgetIndex(
    self:Common Visibility Switcher Object Reference=Get CreativeSwitcher() -> (CreativeSwitcher:Common Visibility Switcher Object Reference)
    Index=Select(Option 0:int=0, Option 1:int=1, Index=@n1) -> (ReturnValue:int)
   )
   n2 = GetActiveWidgetIndex(self=Get Switcher()) -> (ReturnValue=0)
   n3 = Get PrimarySlot()
   n4 = Get SecondarySlot()
   n5 = Cast To BacchusQuickbarSlot!(Object=Select(
     Option 0:Bacchus Quickbar Slot Base Object Reference=Select(Option 0=@n3, Option 1=@n4, Index:int=@n2)
     Option 1:Bacchus Quickbar Slot Base Object Reference=Get CreativeSlot()
     Index=NOT Boolean(A=@n1)
    ) -> (ReturnValue:Bacchus Quickbar Slot Base Object Reference)) -> (AsBacchus Quickbar Slot:Bacchus Quickbar Slot Object Reference, bSuccess:bool=~) {is=title K2Node_DynamicCast}
   > Branch(Condition=@n5.bSuccess)
   > Cancel Interaction(self:Bacchus Quickbar Slot Object Reference=@n5)
  ]}
 TrySetQuickbarType {nodes [
   TrySetQuickbarType!() -> (then, Mode:EFortQuickBars Enum=Primary) {is=title K2Node_FunctionEntry}
   > n1 = Branch(Condition="Equal (Enum)"(A:EComboSlotType Enum=Get SlotType() -> (SlotType:EComboSlotType Enum=Primary), B:EComboSlotType Enum=Combo))
   n2 = Get Mode() -> (Mode:EFortQuickBars Enum=Primary)
   > SetActiveWidgetIndex(
    self=Get Switcher()
    Index=Select(Option 0:int=1, Option 1:int=0, Index:bool="Equal (Enum)"(A:EFortQuickBars Enum=Primary, B:EFortQuickBars Enum=@n2)) -> (ReturnValue:int)
   )
   n3 = "Not Equal (Enum)"!(A:EFortQuickBars Enum=Primary, B:EFortQuickBars Enum=Get Mode() -> (Mode=Primary)) -> (ReturnValue:bool)
   n4 = Get PrimarySlot()
   n5 = Get SecondarySlot()
   > Cancel Interaction(self=Cast To BacchusQuickbarSlot(Object=Select(Option 0:Bacchus Quickbar Slot Base Object Reference=@n5, Option 1:Bacchus Quickbar Slot Base Object Reference=@n4, Index=@n3) -> (ReturnValue:Bacchus Quickbar Slot Base Object Reference)) -> (bSuccess=~))
   SetActiveWidgetIndex(
    execute=@n1.else
    self=Get Switcher()
    Index=Select!(Primary:int=0, Secondary:int=1, Combo:int=0, Creative:int=0, COUNT:int=0, Index:EComboSlotType Enum=Get SlotType() -> (SlotType=Primary)) -> (ReturnValue:int)
   )
  ]}
 UpdateItemPreview {nodes [
   UpdateItemPreview() {is=title K2Node_FunctionEntry}
   n1 = Get Current Preview Item() -> (Current Preview Item:Fort Item Object Reference)
   > n2 = Is Valid(InputObject=@n1)
   > n3 = PreviewItemPickUp!(
    execute
    self:Fort HUDContext Object Reference=Get FortHUDContext()
    InItem:Fort Item Object Reference=@n1
    InQuickBarType:EFortQuickBars Enum=Primary
   ) -> (then, OutSlotIndex:int=0, OutItemsToPush:Map of Integers to Integers, OutItemsToDrop:Array of Integers, ReturnValue:bool)
   n4 = Get SlotIndex() -> (SlotIndex=0)
   n5 = Get Current Preview Item()
   n6 = int + int(
    A=@n3
    B="To Integer (Byte)"!(InByte:Byte=GetNumberOfSlotsTaken!(self:Fort Item Object Reference=@n5) -> (ReturnValue:Byte=0)) -> (ReturnValue:int=0)
   ) -> (ReturnValue=0) {is=title K2Node_CommutativeAssociativeBinaryOperator}
   > n7 = Branch(Condition=AND Boolean!(
     A=@n3.ReturnValue
     B="integer <= integer"!(A:int=@n3, B:int=@n4) -> (ReturnValue:bool) {is=title K2Node_CallFunction}
     C:bool="integer < integer"!(A:int=@n4, B:int=@n6) -> (ReturnValue:bool) {is=title K2Node_CallFunction}
    ) -> (ReturnValue))
   n8 = Get PickupPreview() -> (PickupPreview:Size Box Object Reference)
   > SetVisibility(self=@n8, InVisibility=HitTestInvisible)
   SetVisibility(execute=[@n7.else, @n2.Is Not Valid], self=@n8, InVisibility=Hidden)
  ]}
 UpdateSecondarySlotsVisibility {nodes [
   UpdateSecondarySlotsVisibility() {is=title K2Node_FunctionEntry}
   > Branch(Condition="Equal (Enum)"(A=GetQuickbarType() -> (ReturnValue=Primary), B=Secondary))
   n1 = GetOwningPlayer()
   > n2 = Is Valid(InputObject=@n1)
   n3 = GetAbilitySystemComponent!(Actor:Actor Object Reference=@n1) -> (ReturnValue:Ability System Component Object Reference)
   > n4 = Is Valid(InputObject=@n3)
   n5 = Get RefreshHUDStateTags() -> (RefreshHUDStateTags:Gameplay Tag Container Structure)
   n6 = HasTrapInSlot!() -> (ReturnValue:bool) {is=title K2Node_CallFunction}
   > n7 = Branch(Condition=AND Boolean!(
     A=HasAnyMatchingGameplayTags!(self:Gameplay Tag Asset Interface Interface=@n3, TagContainer:"Gameplay Tag Container Structure (by ref)"=@n5) -> (ReturnValue:bool)
     B=NOT Boolean(A=@n6)
    ) -> (ReturnValue))
   > SetVisibility(InVisibility=Collapsed)
   SetVisibility(execute=[@n7.else, @n4.Is Not Valid, @n2.Is Not Valid], InVisibility=SelfHitTestInvisible)
  ]}
]
variables [Current Preview Item {type=FortItem, default=null}]
widgets {
 info {
  inheritedWidgetCount=6
  namedSlotCount=3
  parentClass=/Script/FortniteUI.BacchusQuickbarComboSlot
  rootWidgetClass=&ScaleBox
  widgetCount=15
 }
 widgets [
  ScaleBox_HUDWidgetScale {widgetClassPath=&ScaleBox, bInherited=true, widget=ScaleBox_HUDWidgetScale}
  ResizeBox {
   widgetClassPath=&ScaleBox
   bIsVariable=true
   parent=ScaleBox_HUDWidgetScale
   widget=ResizeBox
   properties {stretch=UserSpecified, userSpecifiedScale=1.11}
   slot {class=ScaleBoxSlot}
  }
  CreativeSwitcher {
   widgetClassPath=&CommonVisibilitySwitcher
   bIsVariable=true
   parent=ResizeBox
   widget=CreativeSwitcher
   slot {class=ScaleBoxSlot}
  }
  Switcher {
   widgetClassPath=&CommonVisibilitySwitcher
   bIsVariable=true
   parent=CreativeSwitcher
   widget=Switcher
   slot {class=CommonVisibilitySwitcherSlot}
  }
  PrimaryWrapper {
   widgetClassPath=/Game/Athena/HUD/FortHUDElementWrapper
   bIsVariable=true
   parent=Switcher
   widget=PrimaryWrapper
   properties {hUDElementTag {gameplayTags [{tagName=HUD.Quickbar.Combat}, {tagName=HUD.Athena.HideForAbility}, {tagName=Athena.Creative.HUDController.HidePlayerInventory}]}}
   slot {class=CommonVisibilitySwitcherSlot}
  }
  Overlay_0 {widgetClassPath=&Overlay, namedSlotHost=PrimaryWrapper, widget=Overlay_0}
  PrimarySlot {
   widgetClassPath=&/BacchusQuickBarSlot.BacchusQuickbarSlot
   bInherited=true
   bIsVariable=true
   parent=Overlay_0
   widget=PrimarySlot
   properties {
    bConsumePointerInput=false
    bSupportsDragAndDrop=true
    bUseSlotSizeBySlotUsage=false
    use Native=true
    cooldownTypesSupported [Death, WeaponReloading, ItemActivation, AthenaWeaponFireCooldown, AmmoRegeneration]
   }
   slot {class=OverlaySlot}
  }
  PickupPreview {
   widgetClassPath=&SizeBox
   bIsVariable=true
   parent=Overlay_0
   widget=PickupPreview
   properties {
    bOverride_HeightOverride=true
    bOverride_WidthOverride=true
    heightOverride=94
    visibility=Hidden
    widthOverride=94
    renderTransformPivot=<0 0.5>
   }
   slot {class=OverlaySlot}
  }
  Border {
   widgetClassPath=&Border
   parent=PickupPreview
   widget=Border
   properties {
    visibility=SelfHitTestInvisible
    background {
     bIsDynamicallyLoaded=false
     drawAs=Box
     imageType=NoImage
     mirroring=NoMirror
     resourceName=None
     resourceObject=/Game/UI/InGame/Quickbar/M_UI_FramePreview_Inst
     tiling=NoTile
     imageSize=<80 80>
     margin=<ltrb 0 0 0 0>
     outlineSettings {
      bUseBrushTransparency=false
      roundingType=HalfHeightRadius
      width=0
      color {colorUseRule=UseColor_Specified, specifiedColor=<rgba 0 0 0 0>}
      cornerRadii=<0 0 0 0>
     }
     tintColor {colorUseRule=UseColor_Specified, specifiedColor=<rgba 1 1 1 1>}
     uVRegion {bIsValid=false, max=<0 0>, min=<0 0>}
    }
    padding=<ltrb -200 -200 -200 -200>
    renderTransformPivot=<0 0>
   }
   slot {class=SizeBoxSlot, properties {padding=<ltrb -10 -10 -10 -10>}}
  }
  SecondaryWrapper {
   widgetClassPath=&FortHUDElementWrapper
   bIsVariable=true
   parent=Switcher
   widget=SecondaryWrapper
   properties {hUDElementTag {gameplayTags [
      {tagName=HUD.Quickbar.Build}
      {tagName=HUD.Athena.HideForAbility}
      {tagName=HUD.Athena.HideSecondaryForAbility}
      {tagName=Athena.Creative.HUDController.HideBuildMenu}
     ]}}
   slot {class=CommonVisibilitySwitcherSlot}
  }
  SecondarySlot {
   widgetClassPath=&BacchusQuickbarSlot
   bInherited=true
   bIsVariable=true
   namedSlotHost=SecondaryWrapper
   widget=SecondarySlot
   properties {
    bCheckVisibilityBasedOffInventoryCapacity=false
    bConsumePointerInput=false
    bUseSlotSizeBySlotUsage=false
    quickBarType=Secondary
    use Native=true
    cooldownTypesSupported [Death]
   }
  }
  CreativeWrapper {
   widgetClassPath=&FortHUDElementWrapper
   bIsVariable=true
   parent=CreativeSwitcher
   widget=CreativeWrapper
   properties {hUDElementTag {gameplayTags [{tagName=HUD.Athena.HideForAbility}, {tagName=HUD.Quickbar.Creative}]}}
   slot {class=CommonVisibilitySwitcherSlot}
  }
  CreativeSlot {
   widgetClassPath=&BacchusQuickbarSlot
   bInherited=true
   bIsVariable=true
   namedSlotHost=CreativeWrapper
   widget=CreativeSlot
   properties {bConsumePointerInput=false, bUseSlotSizeBySlotUsage=false, quickBarType=Creative, use Native=true, cooldownTypesSupported [Death]}
  }
  Image_Background {widgetClassPath=&Image, bInherited=true, widget=null}
  TutorialHighlight {widgetClassPath=&TutorialHighlight}
 ]
}
