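# Radial speed indicator HUD widget, built on BacchusRadialDisplayBase, that shows a fill ring using a flying-speed indicator material.
# The radial counter update event scales the incoming fill percent by MaxFillAmount and clamps it between zero and that maximum, so the ring fills only part of a circle.
# The result drives the PercentFillAmount parameter on the image's dynamic material.
# The image sits in nested size and scale boxes under an InvalidationBox.
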WidgetBlueprint /Game/Athena/HUD/Bacchus/BacchusRadialSpeedIndicator
graphs [EventGraph {nodes [
   n1 = Event OnUpdateRadialCounterEvent_Blueprint!() -> (OutputDelegate, then, NewCounterFillPercent:float) {is=title K2Node_Event}
   > n2 = GetDynamicMaterial(self=Get Image_RadialCounter() -> (Image_RadialCounter:Image Object Reference))
   n3 = Get MaxFillAmount() -> (MaxFillAmount:double)
   > SetScalarParameterValue(
    self=@n2
    ParameterName=PercentFillAmount
    Value="Clamp (Float)"(Value=float * float(A=@n1.NewCounterFillPercent, B=@n3) {is=title K2Node_PromotableOperator}, Min=0, Max=@n3)
   )
  ]}]
variables [MaxFillAmount {type=number, default=0.45}]
widgets {
 info {inheritedWidgetCount=1, parentClass=/Script/FortniteUI.BacchusRadialDisplayBase, rootWidgetClass=&Overlay, widgetCount=9}
 widgets [
  Overlay_0 {widgetClassPath=&Overlay, widget=Overlay_0}
  ButtonSizeBox {
   widgetClassPath=&SizeBox
   bIsVariable=true
   parent=Overlay_0
   widget=ButtonSizeBox
   properties {bOverride_HeightOverride=true, bOverride_WidthOverride=true, heightOverride=200, widthOverride=200}
   slot {class=OverlaySlot, properties {horizontalAlignment=HAlign_Center, verticalAlignment=VAlign_Center, padding=<ltrb 4 2 4 2>}}
  }
  InvalidationBox_0 {
   widgetClassPath=/Script/UMG.InvalidationBox
   parent=ButtonSizeBox
   widget=InvalidationBox_0
   slot {class=SizeBoxSlot, properties {horizontalAlignment=HAlign_Center, verticalAlignment=VAlign_Center}}
  }
  ImageSizeBox {widgetClassPath=&SizeBox, bIsVariable=true, parent=InvalidationBox_0, widget=ImageSizeBox, slot {class=PanelSlot}}
  ScaleBox_4 {widgetClassPath=&ScaleBox, parent=ImageSizeBox, widget=ScaleBox_4, slot {class=SizeBoxSlot}}
  SizeBox_1 {
   widgetClassPath=&SizeBox
   parent=ScaleBox_4
   widget=SizeBox_1
   properties {bOverride_HeightOverride=true, bOverride_WidthOverride=true, heightOverride=128, widthOverride=128}
   slot {class=ScaleBoxSlot}
  }
  Overlay_1 {widgetClassPath=&Overlay, parent=SizeBox_1, widget=Overlay_1, slot {class=SizeBoxSlot}}
  Image_RadialCounter {
   widgetClassPath=&Image
   bIsVariable=true
   parent=Overlay_1
   widget=Image_RadialCounter
   properties {
    visibility=SelfHitTestInvisible
    brush {
     bIsDynamicallyLoaded=false
     drawAs=Image
     imageType=FullColor
     mirroring=NoMirror
     resourceName=None
     resourceObject=/Game/Athena/HUD/Bacchus/Material/MI_BacchusFlyingSpeedIndicator
     tiling=NoTile
     imageSize=<105 105>
     margin=<ltrb 0 0 0 0>
     outlineSettings {
      bUseBrushTransparency=false
      roundingType=HalfHeightRadius
      width=0
      color {colorUseRule=UseColor_Specified, specifiedColor=<rgba 0 0 0 0>}
      cornerRadii=<0 0 0 0>
     }
     tintColor {colorUseRule=UseColor_Specified, specifiedColor=<rgba 1 1 1 1>}
     uVRegion {bIsValid=false, max=<0 0>, min=<0 0>}
    }
    renderTransform {angle=0, scale=<1.125 1.125>, shear=<0 0>, translation=<0 0>}
   }
   slot {class=OverlaySlot, properties {horizontalAlignment=HAlign_Center, verticalAlignment=VAlign_Center}}
  }
  TutorialHighlight {widgetClassPath=&TutorialHighlight}
 ]
}
