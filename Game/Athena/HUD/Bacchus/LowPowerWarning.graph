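# LowPowerWarning is a HUD widget that briefly shows a yellow "Warning: Low Performance Mode" text.
# On Construct it binds a custom event to the Athena HUD context's low performance mode delegate.
# When the event reports the mode is on, it plays the five-second ShowHide animation.
# That animation scales the root overlay in and back out, and keys the widget's own visibility to hide it at the end.
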WidgetBlueprint /Game/Athena/HUD/Bacchus/LowPowerWarning
animations [ShowHide {
  binding_count=2
  display_label=""
  length_seconds=5
  object=&self:ShowHide
  sequence {bindings [
    Overlay_0 {
     id=#1
     possessed_class=Overlay
     tracks [{
       type=MovieScene2DTransformTrack
       sections [{
         type=MovieScene2DTransformSection
         end_frame=100
         start_frame=0
         channels [
          Translation.X {type=MovieSceneScriptingFloatChannel}
          Translation.Y {type=MovieSceneScriptingFloatChannel}
          Angle {type=MovieSceneScriptingFloatChannel}
          Scale.X {type=MovieSceneScriptingFloatChannel, keys [[0,0],[5,1],[95,1],[100,0]]}
          Scale.Y {type=MovieSceneScriptingFloatChannel, keys [[0,0],[5,1],[95,1],[100,0]]}
          Shear.X {type=MovieSceneScriptingFloatChannel}
          Shear.Y {type=MovieSceneScriptingFloatChannel}
         ]
        }]
      }]
    }
    LowPowerWarning_C_22 {
     id=#2
     possessed_class=LowPowerWarning_C
     tracks [{
       type=MovieSceneByteTrack
       display_name=Visibility
       sections [{type=MovieSceneByteSection, end_frame=100, start_frame=0, channels [None {type=MovieSceneScriptingByteChannel, keys [[0,4],[100,1]]}]}]
      }]
    }
   ]}
 }]
graphs [EventGraph {nodes [
   n1 = In Low Power Mode!() -> (OutputDelegate, then, bInLowPerformanceMode:bool=~) {is=title K2Node_CustomEvent}
   > Branch(Condition=@n1.bInLowPerformanceMode)
   > PlayAnimation(InAnimation=Get ShowHide() -> (ShowHide:Widget Animation Object Reference))
   Event Construct()
   > Bind Event to On Low Performance Mode(
    self:Athena HUDContext Object Reference=Get AthenaHUDContext!() -> (ReturnValue:Athena HUDContext Object Reference) {is=title K2Node_GetBlueprintContextEx}
    Delegate=@n1
   ) {is=title K2Node_AddDelegate}
  ]}]
variables []
widgets {
 info {inheritedWidgetCount=0, parentClass=&UserWidget, rootWidgetClass=&Overlay, widgetCount=2}
 widgets [
  Overlay_0 {widgetClassPath=&Overlay, widget=Overlay_0}
  LowerPowerModeText {
   widgetClassPath=&CommonTextBlock
   bIsVariable=true
   parent=Overlay_0
   widget=LowerPowerModeText
   properties {
    style=/FortUILibrary/Styles/Typography/Foundation/TextStyle-Base-M-B-Yellow.TextStyle-Base-M-B-Yellow_C
    text=Warning: Low Performance Mode
    colorAndOpacity {colorUseRule=UseColor_Specified, specifiedColor=<rgba 1 0.8714 0.0723 1>}
    font {
     bForceMonospaced=false
     bMaterialIsStencil=false
     fontMaterial=null
     fontObject=&NotoSans
     letterSpacing=0
     monospacedWidth=1
     size=24
     skewAmount=0
     typefaceFontName=NotoSans-Bold
     outlineSettings {
      bApplyOutlineToDropShadows=false
      bMiteredCorners=false
      bSeparateFillAlpha=false
      outlineMaterial=null
      outlineSize=0
      outlineColor=<rgba 0 0 0 1>
     }
    }
    shadowOffset=<0 0>
   }
   slot {class=OverlaySlot}
  }
 ]
}
