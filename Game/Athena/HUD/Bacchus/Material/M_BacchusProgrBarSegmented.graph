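# Translucent UI material for the Bacchus HUD that draws a segmented radial progress ring with fill and reload amounts.
# Two smoothstepped radial gradients form the ring, and MF_UI_RadialSegments generates the segments.
# Smoothstepped angle gradients mask fill and reload progress, with ReloadFade scaling part of the shape.
# An offset, scalable texture provides another mask, and nested lerps use background and fill colors.
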Material /Game/Athena/HUD/Bacchus/Material/M_BacchusProgrBarSegmented
graph {
 properties {blendMode=BLEND_Translucent, materialDomain=MD_UI, pixelDepthOffsetMode=PDOM_Legacy, usedShadingModels=Default Lit}
 nodes [
  n1 = NamedRerouteUsage()
  LinearInterpolate(
   A=@n1
   B=LinearInterpolate(
    A=VectorParameter() {parameterName=BGColor, defaultValue=<rgba 0.0052 0.0052 0.0056 0.6>}.RGB
    B=VectorParameter() {parameterName=FillColor, defaultValue=<rgba 1 1 1 1>}.RGB
    Alpha=NamedRerouteUsage()
   )
   Alpha=OneMinus(None=@n1)
  )
  n2 = Subtract(A=NamedRerouteUsage(), B=NamedRerouteUsage())
  n3 = NamedRerouteUsage()
  n4 = Multiply(B=ScalarParameter() {defaultValue=0.5, parameterName=ReloadFade}) {constA=1}
  n5 = OneMinus(None=NamedRerouteUsage())
  Max(A=Multiply(A=Multiply(A=@n2, B=@n3), B=LinearInterpolate(B=@n4, Alpha=@n5) {constA=1}), B=NamedRerouteUsage())
  n6 = ScalarParameter() {defaultValue=-0.25, parameterName=FillStartRotation}
  n7 = OneMinus(None=ScalarParameter() {defaultValue=1, parameterName=PercentReloadAmount})
  n8 = ScalarParameter() {defaultValue=0.0015, parameterName=FillEdgeSmooth}
  NamedRerouteDeclaration(None=MaterialFunctionCall(
    Alpha=MaterialFunctionCall!(UVs, Rotation=@n6) {materialFunction=/FortUILibrary/MaterialFunctions/Gradients/MF_UI_AngleGradient}.Result
    Min=@n7
    Max=Add(A=@n7, B=@n8)
   ) {materialFunction=&SmoothStep}.Result) {name=RadialMask, nodeColor=<rgba 0 0 0 1>}
  n9 = MaterialFunctionCall() {materialFunction=&RadialGradientExponential}
  n10 = ScalarParameter() {defaultValue=0.005, parameterName=OuterSize}
  n11 = Add(A=@n10, B=ScalarParameter() {defaultValue=1, parameterName=EdgeSmooth})
  n12 = Add(A=ScalarParameter() {defaultValue=0.005, parameterName=OuterSize}, B=ScalarParameter() {defaultValue=0.1, parameterName=RingSize})
  n13 = Add(A=@n12, B=ScalarParameter() {defaultValue=1, parameterName=EdgeSmooth})
  NamedRerouteDeclaration(None=Subtract(
    A=MaterialFunctionCall(Alpha=@n9.RadialGradientExponential, Min=@n10, Max=@n11) {materialFunction=&SmoothStep}.Result
    B=MaterialFunctionCall(Alpha=@n9.RadialGradientExponential, Min=@n12, Max=@n13) {materialFunction=&SmoothStep}.Result
   )) {name=Radial Shape, nodeColor=<rgba 0 0 0 1>}
  n14 = Add(
   A=MaterialFunctionCall(UVs=TextureCoordinate(), Texture Scale=ScalarParameter() {defaultValue=1, group=Texture, parameterName=TextureSize}) {materialFunction=&ScaleUVsByCenter}.UVs
   B=AppendVector(A=ScalarParameter() {group=Texture, parameterName=Offset H}, B=ScalarParameter() {group=Texture, parameterName=Offset V})
  )
  NamedRerouteDeclaration(None=Multiply(
    A=TextureSampleParameter2D(UVs=@n14) {group=Texture, parameterName=Texture, samplerSource=SSM_Clamp_WorldGroupSettings}.R
    B=ScalarParameter() {defaultValue=1, group=Texture, parameterName=TextureOpacity}
   )) {name=Texture, nodeColor=<rgba 0 0 0 1>}
  NamedRerouteDeclaration(None=MaterialFunctionCall!(
    UV
    Segment Count=ScalarParameter() {defaultValue=10, parameterName=Repeat}
    Segment Spacing=ScalarParameter() {defaultValue=0.8, parameterName=SegmentSpacing}
    Fixed Spacing
   ) {materialFunction=/FortUILibrary/MaterialFunctions/Basics/MF_UI_RadialSegments}.Segmented Lines Fill) {name=Segments, nodeColor=<rgba 0 0 0 1>}
  n15 = ScalarParameter() {defaultValue=-0.25, parameterName=FillStartRotation}
  n16 = OneMinus(None=ScalarParameter() {defaultValue=0.5, parameterName=PercentFillAmount})
  n17 = ScalarParameter() {defaultValue=0.0015, parameterName=FillEdgeSmooth}
  NamedRerouteDeclaration(None=MaterialFunctionCall(
    Alpha=MaterialFunctionCall(Rotation=@n15) {materialFunction=&MF_UI_AngleGradient}.Result
    Min=@n16
    Max=Add(A=@n16, B=@n17)
   ) {materialFunction=&SmoothStep}.Result) {name=ProgressMask, nodeColor=<rgba 0 0 0 1>}
 ]
}
