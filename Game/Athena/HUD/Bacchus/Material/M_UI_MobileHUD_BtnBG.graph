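# Translucent UI material for mobile HUD button backgrounds, drawing a circle with outer outline, inner outline and center fill.
# On-screen pixel width relative to ReferenceSize adjusts outline widths, which set the circle sizes whose differences form the ring masks.
# Masks are tinted and summed into final color, and weighted by per-layer opacities into final opacity.
# A switchable, invertible exponential radial gradient optionally adds to opacity before the overall multiplier.
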Material /Game/Athena/HUD/Bacchus/Material/M_UI_MobileHUD_BtnBG
graph {
 expressions {
  MaterialExpressionAdd_1 {
   class=MaterialExpressionAdd
   inputs [
    A {from=MaterialExpressionMultiply_2}
    B {from=MaterialExpressionMultiply_4}
   ]
  }
  MaterialExpressionAdd_13 {
   class=MaterialExpressionAdd
   inputs [
    A {from=MaterialExpressionSaturate_1}
    B {from=MaterialExpressionReroute_283}
   ]
  }
  MaterialExpressionAdd_17 {
   class=MaterialExpressionAdd
   inputs [
    A {from=MaterialExpressionReroute_284}
    B {from=MaterialExpressionReroute_285}
   ]
  }
  MaterialExpressionAdd_18 {
   class=MaterialExpressionAdd
   inputs [
    A {from=MaterialExpressionMultiply_28}
    B {from=MaterialExpressionMultiply_30}
   ]
  }
  MaterialExpressionAdd_19 {
   class=MaterialExpressionAdd
   inputs [
    A {from=MaterialExpressionSaturate_3}
    B {from=MaterialExpressionReroute_295}
   ]
  }
  MaterialExpressionAdd_3 {
   class=MaterialExpressionAdd
   inputs [
    A {from=MaterialExpressionScalarParameter_47}
    B {from=MaterialExpressionScalarParameter_44}
   ]
  }
  MaterialExpressionAdd_4 {
   class=MaterialExpressionAdd
   inputs [
    A {from=MaterialExpressionNamedRerouteUsage_12}
    B {from=MaterialExpressionReroute_12}
   ]
  }
  MaterialExpressionClamp_0 {
   class=MaterialExpressionClamp
   inputs [
    None {from=MaterialExpressionDivide_0, output=""}
    Min {from="", output=""}
    Max {from="", output=""}
   ]
   properties {maxDefault=0.98}
  }
  MaterialExpressionClamp_1 {
   class=MaterialExpressionClamp
   inputs [
    None {from=MaterialExpressionDivide_2, output=""}
    Min {from="", output=""}
    Max {from="", output=""}
   ]
   properties {maxDefault=0.98}
  }
  MaterialExpressionComponentMask_3 {
   class=MaterialExpressionComponentMask
   inputs [None {from=MaterialExpressionMaterialFunctionCall_21, output=Pixel Size}]
   properties {desc=Checks width only, r=true}
  }
  MaterialExpressionComposite_2 {
   class=MaterialExpressionComposite
   inputs [OutlineWidth {from=MaterialExpressionScalarParameter_52, output=""}]
   properties {subgraphName=GetOutlineSize}
  }
  MaterialExpressionComposite_3 {class=MaterialExpressionComposite, inputs [OutlineWidth {from=MaterialExpressionAdd_3, output=""}], properties {subgraphName=GetOutlineSize}}
  MaterialExpressionConstant_10 {class=MaterialExpressionConstant, properties {r=1}}
  MaterialExpressionConstant_12 {class=MaterialExpressionConstant, properties {r=0.064}}
  MaterialExpressionConstant_9 {class=MaterialExpressionConstant, properties {r=0.02}}
  MaterialExpressionDivide_0 {
   class=MaterialExpressionDivide
   inputs [
    A {from=MaterialExpressionReroute_0, output=OutlineWidth}
    B {from=MaterialExpressionNamedRerouteUsage_0}
   ]
  }
  MaterialExpressionDivide_1 {
   class=MaterialExpressionDivide
   inputs [
    A {from=MaterialExpressionConstant_12}
    B {from=MaterialExpressionNamedRerouteUsage_23}
   ]
  }
  MaterialExpressionDivide_2 {
   class=MaterialExpressionDivide
   inputs [
    A {from=MaterialExpressionReroute_3, output=OutlineWidth}
    B {from=MaterialExpressionNamedRerouteUsage_1}
   ]
  }
  MaterialExpressionDivide_3 {
   class=MaterialExpressionDivide
   inputs [
    A {from=MaterialExpressionConstant_9}
    B {from=MaterialExpressionNamedRerouteUsage_25}
   ]
  }
  MaterialExpressionDivide_6 {
   class=MaterialExpressionDivide
   inputs [
    A {from=MaterialExpressionReroute_129}
    B {from=MaterialExpressionReroute_160}
   ]
  }
  MaterialExpressionLinearInterpolate_0 {
   class=MaterialExpressionLinearInterpolate
   inputs [
    A {from=MaterialExpressionScalarParameter_1}
    B {from=MaterialExpressionReroute_0}
    Alpha {from=MaterialExpressionScalarParameter_0}
   ]
   properties {constAlpha=1}
  }
  MaterialExpressionLinearInterpolate_1 {
   class=MaterialExpressionLinearInterpolate
   inputs [
    A {from=MaterialExpressionScalarParameter_4}
    B {from=MaterialExpressionReroute_9}
    Alpha {from=MaterialExpressionScalarParameter_7}
   ]
   properties {constAlpha=1}
  }
  MaterialExpressionMaterialFunctionCall_21 {
   class=MaterialExpressionMaterialFunctionCall
   inputs []
   properties {desc=Get the true size of the texture in pixels on screen, materialFunction=&GetUserInterfaceUV}
  }
  MaterialExpressionMaterialFunctionCall_23 {
   class=MaterialExpressionMaterialFunctionCall
   inputs [
    UVs {from=""}
    Size {from=MaterialExpressionSubtract_0}
    OutlineThickness {from=""}
    OutlineAlignment {from=""}
    GlowMin {from=""}
    GlowMax {from=""}
   ]
   properties {materialFunction=&MF_UI_Circle}
  }
  MaterialExpressionMaterialFunctionCall_24 {
   class=MaterialExpressionMaterialFunctionCall
   inputs [
    UVs {from=""}
    Size {from=MaterialExpressionMultiply_7}
    OutlineThickness {from=""}
    OutlineAlignment {from=""}
    GlowMin {from=""}
    GlowMax {from=MaterialExpressionDivide_1}
   ]
   properties {materialFunction=&MF_UI_Circle}
  }
  MaterialExpressionMaterialFunctionCall_25 {
   class=MaterialExpressionMaterialFunctionCall
   inputs [
    UVs {from=""}
    Size {from=MaterialExpressionNamedRerouteUsage_3}
    OutlineThickness {from=""}
    OutlineAlignment {from=""}
    GlowMin {from=""}
    GlowMax {from=""}
   ]
   properties {materialFunction=&MF_UI_Circle}
  }
  MaterialExpressionMaterialFunctionCall_6 {
   class=MaterialExpressionMaterialFunctionCall
   inputs [
    UVs {from=""}
    CenterPosition {from=""}
    Radius {from=MaterialExpressionMultiply_32}
    Density {from=MaterialExpressionScalarParameter_3}
    Invert Density {from=""}
   ]
   properties {materialFunction=&RadialGradientExponential}
  }
  MaterialExpressionMaterialFunctionCall_7 {
   class=MaterialExpressionMaterialFunctionCall
   inputs [
    UVs {from=""}
    CenterPosition {from=""}
    Radius {from=MaterialExpressionMultiply_33}
    Density {from=MaterialExpressionScalarParameter_2}
    Invert Density {from=""}
   ]
   properties {desc=Fill Circle, materialFunction=&RadialGradientExponential}
  }
  MaterialExpressionMax_0 {
   class=MaterialExpressionMax
   inputs [
    A {from=MaterialExpressionDivide_6}
    B {from=""}
   ]
   properties {constB=0, desc=Prevent divide by 0}
  }
  MaterialExpressionMultiply_1 {
   class=MaterialExpressionMultiply
   inputs [
    A {from=MaterialExpressionStaticSwitchParameter_3}
    B {from=MaterialExpressionReroute_297}
   ]
  }
  MaterialExpressionMultiply_10 {
   class=MaterialExpressionMultiply
   inputs [
    A {from=MaterialExpressionMaterialFunctionCall_23, output=Fill}
    B {from=MaterialExpressionOneMinus_14}
   ]
  }
  MaterialExpressionMultiply_2 {
   class=MaterialExpressionMultiply
   inputs [
    A {from=MaterialExpressionMultiply_1}
    B {from=MaterialExpressionScalarParameter_5}
   ]
  }
  MaterialExpressionMultiply_22 {
   class=MaterialExpressionMultiply
   inputs [
    A {from=MaterialExpressionReroute_218}
    B {from=MaterialExpressionOneMinus_15}
   ]
  }
  MaterialExpressionMultiply_23 {
   class=MaterialExpressionMultiply
   inputs [
    A {from=MaterialExpressionVectorParameter_6, output=RGB}
    B {from=MaterialExpressionNamedRerouteUsage_4}
   ]
  }
  MaterialExpressionMultiply_25 {
   class=MaterialExpressionMultiply
   inputs [
    A {from=MaterialExpressionVectorParameter_10, output=RGB}
    B {from=MaterialExpressionNamedRerouteUsage_5}
   ]
  }
  MaterialExpressionMultiply_26 {
   class=MaterialExpressionMultiply
   inputs [
    A {from=MaterialExpressionVectorParameter_11, output=RGB}
    B {from=MaterialExpressionNamedRerouteUsage_6}
   ]
  }
  MaterialExpressionMultiply_27 {
   class=MaterialExpressionMultiply
   inputs [
    A {from=MaterialExpressionStaticSwitchParameter_1}
    B {from=MaterialExpressionScalarParameter_65}
   ]
  }
  MaterialExpressionMultiply_28 {
   class=MaterialExpressionMultiply
   inputs [
    A {from=MaterialExpressionNamedRerouteUsage_7}
    B {from=MaterialExpressionScalarParameter_62}
   ]
  }
  MaterialExpressionMultiply_30 {
   class=MaterialExpressionMultiply
   inputs [
    A {from=MaterialExpressionNamedRerouteUsage_8}
    B {from=MaterialExpressionScalarParameter_63}
   ]
  }
  MaterialExpressionMultiply_31 {
   class=MaterialExpressionMultiply
   inputs [
    A {from=MaterialExpressionNamedRerouteUsage_13}
    B {from=MaterialExpressionScalarParameter_64}
   ]
  }
  MaterialExpressionMultiply_32 {
   class=MaterialExpressionMultiply
   inputs [
    A {from=MaterialExpressionNamedRerouteUsage_14}
    B {from=""}
   ]
   properties {constB=0.5}
  }
  MaterialExpressionMultiply_33 {
   class=MaterialExpressionMultiply
   inputs [
    A {from=MaterialExpressionNamedRerouteUsage_15}
    B {from=""}
   ]
   properties {constB=0.5}
  }
  MaterialExpressionMultiply_4 {
   class=MaterialExpressionMultiply
   inputs [
    A {from=MaterialExpressionReroute_54}
    B {from=MaterialExpressionScalarParameter_6}
   ]
  }
  MaterialExpressionMultiply_7 {
   class=MaterialExpressionMultiply
   inputs [
    A {from=MaterialExpressionConstant_10}
    B {from=MaterialExpressionNamedRerouteUsage_2}
   ]
   properties {constB=0.95}
  }
  MaterialExpressionNamedRerouteDeclaration_0 {
   class=MaterialExpressionNamedRerouteDeclaration
   inputs [None {from=MaterialExpressionMax_0}]
   properties {name=CheckSizeOnScreen, nodeColor=<rgba 0 0 0 1>}
  }
  MaterialExpressionNamedRerouteDeclaration_1 {
   class=MaterialExpressionNamedRerouteDeclaration
   inputs [None {from=MaterialExpressionReroute_1, output=Result}]
   properties {name=OuterOutlineSize, nodeColor=<rgba 0 0 0 1>}
  }
  MaterialExpressionNamedRerouteDeclaration_3 {
   class=MaterialExpressionNamedRerouteDeclaration
   inputs [None {from=MaterialExpressionReroute_8, output=Result}]
   properties {name=InnerOutlineSize, nodeColor=<rgba 0 0 0 1>}
  }
  MaterialExpressionNamedRerouteDeclaration_4 {
   class=MaterialExpressionNamedRerouteDeclaration
   inputs [None {from=MaterialExpressionReroute_40}]
   properties {name=InnerCircle, nodeColor=<rgba 0 0 0 1>}
  }
  MaterialExpressionNamedRerouteDeclaration_5 {
   class=MaterialExpressionNamedRerouteDeclaration
   inputs [None {from=MaterialExpressionMultiply_22}]
   properties {name=InnerOutline, nodeColor=<rgba 0 0 0 1>}
  }
  MaterialExpressionNamedRerouteDeclaration_6 {
   class=MaterialExpressionNamedRerouteDeclaration
   inputs [None {from=MaterialExpressionMultiply_10}]
   properties {name=FullSizeCircle, nodeColor=<rgba 0 0 0 1>}
  }
  MaterialExpressionNamedRerouteDeclaration_7 {
   class=MaterialExpressionNamedRerouteDeclaration
   inputs [None {from=MaterialExpressionSaturate_2}]
   properties {name=FinalColors, nodeColor=<rgba 0 1 0.2369 1>}
  }
  MaterialExpressionNamedRerouteDeclaration_8 {
   class=MaterialExpressionNamedRerouteDeclaration
   inputs [None {from=MaterialExpressionMultiply_27}]
   properties {name=FinalOpacity, nodeColor=<rgba 1 0 0.0929 1>}
  }
  MaterialExpressionNamedRerouteDeclaration_9 {
   class=MaterialExpressionNamedRerouteDeclaration
   inputs [None {from=MaterialExpressionSaturate_5}]
   properties {name=FinalGradiant, nodeColor=<rgba 0 0 0 1>}
  }
  MaterialExpressionNamedRerouteUsage_0 {class=MaterialExpressionNamedRerouteUsage}
  MaterialExpressionNamedRerouteUsage_1 {class=MaterialExpressionNamedRerouteUsage}
  MaterialExpressionNamedRerouteUsage_10 {class=MaterialExpressionNamedRerouteUsage}
  MaterialExpressionNamedRerouteUsage_12 {class=MaterialExpressionNamedRerouteUsage}
  MaterialExpressionNamedRerouteUsage_13 {class=MaterialExpressionNamedRerouteUsage}
  MaterialExpressionNamedRerouteUsage_14 {class=MaterialExpressionNamedRerouteUsage}
  MaterialExpressionNamedRerouteUsage_15 {class=MaterialExpressionNamedRerouteUsage}
  MaterialExpressionNamedRerouteUsage_2 {class=MaterialExpressionNamedRerouteUsage}
  MaterialExpressionNamedRerouteUsage_23 {class=MaterialExpressionNamedRerouteUsage}
  MaterialExpressionNamedRerouteUsage_25 {class=MaterialExpressionNamedRerouteUsage}
  MaterialExpressionNamedRerouteUsage_3 {class=MaterialExpressionNamedRerouteUsage}
  MaterialExpressionNamedRerouteUsage_4 {class=MaterialExpressionNamedRerouteUsage}
  MaterialExpressionNamedRerouteUsage_5 {class=MaterialExpressionNamedRerouteUsage}
  MaterialExpressionNamedRerouteUsage_6 {class=MaterialExpressionNamedRerouteUsage}
  MaterialExpressionNamedRerouteUsage_7 {class=MaterialExpressionNamedRerouteUsage}
  MaterialExpressionNamedRerouteUsage_8 {class=MaterialExpressionNamedRerouteUsage}
  MaterialExpressionNamedRerouteUsage_9 {class=MaterialExpressionNamedRerouteUsage}
  MaterialExpressionOneMinus_0 {class=MaterialExpressionOneMinus, inputs [None {from=MaterialExpressionMaterialFunctionCall_6, output=RadialGradientExponential}]}
  MaterialExpressionOneMinus_1 {class=MaterialExpressionOneMinus, inputs [None {from=MaterialExpressionLinearInterpolate_0}]}
  MaterialExpressionOneMinus_14 {class=MaterialExpressionOneMinus, inputs [None {from=MaterialExpressionMaterialFunctionCall_24, output=Fill}]}
  MaterialExpressionOneMinus_15 {class=MaterialExpressionOneMinus, inputs [None {from=MaterialExpressionMaterialFunctionCall_25, output=Fill}]}
  MaterialExpressionOneMinus_2 {class=MaterialExpressionOneMinus, inputs [None {from=MaterialExpressionLinearInterpolate_1}]}
  MaterialExpressionPinBase_4 {
   class=MaterialExpressionPinBase
   inputs []
   properties {reroutePins [OutlineWidth {expression=&self:MaterialGraph_0.MaterialGraphNode_Composite_5.GetOutlineSize.MaterialExpressionReroute_0}]}
  }
  MaterialExpressionPinBase_5 {
   class=MaterialExpressionPinBase
   inputs [Result {from=MaterialExpressionOneMinus_1, output=""}]
   properties {reroutePins [Result {expression=&self:MaterialGraph_0.MaterialGraphNode_Composite_5.GetOutlineSize.MaterialExpressionReroute_1}]}
  }
  MaterialExpressionPinBase_6 {
   class=MaterialExpressionPinBase
   inputs []
   properties {reroutePins [OutlineWidth {expression=&self:MaterialExpressionReroute_3}]}
  }
  MaterialExpressionPinBase_7 {
   class=MaterialExpressionPinBase
   inputs [Result {from=MaterialExpressionOneMinus_2, output=""}]
   properties {reroutePins [Result {expression=&self:MaterialExpressionReroute_8}]}
  }
  MaterialExpressionReroute_0 {class=MaterialExpressionReroute, inputs [None {from=MaterialExpressionClamp_0}]}
  MaterialExpressionReroute_11 {class=MaterialExpressionReroute, inputs [None {from=MaterialExpressionMaterialFunctionCall_25, output=Fill}]}
  MaterialExpressionReroute_12 {class=MaterialExpressionReroute, inputs [None {from=MaterialExpressionSaturate_4}]}
  MaterialExpressionReroute_129 {class=MaterialExpressionReroute, inputs [None {from=MaterialExpressionComponentMask_3}]}
  MaterialExpressionReroute_159 {class=MaterialExpressionReroute, inputs [None {from=MaterialExpressionScalarParameter_48}]}
  MaterialExpressionReroute_160 {class=MaterialExpressionReroute, inputs [None {from=MaterialExpressionReroute_159}]}
  MaterialExpressionReroute_18 {class=MaterialExpressionReroute, inputs [None {from=MaterialExpressionReroute_20}]}
  MaterialExpressionReroute_185 {class=MaterialExpressionReroute, inputs [None {from=MaterialExpressionMaterialFunctionCall_24, output=Fill}]}
  MaterialExpressionReroute_20 {class=MaterialExpressionReroute, inputs [None {from=MaterialExpressionConstant_10}]}
  MaterialExpressionReroute_218 {class=MaterialExpressionReroute, inputs [None {from=MaterialExpressionReroute_185}]}
  MaterialExpressionReroute_283 {class=MaterialExpressionReroute, inputs [None {from=MaterialExpressionMultiply_26}]}
  MaterialExpressionReroute_284 {class=MaterialExpressionReroute, inputs [None {from=MaterialExpressionMultiply_23}]}
  MaterialExpressionReroute_285 {class=MaterialExpressionReroute, inputs [None {from=MaterialExpressionMultiply_25}]}
  MaterialExpressionReroute_295 {class=MaterialExpressionReroute, inputs [None {from=MaterialExpressionMultiply_31}]}
  MaterialExpressionReroute_297 {class=MaterialExpressionReroute, inputs [None {from=MaterialExpressionMaterialFunctionCall_7, output=RadialGradientExponential}]}
  MaterialExpressionReroute_3 {class=MaterialExpressionReroute, inputs [None {from=MaterialExpressionAdd_3}]}
  MaterialExpressionReroute_4 {class=MaterialExpressionReroute, inputs [None {from=MaterialExpressionSaturate_4}]}
  MaterialExpressionReroute_40 {class=MaterialExpressionReroute, inputs [None {from=MaterialExpressionReroute_11}]}
  MaterialExpressionReroute_5 {class=MaterialExpressionReroute, inputs [None {from=MaterialExpressionMaterialFunctionCall_6, output=RadialGradientExponential}]}
  MaterialExpressionReroute_54 {class=MaterialExpressionReroute, inputs [None {from=MaterialExpressionReroute_7}]}
  MaterialExpressionReroute_6 {class=MaterialExpressionReroute, inputs [None {from=MaterialExpressionReroute_297}]}
  MaterialExpressionReroute_7 {class=MaterialExpressionReroute, inputs [None {from=MaterialExpressionReroute_6}]}
  MaterialExpressionReroute_8 {class=MaterialExpressionReroute, inputs [None {from=MaterialExpressionOneMinus_2}]}
  MaterialExpressionReroute_9 {class=MaterialExpressionReroute, inputs [None {from=MaterialExpressionClamp_1}]}
  MaterialExpressionSaturate_1 {class=MaterialExpressionSaturate, inputs [None {from=MaterialExpressionAdd_17}]}
  MaterialExpressionSaturate_2 {class=MaterialExpressionSaturate, inputs [None {from=MaterialExpressionAdd_13}]}
  MaterialExpressionSaturate_3 {class=MaterialExpressionSaturate, inputs [None {from=MaterialExpressionAdd_18}]}
  MaterialExpressionSaturate_4 {class=MaterialExpressionSaturate, inputs [None {from=MaterialExpressionAdd_19}]}
  MaterialExpressionSaturate_5 {class=MaterialExpressionSaturate, inputs [None {from=MaterialExpressionAdd_1}]}
  MaterialExpressionSaturate_7 {class=MaterialExpressionSaturate, inputs [None {from=MaterialExpressionAdd_4}]}
  MaterialExpressionScalarParameter_0 {
   class=MaterialExpressionScalarParameter
   properties {
    defaultValue=1
    desc="How much we adjust the outline with size, \r\n1 = max, 0 none"
    group=General
    parameterName=Outline_SizeAdjustement
    sliderMax=1
   }
  }
  MaterialExpressionScalarParameter_1 {
   class=MaterialExpressionScalarParameter
   properties {defaultValue=0.4, group=Outer Outline, parameterName=OuterOutline_Width, sliderMax=1, sliderMin=0.0001}
  }
  MaterialExpressionScalarParameter_2 {
   class=MaterialExpressionScalarParameter
   properties {defaultValue=360, desc="The sharpness of the outline, near 0 is blurry, near 360 is sharp.", group=General, parameterName=Sharpness}
  }
  MaterialExpressionScalarParameter_3 {class=MaterialExpressionScalarParameter, properties {defaultValue=2.33, group=Gradient, parameterName=Gradient_Density, sliderMax=10}}
  MaterialExpressionScalarParameter_4 {
   class=MaterialExpressionScalarParameter
   properties {defaultValue=0.4, group=Outer Outline, parameterName=OuterOutline_Width, sliderMax=1, sliderMin=0.0001}
  }
  MaterialExpressionScalarParameter_44 {class=MaterialExpressionScalarParameter, properties {defaultValue=0.3, group=Inner Outline, parameterName=InnerOutline_Width, sliderMax=0.5}}
  MaterialExpressionScalarParameter_47 {class=MaterialExpressionScalarParameter, properties {defaultValue=0.4, group=Outer Outline, parameterName=OuterOutline_Width, sliderMax=1}}
  MaterialExpressionScalarParameter_48 {
   class=MaterialExpressionScalarParameter
   properties {defaultValue=64, desc="Size use for reference, 128px is ActionButton", parameterName=ReferenceSize, sliderMax=256, sliderMin=50}
  }
  MaterialExpressionScalarParameter_5 {class=MaterialExpressionScalarParameter, properties {defaultValue=0.4, group=Gradient, parameterName=Gradiant_Opacity, sliderMax=1}}
  MaterialExpressionScalarParameter_52 {class=MaterialExpressionScalarParameter, properties {defaultValue=0.4, group=Outer Outline, parameterName=OuterOutline_Width, sliderMax=0.5}}
  MaterialExpressionScalarParameter_6 {class=MaterialExpressionScalarParameter, properties {defaultValue=0.05, group=Gradient, parameterName=GradientFill_Opacity, sliderMax=1}}
  MaterialExpressionScalarParameter_62 {class=MaterialExpressionScalarParameter, properties {defaultValue=1, group=Outer Outline, parameterName=OuterOutline_Opacity, sliderMax=1}}
  MaterialExpressionScalarParameter_63 {class=MaterialExpressionScalarParameter, properties {defaultValue=1, group=Inner Outline, parameterName=InnerOutline_Opacity, sliderMax=1}}
  MaterialExpressionScalarParameter_64 {class=MaterialExpressionScalarParameter, properties {defaultValue=1, group=Gradient, parameterName=Fill_Opacity, sliderMax=1}}
  MaterialExpressionScalarParameter_65 {class=MaterialExpressionScalarParameter, properties {defaultValue=1, group=General, parameterName=Overall_Opacity, sliderMax=1}}
  MaterialExpressionScalarParameter_7 {
   class=MaterialExpressionScalarParameter
   properties {
    defaultValue=1
    desc="How much we adjust the outline with size, \r\n1 = max, 0 none"
    group=General
    parameterName=Outline_SizeAdjustement
    sliderMax=1
   }
  }
  MaterialExpressionStaticSwitchParameter_1 {
   class=MaterialExpressionStaticSwitchParameter
   inputs [
    True {from=MaterialExpressionSaturate_7, output=""}
    False {from=MaterialExpressionReroute_4, output=""}
   ]
   properties {defaultValue=true, group=General, parameterName=bUseGradiantFill}
  }
  MaterialExpressionStaticSwitchParameter_3 {
   class=MaterialExpressionStaticSwitchParameter
   inputs [
    True {from=MaterialExpressionReroute_5, output=""}
    False {from=MaterialExpressionOneMinus_0, output=""}
   ]
   properties {group=Gradient, parameterName=bInvertGradiantFill}
  }
  MaterialExpressionSubtract_0 {
   class=MaterialExpressionSubtract
   inputs [
    A {from=MaterialExpressionReroute_18}
    B {from=MaterialExpressionDivide_3}
   ]
  }
  MaterialExpressionVectorParameter_10 {
   class=MaterialExpressionVectorParameter
   properties {desc=Inner outline color, group=Inner Outline, parameterName=InnerOutline_Color, defaultValue=<rgba 0 1 0 1>}
  }
  MaterialExpressionVectorParameter_11 {
   class=MaterialExpressionVectorParameter
   properties {desc=Fill color, group=Fill Center, parameterName=Fill_Color, defaultValue=<rgba 0 0 1 1>}
  }
  MaterialExpressionVectorParameter_6 {
   class=MaterialExpressionVectorParameter
   properties {desc=Outer outline color, group=Outer Outline, parameterName=OuterOutline_Color, defaultValue=<rgba 1 0 0.0082 1>}
  }
 }
 properties {blendMode=BLEND_Translucent, materialDomain=MD_UI, pixelDepthOffsetMode=PDOM_Legacy, usedShadingModels=Default Lit}
}
