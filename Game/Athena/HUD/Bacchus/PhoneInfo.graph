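# Mobile (Bacchus) HUD element showing phone status: battery bars, charging icon and local time.
# On Construct it sets three looping timers at different intervals, runs the clock, battery and charging updates, and binds low-performance mode to tint the battery icon yellow.
# Updates read AthenaHUDContext: battery level becomes one to four bars, with zero showing all bars.
# The clock subtracts 12 from hours above 12, without AM/PM, so midnight shows as 0.
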WidgetBlueprint /Game/Athena/HUD/Bacchus/PhoneInfo
graphs [
 BatteryTimerUpdate {nodes [
   BatteryTimerUpdate() {is=title K2Node_FunctionEntry}
   n1 = GetBatteryLevel!(self:Athena HUDContext Object Reference=Get AthenaHUDContext!() -> (ReturnValue:Athena HUDContext Object Reference) {is=title K2Node_GetBlueprintContextEx}) -> (ReturnValue:float)
   n2 = "Nearly Equal (Float)"(A=0, B=@n1, ErrorTolerance=0)
   n3 = Get MaxBatteryBars() -> (MaxBatteryBars:double)
   n4 = "Clamp (Integer64)"!(
    Value:Integer64=Round to Integer64!(A:double=float * float(A=@n1, B=@n3) {is=title K2Node_PromotableOperator}) -> (ReturnValue:Integer64=0)
    Min:Integer64=1
    Max:Integer64=4
   ) -> (ReturnValue:Integer64=0)
   n5 = "integer64 > integer64"(A:Integer64=@n4, B:Integer64=1) -> (ReturnValue:bool=~) {is=title K2Node_PromotableOperator}
   > SetVisibility(
    self=Get 50() -> (50:Image Object Reference)
    InVisibility=Select(Option 0=Hidden, Option 1=SelfHitTestInvisible, Index=OR Boolean(A=@n5, B=@n2))
   )
   n6 = "Nearly Equal (Float)"(A=0, B=@n1, ErrorTolerance=0)
   n7 = "integer64 > integer64"(A=@n4, B=2) -> (ReturnValue=~)
   > SetVisibility(
    self=Get 75() -> (75:Image Object Reference)
    InVisibility=Select(Option 0=Hidden, Option 1=SelfHitTestInvisible, Index=OR Boolean(A=@n7, B=@n6))
   )
   n8 = "Nearly Equal (Float)"(A=0, B=@n1, ErrorTolerance=0)
   n9 = "integer64 > integer64"(A=@n4, B=3) -> (ReturnValue=~)
   > SetVisibility(
    self=Get 100() -> (100:Image Object Reference)
    InVisibility=Select(Option 0=Hidden, Option 1=SelfHitTestInvisible, Index=OR Boolean(A=@n9, B=@n8))
   )
  ]}
 ChargingIconTimerUpdate {nodes [
   ChargingIconTimerUpdate() {is=title K2Node_FunctionEntry}
   n1 = Get AthenaHUDContext()
   > SetVisibility(
    self=Get ChargingIcon() -> (ChargingIcon:Image Object Reference)
    InVisibility=Select(Option 0=HitTestInvisible, Option 1=Collapsed, Index=IsRunningOnBattery!(self:Athena HUDContext Object Reference=@n1) -> (ReturnValue:bool))
   )
  ]}
 ClockTimerUpdate {nodes [
   ClockTimerUpdate() {is=title K2Node_FunctionEntry}
   n1 = GetLocalTime!(self:Athena HUDContext Object Reference=Get AthenaHUDContext()) -> (Hours:int=0, Minutes:int=0)
   n2 = Select(
    Option 0:int=int - int(A=@n1, B=12) {is=title K2Node_CallFunction}
    Option 1:int=@n1
    Index="integer <= integer"!(A:int=@n1, B:int=12) -> (ReturnValue:bool) {is=title K2Node_CallFunction}
   ) -> (ReturnValue:int)
   > "SetText (Text)"(
    self=Get LocalTime() -> (LocalTime:Common Text Object Reference)
    InText=Format Text!(
     Format="LOCTEXT(\"{0}:{1}\")"
     0:Text="To Text (Integer)"(Value=@n2, bUseGrouping=false, MaximumIntegralDigits=2)
     1:Text="To Text (Integer)"(Value=@n1.Minutes, bUseGrouping=false, MinimumIntegralDigits=2, MaximumIntegralDigits=2)
    ) -> (Result)
   )
  ]}
 EventGraph {nodes [
   n1 = On Low Perf Mode!() -> (OutputDelegate, then, bInLowPerformanceMode:bool=~) {is=title K2Node_CustomEvent}
   > SetColorAndOpacity(
    self:Image Object Reference=Get BatteryIcon() -> (BatteryIcon:Image Object Reference)
    InColorAndOpacity:Linear Color Structure=Select(Option 0:Linear Color Structure=(R=1,G=1,B=1,A=1), Option 1:Linear Color Structure=(R=1,G=0.8036,B=0,A=1), Index=@n1.bInLowPerformanceMode) -> (ReturnValue:Linear Color Structure)
   )
   Event Construct()
   > Set Timer by Event(Delegate=Create Event(self=Self-Reference()), Time=30, bLooping=true)
   > Set Timer by Event(Delegate=Create Event(self=Self-Reference()), Time=300, bLooping=true)
   > Set Timer by Event(Delegate=Create Event(self=Self-Reference()), Time=2, bLooping=true)
   > ClockTimerUpdate() {is=title K2Node_CallFunction}
   > BatteryTimerUpdate() {is=title K2Node_CallFunction}
   > ChargingIconTimerUpdate() {is=title K2Node_CallFunction}
   n2 = Get AthenaHUDContext()
   > Bind Event to On Low Performance Mode(self:Athena HUDContext Object Reference=@n2, Delegate=@n1) {is=title K2Node_AddDelegate}
   > On Low Perf Mode!(execute, self, bInLowPerformanceMode:bool=GetInLowPerformanceMode!(self:Athena HUDContext Object Reference=@n2) -> (ReturnValue:bool)) -> (then) {is=title K2Node_CallFunction}
  ]}
]
variables [NewVar_0 {type=boolean}]
widgets {
 info {inheritedWidgetCount=3, parentClass=/Script/FortniteUI.BacchusHUDElement, rootWidgetClass=&ScaleBox, widgetCount=17}
 widgets [
  ScaleBox_HUDWidgetScale {widgetClassPath=&ScaleBox, bInherited=true, widget=ScaleBox_HUDWidgetScale, properties {stretch=UserSpecified}}
  SizeBox_1 {
   widgetClassPath=&SizeBox
   parent=ScaleBox_HUDWidgetScale
   widget=SizeBox_1
   properties {bOverride_WidthOverride=true, widthOverride=200}
   slot {class=ScaleBoxSlot, properties {horizontalAlignment=HAlign_Fill, verticalAlignment=VAlign_Fill}}
  }
  Border_Background {
   widgetClassPath=&Border
   parent=SizeBox_1
   widget=Border_Background
   properties {
    visibility=SelfHitTestInvisible
    background {
     bIsDynamicallyLoaded=false
     drawAs=RoundedBox
     imageType=NoImage
     mirroring=NoMirror
     resourceName=None
     resourceObject=null
     tiling=NoTile
     imageSize=<32 32>
     margin=<ltrb 0 0 0 0>
     outlineSettings {
      bUseBrushTransparency=false
      roundingType=FixedRadius
      width=0
      color {colorUseRule=UseColor_Specified, specifiedColor=<rgba 0 0 0 0>}
      cornerRadii=<5 5 5 5>
     }
     tintColor {colorUseRule=UseColor_Specified, specifiedColor=<rgba 1 1 1 1>}
     uVRegion {bIsValid=false, max=<0 0>, min=<0 0>}
    }
    brushColor=<rgba 0 0 0 0.2>
    padding=<ltrb 0 -5 10 -5>
   }
   slot {class=SizeBoxSlot, properties {horizontalAlignment=HAlign_Left}}
  }
  HorizontalBox_0 {
   widgetClassPath=&HorizontalBox
   parent=Border_Background
   widget=HorizontalBox_0
   slot {class=BorderSlot, properties {padding=<ltrb 0 -5 10 -5>}}
  }
  Overlay_BatteryIcon {
   widgetClassPath=&Overlay
   parent=HorizontalBox_0
   widget=Overlay_BatteryIcon
   slot {class=HorizontalBoxSlot, properties {verticalAlignment=VAlign_Center, padding=<ltrb 0 0 10 0>}}
  }
  BatteryIcon {
   widgetClassPath=&Image
   bIsVariable=true
   parent=Overlay_BatteryIcon
   widget=BatteryIcon
   properties {
    visibility=SelfHitTestInvisible
    brush {
     bIsDynamicallyLoaded=false
     drawAs=Image
     imageType=NoImage
     mirroring=NoMirror
     resourceName=None
     resourceObject=/Game/Athena/HUD/Atlas/Bacchus/battery_critical_Sprite
     tiling=NoTile
     imageSize=<40 40>
     margin=<ltrb 0 0 0 0>
     outlineSettings {
      bUseBrushTransparency=false
      roundingType=HalfHeightRadius
      width=0
      color {colorUseRule=UseColor_Specified, specifiedColor=<rgba 0 0 0 0>}
      cornerRadii=<0 0 0 0>
     }
     tintColor {colorUseRule=UseColor_Specified, specifiedColor=<rgba 1 1 1 1>}
     uVRegion {bIsValid=false, max=<0 0>, min=<0 0>}
    }
   }
   slot {class=OverlaySlot}
  }
  SizeBox_0 {
   widgetClassPath=&SizeBox
   parent=Overlay_BatteryIcon
   widget=SizeBox_0
   properties {
    bOverride_HeightOverride=true
    bOverride_WidthOverride=true
    heightOverride=11
    widthOverride=25
    renderTransform {angle=0, scale=<1 1>, shear=<0 0>, translation=<-1 0.5>}
   }
   slot {class=OverlaySlot, properties {horizontalAlignment=HAlign_Center, verticalAlignment=VAlign_Center}}
  }
  ScaleBox_0 {widgetClassPath=&ScaleBox, parent=SizeBox_0, widget=ScaleBox_0, properties {stretch=ScaleToFill}, slot {class=SizeBoxSlot}}
  HorizontalBox_1 {
   widgetClassPath=&HorizontalBox
   parent=ScaleBox_0
   widget=HorizontalBox_1
   properties {renderTransform {angle=0, scale=<1 1>, shear=<0 0>, translation=<-0.5 0>}}
   slot {class=ScaleBoxSlot, properties {horizontalAlignment=HAlign_Fill, verticalAlignment=VAlign_Fill}}
  }
  25 {
   widgetClassPath=&Image
   bIsVariable=true
   parent=HorizontalBox_1
   widget=25
   properties {
    visibility=SelfHitTestInvisible
    brush {
     bIsDynamicallyLoaded=false
     drawAs=Image
     imageType=NoImage
     mirroring=NoMirror
     resourceName=None
     resourceObject=null
     tiling=NoTile
     imageSize=<5 32>
     margin=<ltrb 0 0 0 0>
     outlineSettings {
      bUseBrushTransparency=false
      roundingType=HalfHeightRadius
      width=0
      color {colorUseRule=UseColor_Specified, specifiedColor=<rgba 0 0 0 0>}
      cornerRadii=<0 0 0 0>
     }
     tintColor {colorUseRule=UseColor_Specified, specifiedColor=<rgba 1 1 1 1>}
     uVRegion {bIsValid=false, max=<0 0>, min=<0 0>}
    }
   }
   slot {class=HorizontalBoxSlot, properties {padding=<ltrb 1.5 1.5 1 1.5>}}
  }
  50 {
   widgetClassPath=&Image
   bIsVariable=true
   parent=HorizontalBox_1
   widget=50
   properties {
    visibility=SelfHitTestInvisible
    brush {
     bIsDynamicallyLoaded=false
     drawAs=Image
     imageType=NoImage
     mirroring=NoMirror
     resourceName=None
     resourceObject=null
     tiling=NoTile
     imageSize=<5 32>
     margin=<ltrb 0 0 0 0>
     outlineSettings {
      bUseBrushTransparency=false
      roundingType=HalfHeightRadius
      width=0
      color {colorUseRule=UseColor_Specified, specifiedColor=<rgba 0 0 0 0>}
      cornerRadii=<0 0 0 0>
     }
     tintColor {colorUseRule=UseColor_Specified, specifiedColor=<rgba 1 1 1 1>}
     uVRegion {bIsValid=false, max=<0 0>, min=<0 0>}
    }
   }
   slot {class=HorizontalBoxSlot, properties {padding=<ltrb 1 1.5 1 1.5>}}
  }
  75 {
   widgetClassPath=&Image
   bIsVariable=true
   parent=HorizontalBox_1
   widget=75
   properties {
    visibility=SelfHitTestInvisible
    brush {
     bIsDynamicallyLoaded=false
     drawAs=Image
     imageType=NoImage
     mirroring=NoMirror
     resourceName=None
     resourceObject=null
     tiling=NoTile
     imageSize=<5 32>
     margin=<ltrb 0 0 0 0>
     outlineSettings {
      bUseBrushTransparency=false
      roundingType=HalfHeightRadius
      width=0
      color {colorUseRule=UseColor_Specified, specifiedColor=<rgba 0 0 0 0>}
      cornerRadii=<0 0 0 0>
     }
     tintColor {colorUseRule=UseColor_Specified, specifiedColor=<rgba 1 1 1 1>}
     uVRegion {bIsValid=false, max=<0 0>, min=<0 0>}
    }
   }
   slot {class=HorizontalBoxSlot, properties {padding=<ltrb 1 1.5 1 1.5>}}
  }
  100 {
   widgetClassPath=&Image
   bIsVariable=true
   parent=HorizontalBox_1
   widget=100
   properties {
    visibility=SelfHitTestInvisible
    brush {
     bIsDynamicallyLoaded=false
     drawAs=Image
     imageType=NoImage
     mirroring=NoMirror
     resourceName=None
     resourceObject=null
     tiling=NoTile
     imageSize=<5 32>
     margin=<ltrb 0 0 0 0>
     outlineSettings {
      bUseBrushTransparency=false
      roundingType=HalfHeightRadius
      width=0
      color {colorUseRule=UseColor_Specified, specifiedColor=<rgba 0 0 0 0>}
      cornerRadii=<0 0 0 0>
     }
     tintColor {colorUseRule=UseColor_Specified, specifiedColor=<rgba 1 1 1 1>}
     uVRegion {bIsValid=false, max=<0 0>, min=<0 0>}
    }
   }
   slot {class=HorizontalBoxSlot, properties {padding=<ltrb 1 1.5 1.5 1.5>}}
  }
  ChargingIcon {
   widgetClassPath=&Image
   bIsVariable=true
   parent=Overlay_BatteryIcon
   widget=ChargingIcon
   properties {
    visibility=SelfHitTestInvisible
    brush {
     bIsDynamicallyLoaded=false
     drawAs=Image
     imageType=NoImage
     mirroring=NoMirror
     resourceName=None
     resourceObject=/Game/Athena/HUD/Atlas/Bacchus/Charging_Icon_Sprite
     tiling=NoTile
     imageSize=<18 18>
     margin=<ltrb 0 0 0 0>
     outlineSettings {
      bUseBrushTransparency=false
      roundingType=HalfHeightRadius
      width=0
      color {colorUseRule=UseColor_Specified, specifiedColor=<rgba 0 0 0 0>}
      cornerRadii=<0 0 0 0>
     }
     tintColor {colorUseRule=UseColor_Specified, specifiedColor=<rgba 1 1 1 1>}
     uVRegion {bIsValid=false, max=<0 0>, min=<0 0>}
    }
    renderTransform {angle=0, scale=<1 1>, shear=<0 0>, translation=<27 0>}
   }
   slot {class=OverlaySlot, properties {horizontalAlignment=HAlign_Center, verticalAlignment=VAlign_Center}}
  }
  LocalTime {
   widgetClassPath=&CommonTextBlock
   bIsVariable=true
   parent=HorizontalBox_0
   widget=LocalTime
   properties {
    style=/FortUILibrary/Styles/Typography/Foundation/TextStyle-Base-S-B.TextStyle-Base-S-B_C
    text=12:59
    font {
     bForceMonospaced=false
     bMaterialIsStencil=false
     fontMaterial=null
     fontObject=&NotoSans
     letterSpacing=0
     monospacedWidth=1
     size=18
     skewAmount=0
     typefaceFontName=NotoSans-Bold
     outlineSettings {
      bApplyOutlineToDropShadows=false
      bMiteredCorners=false
      bSeparateFillAlpha=false
      outlineMaterial=null
      outlineSize=0
      outlineColor=<rgba 0 0 0 1>
     }
    }
    shadowOffset=<0 0>
   }
   slot {class=HorizontalBoxSlot, properties {horizontalAlignment=HAlign_Center, verticalAlignment=VAlign_Center, padding=<ltrb 10 0 0 0>}}
  }
  Image_Background {widgetClassPath=&Image, bInherited=true, widget=null}
  TutorialHighlight {widgetClassPath=&TutorialHighlight}
 ]
}
