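# Widget Blueprint for the touch look stick on the Bacchus HUD, providing the visual layout for the native FortTouchLookStick parent.
# A scaled canvas centers a thumb image over a background with four faint directional arrows, made from one rotated or mirrored sprite.
# The overlay also holds a collapsed gold up-arrow, tutorial highlights and a hidden DBNO carry-throw button.
# The event graph is empty, and highlight and normal colors are variables.
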WidgetBlueprint /Game/Athena/HUD/Bacchus/TouchLookStick
graphs [EventGraph {links [], nodes []}]
variables [
 HighlightColor {type=SlateColor, default {colorUseRule=UseColor_Specified, specifiedColor=<rgba 1 0.7 0.25 1>}}
 NormalColor {type=SlateColor, default {colorUseRule=UseColor_Specified, specifiedColor=<rgba 1 1 1 1>}}
]
widgets {
 info {inheritedWidgetCount=7, parentClass=/Script/FortniteUI.FortTouchLookStick, rootWidgetClass=&Overlay, widgetCount=16}
 widgets [
  Overlay_173 {widgetClassPath=&Overlay, widget=Overlay_173}
  ScaleBox_HUDWidgetScale {
   widgetClassPath=&ScaleBox
   bInherited=true
   parent=Overlay_173
   widget=ScaleBox_HUDWidgetScale
   properties {stretch=UserSpecified}
   slot {class=OverlaySlot, properties {horizontalAlignment=HAlign_Fill, verticalAlignment=VAlign_Fill}}
  }
  Overlay {widgetClassPath=&Overlay, parent=ScaleBox_HUDWidgetScale, widget=Overlay, slot {class=ScaleBoxSlot}}
  MainSizeBox {
   widgetClassPath=&SizeBox
   bInherited=true
   parent=Overlay
   widget=MainSizeBox
   slot {class=OverlaySlot, properties {horizontalAlignment=HAlign_Center, verticalAlignment=VAlign_Center}}
  }
  MoveStickCanvas {
   widgetClassPath=&CanvasPanel
   bInherited=true
   bIsVariable=true
   parent=MainSizeBox
   widget=MoveStickCanvas
   slot {class=SizeBoxSlot}
  }
  Overlay_1 {
   widgetClassPath=&Overlay
   parent=MoveStickCanvas
   widget=Overlay_1
   slot {
    class=CanvasPanelSlot
    properties {layoutData {alignment=<0 0>, anchors {maximum=<0.5 0.5>, minimum=<0.5 0.5>}, offsets=<ltrb -160 -160 320 320>}}
   }
  }
  Image_Background {
   widgetClassPath=&Image
   bInherited=true
   bIsVariable=true
   parent=Overlay_1
   widget=Image_Background
   properties {brush {
     bIsDynamicallyLoaded=false
     drawAs=Image
     imageType=FullColor
     mirroring=NoMirror
     resourceName=None
     resourceObject=/Game/Athena/HUD/Bacchus/Material/MI_UI_ButtonBG_Dark
     tiling=NoTile
     imageSize=<300 300>
     margin=<ltrb 0 0 0 0>
     outlineSettings {
      bUseBrushTransparency=false
      roundingType=HalfHeightRadius
      width=0
      color {colorUseRule=UseColor_Specified, specifiedColor=<rgba 0 0 0 0>}
      cornerRadii=<0 0 0 0>
     }
     tintColor {colorUseRule=UseColor_Specified, specifiedColor=<rgba 1 1 1 1>}
     uVRegion {bIsValid=false, max=<0 0>, min=<0 0>}
    }}
   slot {class=OverlaySlot, properties {horizontalAlignment=HAlign_Center, verticalAlignment=VAlign_Center, padding=<ltrb 10 10 10 10>}}
  }
  Tutorial_Highlight_MoveFoward {
   widgetClassPath=/Game/Athena/HUD/Tutorial/Tutorial_Highlight_MoveFoward
   bIsVariable=true
   parent=Overlay_1
   widget=Tutorial_Highlight_MoveFoward
   properties {
    visibility=Collapsed
    elementTag {tagName=HUD.Bacchus.TouchButton}
    renderTransform {angle=0, scale=<1 1>, shear=<0 0>, translation=<0 -128>}
   }
   slot {class=OverlaySlot}
  }
  TutorialHighlight {
   widgetClassPath=&TutorialHighlight
   bIsVariable=true
   parent=Overlay_1
   widget=TutorialHighlight
   properties {customHighlightWidget=/Game/Athena/HUD/Tutorial/Tutorial_Highlight_Circle.Tutorial_Highlight_Circle_C}
   slot {class=OverlaySlot, properties {horizontalAlignment=HAlign_Fill, verticalAlignment=VAlign_Fill}}
  }
  arrow_r {
   widgetClassPath=&Image
   bIsVariable=true
   parent=Overlay_1
   widget=arrow_r
   properties {
    brush {
     bIsDynamicallyLoaded=false
     drawAs=Image
     imageType=NoImage
     mirroring=NoMirror
     resourceName=None
     resourceObject=/Game/Athena/HUD/Atlas/Bacchus/T_Thumbstick_Arrow01_Sprite
     tiling=NoTile
     imageSize=<30 70>
     margin=<ltrb 0 0 0 0>
     outlineSettings {
      bUseBrushTransparency=false
      roundingType=HalfHeightRadius
      width=0
      color {colorUseRule=UseColor_Specified, specifiedColor=<rgba 0 0 0 0>}
      cornerRadii=<0 0 0 0>
     }
     tintColor {colorUseRule=UseColor_Specified, specifiedColor=<rgba 1 1 1 1>}
     uVRegion {bIsValid=false, max=<0 0>, min=<0 0>}
    }
    colorAndOpacity=<rgba 1 1 1 0.2>
   }
   slot {class=OverlaySlot, properties {horizontalAlignment=HAlign_Center, verticalAlignment=VAlign_Center, padding=<ltrb 120 0 0 0>}}
  }
  arrow_l {
   widgetClassPath=&Image
   bIsVariable=true
   parent=Overlay_1
   widget=arrow_l
   properties {
    brush {
     bIsDynamicallyLoaded=false
     drawAs=Image
     imageType=NoImage
     mirroring=NoMirror
     resourceName=None
     resourceObject=&T_Thumbstick_Arrow01_Sprite
     tiling=NoTile
     imageSize=<30 70>
     margin=<ltrb 0 0 0 0>
     outlineSettings {
      bUseBrushTransparency=false
      roundingType=HalfHeightRadius
      width=0
      color {colorUseRule=UseColor_Specified, specifiedColor=<rgba 0 0 0 0>}
      cornerRadii=<0 0 0 0>
     }
     tintColor {colorUseRule=UseColor_Specified, specifiedColor=<rgba 1 1 1 1>}
     uVRegion {bIsValid=false, max=<0 0>, min=<0 0>}
    }
    colorAndOpacity=<rgba 1 1 1 0.2>
    renderTransform {angle=0, scale=<-1 1>, shear=<0 0>, translation=<0 0>}
   }
   slot {class=OverlaySlot, properties {horizontalAlignment=HAlign_Center, verticalAlignment=VAlign_Center, padding=<ltrb 0 0 120 0>}}
  }
  arrow_u_Activated {
   widgetClassPath=&Image
   bIsVariable=true
   parent=Overlay_1
   widget=arrow_u_Activated
   properties {
    visibility=Collapsed
    brush {
     bIsDynamicallyLoaded=false
     drawAs=Image
     imageType=NoImage
     mirroring=NoMirror
     resourceName=None
     resourceObject=&T_Thumbstick_Arrow01_Sprite
     tiling=NoTile
     imageSize=<30 70>
     margin=<ltrb 0 0 0 0>
     outlineSettings {
      bUseBrushTransparency=false
      roundingType=HalfHeightRadius
      width=0
      color {colorUseRule=UseColor_Specified, specifiedColor=<rgba 0 0 0 0>}
      cornerRadii=<0 0 0 0>
     }
     tintColor {colorUseRule=UseColor_Specified, specifiedColor=<rgba 1 1 1 1>}
     uVRegion {bIsValid=false, max=<0 0>, min=<0 0>}
    }
    colorAndOpacity=<rgba 0.9131 0.6308 0.0802 1>
    renderTransform {angle=-90, scale=<1 1>, shear=<0 0>, translation=<0 0>}
   }
   slot {class=OverlaySlot, properties {horizontalAlignment=HAlign_Center, verticalAlignment=VAlign_Center, padding=<ltrb 0 -126 0 0>}}
  }
  arrow_u {
   widgetClassPath=&Image
   bIsVariable=true
   parent=Overlay_1
   widget=arrow_u
   properties {
    brush {
     bIsDynamicallyLoaded=false
     drawAs=Image
     imageType=NoImage
     mirroring=NoMirror
     resourceName=None
     resourceObject=&T_Thumbstick_Arrow01_Sprite
     tiling=NoTile
     imageSize=<30 70>
     margin=<ltrb 0 0 0 0>
     outlineSettings {
      bUseBrushTransparency=false
      roundingType=HalfHeightRadius
      width=0
      color {colorUseRule=UseColor_Specified, specifiedColor=<rgba 0 0 0 0>}
      cornerRadii=<0 0 0 0>
     }
     tintColor {colorUseRule=UseColor_Specified, specifiedColor=<rgba 1 1 1 1>}
     uVRegion {bIsValid=false, max=<0 0>, min=<0 0>}
    }
    colorAndOpacity=<rgba 1 1 1 0.2>
    renderTransform {angle=-90, scale=<1 1>, shear=<0 0>, translation=<0 0>}
   }
   slot {class=OverlaySlot, properties {horizontalAlignment=HAlign_Center, verticalAlignment=VAlign_Center, padding=<ltrb 0 -120 0 0>}}
  }
  arrow_d {
   widgetClassPath=&Image
   bIsVariable=true
   parent=Overlay_1
   widget=arrow_d
   properties {
    brush {
     bIsDynamicallyLoaded=false
     drawAs=Image
     imageType=NoImage
     mirroring=NoMirror
     resourceName=None
     resourceObject=&T_Thumbstick_Arrow01_Sprite
     tiling=NoTile
     imageSize=<30 70>
     margin=<ltrb 0 0 0 0>
     outlineSettings {
      bUseBrushTransparency=false
      roundingType=HalfHeightRadius
      width=0
      color {colorUseRule=UseColor_Specified, specifiedColor=<rgba 0 0 0 0>}
      cornerRadii=<0 0 0 0>
     }
     tintColor {colorUseRule=UseColor_Specified, specifiedColor=<rgba 1 1 1 1>}
     uVRegion {bIsValid=false, max=<0 0>, min=<0 0>}
    }
    colorAndOpacity=<rgba 1 1 1 0.2>
    renderTransform {angle=90, scale=<1 1>, shear=<0 0>, translation=<0 0>}
   }
   slot {class=OverlaySlot, properties {horizontalAlignment=HAlign_Center, verticalAlignment=VAlign_Center, padding=<ltrb 0 120 0 0>}}
  }
  Button_DBNOCarryThrow {
   widgetClassPath=/Script/FortniteUI.BacchusActionButton
   bInherited=true
   bIsVariable=true
   parent=Overlay_1
   widget=Button_DBNOCarryThrow
   properties {visibility=Hidden, elementTag {tagName=Athena.HideForCapture}}
   slot {class=OverlaySlot}
  }
  Image_Move_Thumb {
   widgetClassPath=&Image
   bInherited=true
   bIsVariable=true
   parent=MoveStickCanvas
   widget=Image_Move_Thumb
   properties {
    visibility=HitTestInvisible
    brush {
     bIsDynamicallyLoaded=false
     drawAs=Image
     imageType=FullColor
     mirroring=NoMirror
     resourceName=None
     resourceObject=/Game/Athena/HUD/Art/Materials/MI_Hud_Thumbstick_Unlocked
     tiling=NoTile
     imageSize=<150 150>
     margin=<ltrb 0 0 0 0>
     outlineSettings {
      bUseBrushTransparency=false
      roundingType=HalfHeightRadius
      width=0
      color {colorUseRule=UseColor_Specified, specifiedColor=<rgba 0 0 0 0>}
      cornerRadii=<0 0 0 0>
     }
     tintColor {colorUseRule=UseColor_Specified, specifiedColor=<rgba 1 1 1 1>}
     uVRegion {bIsValid=false, max=<0 0>, min=<0 0>}
    }
   }
   slot {
    class=CanvasPanelSlot
    properties {bAutoSize=true, layoutData {alignment=<0.5 0.5>, anchors {maximum=<0.5 0.5>, minimum=<0.5 0.5>}, offsets=<ltrb 0 0 100 30>}}
   }
  }
 ]
}
