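# Touch look-stick widget for the Bacchus HUD, a layout-only Blueprint over the native FortLookStickV2 class.
# Inherited hitbox and visual SizeBoxes wrap a scaled overlay that layers a thumbstick image beneath a smaller, centred button with background and icon.
# Both circle layers share the dark button material; the icon brush is set to draw no image.
# The event graph holds only unconnected stub events, and inherited charge, glow and notification slots are unfilled.
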WidgetBlueprint /Game/Athena/HUD/Bacchus/TouchLookStickV2
graphs [EventGraph {nodes [
   Event PreConstruct()
   Event Construct()
   Event Tick()
  ]}]
variables []
widgets {
 info {inheritedWidgetCount=13, parentClass=/Script/FortniteUI.FortLookStickV2, rootWidgetClass=&SizeBox, widgetCount=17}
 widgets [
  SizeBox_HitboxSize {
   widgetClassPath=&SizeBox
   bInherited=true
   widget=SizeBox_HitboxSize
   properties {bOverride_HeightOverride=true, bOverride_WidthOverride=true, heightOverride=200, visibility=Visible, widthOverride=200}
  }
  SizeBox_VisualSize {
   widgetClassPath=&SizeBox
   bInherited=true
   parent=SizeBox_HitboxSize
   widget=SizeBox_VisualSize
   properties {bOverride_HeightOverride=true, bOverride_WidthOverride=true, heightOverride=200, widthOverride=200}
   slot {class=SizeBoxSlot, properties {horizontalAlignment=HAlign_Center, verticalAlignment=VAlign_Center}}
  }
  ScaleBox_1 {widgetClassPath=&ScaleBox, parent=SizeBox_VisualSize, widget=ScaleBox_1, slot {class=SizeBoxSlot}}
  Overlay_0 {widgetClassPath=&Overlay, parent=ScaleBox_1, widget=Overlay_0, slot {class=ScaleBoxSlot}}
  Image_Thumbstick {
   widgetClassPath=&Image
   bInherited=true
   bIsVariable=true
   parent=Overlay_0
   widget=Image_Thumbstick
   properties {brush {
     bIsDynamicallyLoaded=false
     drawAs=Image
     imageType=FullColor
     mirroring=NoMirror
     resourceName=None
     resourceObject=/Game/Athena/HUD/Bacchus/Material/MI_UI_ButtonBG_Dark
     tiling=NoTile
     imageSize=<200 200>
     margin=<ltrb 0 0 0 0>
     outlineSettings {
      bUseBrushTransparency=false
      roundingType=HalfHeightRadius
      width=0
      color {colorUseRule=UseColor_Specified, specifiedColor=<rgba 0 0 0 0>}
      cornerRadii=<0 0 0 0>
     }
     tintColor {colorUseRule=UseColor_Specified, specifiedColor=<rgba 1 1 1 1>}
     uVRegion {bIsValid=false, max=<0 0>, min=<0 0>}
    }}
   slot {class=OverlaySlot, properties {horizontalAlignment=HAlign_Fill, verticalAlignment=VAlign_Fill}}
  }
  SizeBox_2 {
   widgetClassPath=&SizeBox
   parent=Overlay_0
   widget=SizeBox_2
   properties {bOverride_HeightOverride=true, bOverride_WidthOverride=true, heightOverride=128, widthOverride=128}
   slot {class=OverlaySlot, properties {horizontalAlignment=HAlign_Center, verticalAlignment=VAlign_Center}}
  }
  Overlay_ButtonContents {widgetClassPath=&Overlay, bInherited=true, parent=SizeBox_2, widget=Overlay_ButtonContents, slot {class=SizeBoxSlot}}
  Image_Background {
   widgetClassPath=&Image
   bInherited=true
   bIsVariable=true
   parent=Overlay_ButtonContents
   widget=Image_Background
   properties {brush {
     bIsDynamicallyLoaded=false
     drawAs=Image
     imageType=FullColor
     mirroring=NoMirror
     resourceName=None
     resourceObject=&MI_UI_ButtonBG_Dark
     tiling=NoTile
     imageSize=<128 128>
     margin=<ltrb 0 0 0 0>
     outlineSettings {
      bUseBrushTransparency=false
      roundingType=HalfHeightRadius
      width=0
      color {colorUseRule=UseColor_Specified, specifiedColor=<rgba 0 0 0 0>}
      cornerRadii=<0 0 0 0>
     }
     tintColor {colorUseRule=UseColor_Specified, specifiedColor=<rgba 1 1 1 1>}
     uVRegion {bIsValid=false, max=<0 0>, min=<0 0>}
    }}
   slot {class=OverlaySlot, properties {horizontalAlignment=HAlign_Fill, verticalAlignment=VAlign_Fill}}
  }
  SizeBox_0 {
   widgetClassPath=&SizeBox
   parent=Overlay_ButtonContents
   widget=SizeBox_0
   properties {
    bOverride_MaxDesiredHeight=true
    bOverride_MaxDesiredWidth=true
    bOverride_MinDesiredHeight=true
    bOverride_MinDesiredWidth=true
    heightOverride=128
    maxDesiredHeight=128
    maxDesiredWidth=128
    minDesiredHeight=60
    minDesiredWidth=60
    widthOverride=128
   }
   slot {class=OverlaySlot, properties {horizontalAlignment=HAlign_Center, verticalAlignment=VAlign_Center}}
  }
  Image_Icon {
   widgetClassPath=&Image
   bInherited=true
   bIsVariable=true
   parent=SizeBox_0
   widget=Image_Icon
   properties {
    visibility=HitTestInvisible
    brush {
     bIsDynamicallyLoaded=false
     drawAs=Image
     imageType=NoImage
     mirroring=NoMirror
     resourceName=None
     resourceObject=/Game/Athena/HUD/Atlas/Bacchus/EditMode_Normal_Sprite
     tiling=NoTile
     imageSize=<84 76>
     margin=<ltrb 0 0 0 0>
     outlineSettings {
      bUseBrushTransparency=false
      roundingType=HalfHeightRadius
      width=0
      color {colorUseRule=UseColor_Specified, specifiedColor=<rgba 0 0 0 0>}
      cornerRadii=<0 0 0 0>
     }
     tintColor {colorUseRule=UseColor_Specified, specifiedColor=<rgba 1 1 1 1>}
     uVRegion {bIsValid=false, max=<0 0>, min=<0 0>}
    }
    colorAndOpacity=<rgba 1 1 1 0.8>
   }
   slot {class=SizeBoxSlot}
  }
  Image_Bang_Notification {widgetClassPath=/Script/UMG.Widget, bInherited=true, widget=null}
  Image_Glow {widgetClassPath=&Widget, bInherited=true, widget=null}
  SizeBox_Charges {widgetClassPath=&Widget, bInherited=true, widget=null}
  Text_Charges {widgetClassPath=&CommonTextBlock, bInherited=true, widget=null}
  ScaleBox_Content {widgetClassPath=&ScaleBox, bInherited=true, widget=null}
  ScaleBox_HUDWidgetScale {widgetClassPath=&ScaleBox, bInherited=true, widget=null}
  TutorialHighlight {widgetClassPath=&TutorialHighlight}
 ]
}
