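# Touch HUD movement stick extending FortTouchMoveStick, with thumbstick and D-pad layouts plus an autorun lock zone above.
# Lock-zone events drive material-animated intro and looping arrows, while movement lock swaps the thumb for an animated autorun icon.
# HUD state updates hide it for passengers except in Quadcrasher or PushCannon, force it visible for aircraft drivers, and otherwise defer to the parent.
# ShowHighlight toggles a move-forward tutorial highlight.
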WidgetBlueprint /Game/Athena/HUD/Bacchus/TouchMoveStick
animations [
 Anim_Arrows_In {
  binding_count=1
  length_seconds=3.15
  object=&self:Anim_Arrows_In
  sequence {bindings [Image_AutorunDirectionArrow {
     id=#1
     possessed_class=Image
     tracks [{
       type=MovieSceneWidgetMaterialTrack
       sections [{
         type=MovieSceneParameterSection
         end_frame=63
         start_frame=0
         channels [
          BlueChannel_Opacity {type=MovieSceneScriptingFloatChannel, keys [[0,0],[5,0],[7,1],[23,1]]}
          BlueChannel_Position {type=MovieSceneScriptingFloatChannel, keys [[0,0],[5,0],[7,1]]}
          GreenChannel_Opacity {type=MovieSceneScriptingFloatChannel, keys [[0,0],[2,0],[5,1],[23,1]]}
          GreenChannel_Position {type=MovieSceneScriptingFloatChannel, keys [[0,0],[2,0],[5,1]]}
          RedChannel_Opacity {type=MovieSceneScriptingFloatChannel, keys [[0,0],[2,1],[23,1]]}
          RedChannel_Position {type=MovieSceneScriptingFloatChannel, keys [[0,0],[2,1],[23,1]]}
          All_Opacity {type=MovieSceneScriptingFloatChannel, keys [[0,1],[23,1],[26,0.25]]}
         ]
        }]
      }]
    }]}
 }
 Anim_Arrows_Loop {
  binding_count=1
  length_seconds=3
  object=&self:Anim_Arrows_Loop
  sequence {bindings [Image_AutorunDirectionArrow {
     id=#2
     possessed_class=Image
     tracks [{
       type=MovieSceneWidgetMaterialTrack
       sections [{
         type=MovieSceneParameterSection
         end_frame=60
         start_frame=0
         channels [
          All_Opacity {type=MovieSceneScriptingFloatChannel, const [1,1]}
          BlueChannel_Opacity {type=MovieSceneScriptingFloatChannel, keys [[0,0.25],[3,1],[18,1],[21,0.25]]}
          GreenChannel_Opacity {type=MovieSceneScriptingFloatChannel, keys [[0,0.25],[2,0.25],[5,1],[20,1],[23,0.25]]}
          RedChannel_Opacity {type=MovieSceneScriptingFloatChannel, keys [[0,0.25],[4,0.25],[7,1],[22,1],[25,0.25]]}
          BlueChannel_Position {type=MovieSceneScriptingFloatChannel, const [1,1]}
          GreenChannel_Position {type=MovieSceneScriptingFloatChannel, const [1,1]}
          RedChannel_Position {type=MovieSceneScriptingFloatChannel, const [1,1]}
         ]
        }]
      }]
    }]}
 }
 Anim_Autorun_Activated {
  binding_count=3
  length_seconds=1.5
  object=&self:Anim_Autorun_Activated
  sequence {bindings [
    Image_Autorun_Activated {
     id=#3
     possessed_class=Image
     tracks [
      {
       type=MovieSceneWidgetMaterialTrack
       sections [{
         type=MovieSceneParameterSection
         end_frame=30
         start_frame=0
         channels [
          Icon_RedChannel_Opacity {type=MovieSceneScriptingFloatChannel, thinned_from=4, keys [[0,1],[8.5563,1],[23,0.5]]}
          Icon_GreenChannel_Opacity {type=MovieSceneScriptingFloatChannel, keys [[0,0.5],[4.278,1],[8.5563,1],[23,0.5]]}
          Icon_BlueChannel_Opacity {type=MovieSceneScriptingFloatChannel, keys [[0,0.5],[4.278,0.5],[8.5563,1],[23,0.5]]}
          UseArc {type=MovieSceneScriptingFloatChannel, const [0,1]}
         ]
        }]
      }
      {
       type=MovieSceneByteTrack
       display_name=Visibility
       sections [{type=MovieSceneByteSection, end_frame=0, start_frame=0, channels [None {type=MovieSceneScriptingByteChannel, const [4,1]}]}]
      }
     ]
    }
    arrow_u_Activated {
     id=#4
     possessed_class=Image
     tracks [{
       type=MovieSceneByteTrack
       display_name=Visibility
       sections [{type=MovieSceneByteSection, end_frame=0, start_frame=0, channels [None {type=MovieSceneScriptingByteChannel, const [4,1]}]}]
      }]
    }
    Image_Move_Thumb {
     id=#5
     possessed_class=Image
     tracks [{
       type=MovieSceneByteTrack
       display_name=Visibility
       sections [{type=MovieSceneByteSection, end_frame=0, start_frame=0, channels [None {type=MovieSceneScriptingByteChannel, const [1,1]}]}]
      }]
    }
   ]}
 }
 Anim_Autorun_Deactivated {
  binding_count=3
  length_seconds=1
  object=&self:Anim_Autorun_Deactivated
  sequence {bindings [
    Image_Autorun_Activated {
     id=#3
     possessed_class=Image
     tracks [{
       type=MovieSceneByteTrack
       display_name=Visibility
       sections [{type=MovieSceneByteSection, end_frame=0, start_frame=0, channels [None {type=MovieSceneScriptingByteChannel, const [1,1]}]}]
      }]
    }
    arrow_u_Activated {
     id=#6
     possessed_class=Image
     tracks [{
       type=MovieSceneByteTrack
       display_name=Visibility
       sections [{type=MovieSceneByteSection, end_frame=0, start_frame=0, channels [None {type=MovieSceneScriptingByteChannel, const [1,1]}]}]
      }]
    }
    Image_Move_Thumb {
     id=#5
     possessed_class=Image
     tracks [{
       type=MovieSceneByteTrack
       display_name=Visibility
       sections [{type=MovieSceneByteSection, end_frame=0, start_frame=0, channels [None {type=MovieSceneScriptingByteChannel, const [4,1]}]}]
      }]
    }
   ]}
 }
]
graphs [
 EventGraph {nodes [
   n1 = Event SetMovementLockStatus!() -> (OutputDelegate, then, MovementIsLocked:bool) {is=title K2Node_Event}
   > Parent: SetMovementLockStatus!(execute, MovementIsLocked:bool=@n1.MovementIsLocked) -> (then) {is=title K2Node_CallParentFunction}
   > SetImprovedLockedMovestick!(execute, self, IsLocked:bool=@n1.MovementIsLocked) -> (then)
   n2 = Event ShowHighlight!() -> (OutputDelegate, then, Show:bool) {is=title K2Node_Event}
   > n3 = Branch(Condition=@n2.Show)
   n4 = Get Tutorial_Highlight_MoveFoward() -> (Tutorial_Highlight_MoveFoward:Tutorial Highlight Move Foward Object Reference)
   > SetVisibility(self=@n4, InVisibility=SelfHitTestInvisible)
   SetVisibility(execute=@n3.else, self=@n4, InVisibility=Collapsed)
   n5 = Event OnHUDStateUpdate!() -> (OutputDelegate, then, NewState:"Fort HUD State Structure (by ref)") {is=title K2Node_Event}
   n6 = IsRidingQuadcrasher!() -> (bIsRidingQuadcrasher:bool)
   n7 = IsInPushCannon!() -> (bIsInPushCannon:bool)
   > n8 = Branch(Condition=AND Boolean!(A=IsPassenger!() -> (bIsPassenger:bool), B=NOT Boolean(A=@n6), C:bool=NOT Boolean(A=@n7)) -> (ReturnValue))
   > SetVisibility(InVisibility=Collapsed)
   n9 = Branch(execute=@n8.else, Condition=IsAircraftDriver!() -> (IsAircraftDriver:bool))
   > SetVisibility(InVisibility=Visible)
   Parent: OnHUDStateUpdate!(execute=@n9.else, NewState:"Fort HUD State Structure (by ref)"=@n5.NewState) -> (then) {is=title K2Node_CallParentFunction}
   n10 = Event OnDisplayAutorunLockZone!() -> (OutputDelegate, then, bShouldShowArrow:bool) {is=title K2Node_Event}
   > n11 = Set CachedShouldShowArrow(CachedShouldShowArrow:bool=@n10.bShouldShowArrow) -> (Output_Get:bool)
   > n12 = Branch(Condition=@n11)
   n13 = Get AutoRunAnimSpeed() -> (AutoRunAnimSpeed:float)
   > n14 = Play Animation with Finished event!(
    execute:Exec
    Widget:User Widget Object Reference=Self-Reference()
    InAnimation:Widget Animation Object Reference=Get Anim_Arrows_In() -> (Anim_Arrows_In:Widget Animation Object Reference)
    StartAtTime:float=0
    NumLoopsToPlay:int=1
    PlayMode:EUMGSequencePlayMode Enum
    PlaybackSpeed:float=@n13
   ) -> (then:Exec, Result:Widget Animation Handle Structure, Finished:Exec) {is=title K2Node_PlayAnimation2}
   PlayAnimation(
    execute=@n14.Finished
    InAnimation=Get Anim_Arrows_Loop() -> (Anim_Arrows_Loop:Widget Animation Object Reference)
    NumLoopsToPlay=0
    PlaybackSpeed=@n13
   )
   > DelayUntilNextTick!(execute) -> (then)
   > n15 = Branch(Condition=Get CachedShouldShowArrow() -> (CachedShouldShowArrow:bool))
   StopAnimation(execute=@n15.else, InAnimation=Get Anim_Arrows_Loop())
   n16 = Cast To FortPlayerController(execute=@n12.else, Object=GetPlayerController!(PlayerIndex:int=0) -> (ReturnValue:Player Controller Object Reference))
   > n17 = Branch(Condition=GetAutoRunOn!(self:Fort Player Controller Object Reference=@n16) -> (ReturnValue:bool))
   StopAnimation(execute=[@n17.then, @n16.CastFailed], InAnimation=Get Anim_Arrows_In())
   > StopAnimation(InAnimation=Get Anim_Arrows_Loop())
   "Animation Started (Anim_Arrows_In)"!() -> (OutputDelegate:Delegate, then:Exec) {is=title K2Node_WidgetAnimationEvent}
   > StopAnimation(InAnimation=Get Anim_Arrows_Loop())
  ]}
 IsAircraftDriver {nodes [
   IsAircraftDriver() {is=title K2Node_FunctionEntry}
   > n1 = Cast To FortPlayerController(Object=GetOwningPlayer())
   > n2 = GetAircraftComponent!(execute, self:Fort Player Controller Object Reference=@n1) -> (then, ReturnValue:Fort Controller Component Aircraft Object Reference)
   > Return Node!(
    execute
    IsAircraftDriver:bool=IsAircraftDriver!(self:Fort Controller Component Aircraft Object Reference=@n2) -> (ReturnValue:bool) {is=title K2Node_CallFunction}
   )
  ]}
 IsInPushCannon {nodes [
   IsInPushCannon() {is=title K2Node_FunctionEntry}
   > n1 = Cast To FortPlayerPawn(Object=Get Controlled Pawn!(self:Controller Object Reference=GetPlayerController(PlayerIndex=0)) -> (ReturnValue:Pawn Object Reference) {is=title K2Node_CallFunction})
   > n2 = Branch(Condition=IsInVehicle!(self:Fort Player Pawn Object Reference=@n1) -> (ReturnValue:bool))
   n3 = GetVehicleInterface!(self:Fort Player Pawn Object Reference=@n1) -> (ReturnValue:Fort Vehicle Interface Interface)
   > Return Node!(
    execute
    bIsInPushCannon:bool=HasMatchingGameplayTag!(
     self:Gameplay Tag Asset Interface Interface=GetVehicleAbilitySystemComponent!(self:Fort Vehicle Interface Interface=@n3) -> (ReturnValue:Ability System Component Object Reference)
     TagToCheck:Gameplay Tag Structure=(TagName="Athena.Vehicle.PushCannon")
    ) -> (ReturnValue:bool)
   )
   Return Node(execute=[@n1.CastFailed, @n2.else], bIsInPushCannon=false)
  ]}
 IsPassenger {nodes [
   IsPassenger() {is=title K2Node_FunctionEntry}
   > n1 = Cast To FortPlayerPawn(Object=Get Controlled Pawn(self=GetPlayerController(PlayerIndex=0)))
   > n2 = Branch(Condition=IsPassengerInVehicle!(self:Fort Player Pawn Object Reference=@n1) -> (ReturnValue:bool))
   > Return Node!(execute, bIsPassenger:bool=true)
   Return Node(execute=@n2.else, bIsPassenger=false)
   Return Node(execute=@n1.CastFailed, bIsPassenger=false)
  ]}
 IsRidingQuadcrasher {nodes [
   IsRidingQuadcrasher() {is=title K2Node_FunctionEntry}
   > n1 = Cast To FortPlayerPawn(Object=Get Controlled Pawn(self=GetPlayerController(PlayerIndex=0)))
   > n2 = Branch(Condition=IsInVehicle(self=@n1))
   n3 = GetVehicleInterface(self=@n1)
   > Return Node!(
    execute
    bIsRidingQuadcrasher:bool=HasTag!(
     TagContainer:"Gameplay Tag Container Structure (by ref)"=GetVehicleTags!(self:Fort Vehicle Interface Interface=@n3) -> (ReturnValue:"Gameplay Tag Container Structure (by ref)")
     Tag:Gameplay Tag Structure=(TagName="Vehicle.Antelope")
     bExactMatch:bool=true
    ) -> (ReturnValue:bool)
   )
   Return Node(execute=[@n1.CastFailed, @n2.else], bIsRidingQuadcrasher=false)
  ]}
 SetImprovedLockedMovestick {nodes [
   n1 = SetImprovedLockedMovestick!() -> (then, IsLocked:bool=~) {is=title K2Node_FunctionEntry}
   > StopAllAnimations(self:User Widget Object Reference)
   > n2 = Branch(Condition=@n1)
   > PlayAnimationForward(InAnimation=Get Anim_Autorun_Activated() -> (Anim_Autorun_Activated:Widget Animation Object Reference))
   PlayAnimation(execute=@n2.else, InAnimation=Get Anim_Autorun_Deactivated() -> (Anim_Autorun_Deactivated:Widget Animation Object Reference))
  ]}
]
variables [
 AutoRunAnimSpeed {type=number, default=1}
 AutorunLockZoneArrowRenderScale {type=Vector2D, default=<0 0>}
 AutorunLockZoneRenderScale {type=Vector2D, default=<0 0>}
 CachedShouldShowArrow {type=boolean}
 HighlightColor {type=SlateColor, default {colorUseRule=UseColor_Specified, specifiedColor=<rgba 1 0.7 0.25 1>}}
 NormalColor {type=SlateColor, default {colorUseRule=UseColor_Specified, specifiedColor=<rgba 1 1 1 1>}}
]
widgets {
 info {inheritedWidgetCount=16, parentClass=/Script/FortniteUI.FortTouchMoveStick, rootWidgetClass=&Overlay, widgetCount=33}
 widgets [
  Overlay_173 {widgetClassPath=&Overlay, widget=Overlay_173}
  SizeBox_AutorunLockZoneHidden {
   widgetClassPath=&SizeBox
   bInherited=true
   bIsVariable=true
   parent=Overlay_173
   widget=SizeBox_AutorunLockZoneHidden
   properties {bOverride_HeightOverride=true, heightOverride=395}
   slot {class=OverlaySlot, properties {horizontalAlignment=HAlign_Fill, verticalAlignment=VAlign_Fill, padding=<ltrb 0 -395 0 0>}}
  }
  Overlay_AutorunLockZone {
   widgetClassPath=&Overlay
   bInherited=true
   bIsVariable=true
   parent=SizeBox_AutorunLockZoneHidden
   widget=Overlay_AutorunLockZone
   slot {class=SizeBoxSlot}
  }
  VerticalBox_2 {
   widgetClassPath=&VerticalBox
   parent=Overlay_AutorunLockZone
   widget=VerticalBox_2
   slot {class=OverlaySlot, properties {horizontalAlignment=HAlign_Fill, verticalAlignment=VAlign_Fill}}
  }
  SizeBox_AutorunLockZoneSize {
   widgetClassPath=&SizeBox
   bInherited=true
   parent=VerticalBox_2
   widget=SizeBox_AutorunLockZoneSize
   properties {bOverride_HeightOverride=true, bOverride_WidthOverride=true, heightOverride=160, widthOverride=160}
   slot {class=VerticalBoxSlot}
  }
  AutorunLockZone {
   widgetClassPath=&/AutorunLockZone
   bInherited=true
   bIsVariable=true
   parent=SizeBox_AutorunLockZoneSize
   widget=AutorunLockZone
   properties {bIsAutorunLockEnabled=true, bIsMoveStickAutorun=true, visibility=Collapsed}
   slot {class=SizeBoxSlot, properties {padding=<ltrb 0 30 0 30>}}
  }
  SizeBox_AutorunDirectionArrowSpaceSize {
   widgetClassPath=&SizeBox
   bInherited=true
   parent=VerticalBox_2
   widget=SizeBox_AutorunDirectionArrowSpaceSize
   properties {bOverride_HeightOverride=true, heightOverride=100}
   slot {class=VerticalBoxSlot, properties {padding=<ltrb 0 -20 0 0>}}
  }
  Image_AutorunDirectionArrow {
   widgetClassPath=&Image
   bInherited=true
   bIsVariable=true
   parent=SizeBox_AutorunDirectionArrowSpaceSize
   widget=Image_AutorunDirectionArrow
   properties {
    visibility=Collapsed
    brush {
     bIsDynamicallyLoaded=false
     drawAs=Image
     imageType=NoImage
     mirroring=NoMirror
     resourceName=None
     resourceObject=/Game/Athena/HUD/Bacchus/Material/Autorun/M_Autorun_Arrow_Inst
     tiling=NoTile
     imageSize=<128 128>
     margin=<ltrb 0 0 0 0>
     outlineSettings {
      bUseBrushTransparency=false
      roundingType=HalfHeightRadius
      width=0
      color {colorUseRule=UseColor_Specified, specifiedColor=<rgba 0 0 0 0>}
      cornerRadii=<0 0 0 0>
     }
     tintColor {colorUseRule=UseColor_Specified, specifiedColor=<rgba 1 1 1 1>}
     uVRegion {bIsValid=false, max=<0 0>, min=<0 0>}
    }
    renderTransform {angle=0, scale=<1 1.2>, shear=<0 0>, translation=<0 0>}
   }
   slot {class=SizeBoxSlot, properties {horizontalAlignment=HAlign_Center, verticalAlignment=VAlign_Center}}
  }
  ScaleBox_HUDWidgetScale {
   widgetClassPath=&ScaleBox
   bInherited=true
   parent=Overlay_173
   widget=ScaleBox_HUDWidgetScale
   properties {stretch=UserSpecified}
   slot {class=OverlaySlot, properties {horizontalAlignment=HAlign_Fill, verticalAlignment=VAlign_Fill}}
  }
  Overlay {widgetClassPath=&Overlay, parent=ScaleBox_HUDWidgetScale, widget=Overlay, slot {class=ScaleBoxSlot}}
  MoveStickCanvas {
   widgetClassPath=&CanvasPanel
   bInherited=true
   bIsVariable=true
   parent=Overlay
   widget=MoveStickCanvas
   slot {class=OverlaySlot, properties {horizontalAlignment=HAlign_Center, verticalAlignment=VAlign_Center}}
  }
  Overlay_1 {
   widgetClassPath=&Overlay
   parent=MoveStickCanvas
   widget=Overlay_1
   slot {
    class=CanvasPanelSlot
    properties {layoutData {alignment=<0 0>, anchors {maximum=<0.5 0.5>, minimum=<0.5 0.5>}, offsets=<ltrb -160 -160 320 320>}}
   }
  }
  Image_Background {
   widgetClassPath=&Image
   bInherited=true
   bIsVariable=true
   parent=Overlay_1
   widget=Image_Background
   properties {brush {
     bIsDynamicallyLoaded=false
     drawAs=Image
     imageType=FullColor
     mirroring=NoMirror
     resourceName=None
     resourceObject=/Game/Athena/HUD/Bacchus/Material/MI_UI_ButtonBG_Dark
     tiling=NoTile
     imageSize=<300 300>
     margin=<ltrb 0 0 0 0>
     outlineSettings {
      bUseBrushTransparency=false
      roundingType=HalfHeightRadius
      width=0
      color {colorUseRule=UseColor_Specified, specifiedColor=<rgba 0 0 0 0>}
      cornerRadii=<0 0 0 0>
     }
     tintColor {colorUseRule=UseColor_Specified, specifiedColor=<rgba 1 1 1 1>}
     uVRegion {bIsValid=false, max=<0 0>, min=<0 0>}
    }}
   slot {class=OverlaySlot, properties {horizontalAlignment=HAlign_Center, verticalAlignment=VAlign_Center, padding=<ltrb 10 10 10 10>}}
  }
  Tutorial_Highlight_MoveFoward {
   widgetClassPath=/Game/Athena/HUD/Tutorial/Tutorial_Highlight_MoveFoward
   bIsVariable=true
   parent=Overlay_1
   widget=Tutorial_Highlight_MoveFoward
   properties {
    visibility=Collapsed
    elementTag {tagName=HUD.Bacchus.TouchButton}
    renderTransform {angle=0, scale=<1 1>, shear=<0 0>, translation=<0 -128>}
   }
   slot {class=OverlaySlot}
  }
  TutorialHighlight {
   widgetClassPath=&TutorialHighlight
   bIsVariable=true
   parent=Overlay_1
   widget=TutorialHighlight
   properties {customHighlightWidget=/Game/Athena/HUD/Tutorial/Tutorial_Highlight_Circle.Tutorial_Highlight_Circle_C}
   slot {class=OverlaySlot, properties {horizontalAlignment=HAlign_Fill, verticalAlignment=VAlign_Fill}}
  }
  arrow_r {
   widgetClassPath=&Image
   bIsVariable=true
   parent=Overlay_1
   widget=arrow_r
   properties {
    brush {
     bIsDynamicallyLoaded=false
     drawAs=Image
     imageType=NoImage
     mirroring=NoMirror
     resourceName=None
     resourceObject=/Game/Athena/HUD/Atlas/Bacchus/T_Thumbstick_Arrow01_Sprite
     tiling=NoTile
     imageSize=<30 70>
     margin=<ltrb 0 0 0 0>
     outlineSettings {
      bUseBrushTransparency=false
      roundingType=HalfHeightRadius
      width=0
      color {colorUseRule=UseColor_Specified, specifiedColor=<rgba 0 0 0 0>}
      cornerRadii=<0 0 0 0>
     }
     tintColor {colorUseRule=UseColor_Specified, specifiedColor=<rgba 1 1 1 1>}
     uVRegion {bIsValid=false, max=<0 0>, min=<0 0>}
    }
    colorAndOpacity=<rgba 1 1 1 0.2>
   }
   slot {class=OverlaySlot, properties {horizontalAlignment=HAlign_Center, verticalAlignment=VAlign_Center, padding=<ltrb 120 0 0 0>}}
  }
  arrow_l {
   widgetClassPath=&Image
   bIsVariable=true
   parent=Overlay_1
   widget=arrow_l
   properties {
    brush {
     bIsDynamicallyLoaded=false
     drawAs=Image
     imageType=NoImage
     mirroring=NoMirror
     resourceName=None
     resourceObject=&T_Thumbstick_Arrow01_Sprite
     tiling=NoTile
     imageSize=<30 70>
     margin=<ltrb 0 0 0 0>
     outlineSettings {
      bUseBrushTransparency=false
      roundingType=HalfHeightRadius
      width=0
      color {colorUseRule=UseColor_Specified, specifiedColor=<rgba 0 0 0 0>}
      cornerRadii=<0 0 0 0>
     }
     tintColor {colorUseRule=UseColor_Specified, specifiedColor=<rgba 1 1 1 1>}
     uVRegion {bIsValid=false, max=<0 0>, min=<0 0>}
    }
    colorAndOpacity=<rgba 1 1 1 0.2>
    renderTransform {angle=0, scale=<-1 1>, shear=<0 0>, translation=<0 0>}
   }
   slot {class=OverlaySlot, properties {horizontalAlignment=HAlign_Center, verticalAlignment=VAlign_Center, padding=<ltrb 0 0 120 0>}}
  }
  arrow_u_Activated {
   widgetClassPath=&Image
   bIsVariable=true
   parent=Overlay_1
   widget=arrow_u_Activated
   properties {
    visibility=Collapsed
    brush {
     bIsDynamicallyLoaded=false
     drawAs=Image
     imageType=NoImage
     mirroring=NoMirror
     resourceName=None
     resourceObject=&T_Thumbstick_Arrow01_Sprite
     tiling=NoTile
     imageSize=<30 70>
     margin=<ltrb 0 0 0 0>
     outlineSettings {
      bUseBrushTransparency=false
      roundingType=HalfHeightRadius
      width=0
      color {colorUseRule=UseColor_Specified, specifiedColor=<rgba 0 0 0 0>}
      cornerRadii=<0 0 0 0>
     }
     tintColor {colorUseRule=UseColor_Specified, specifiedColor=<rgba 1 1 1 1>}
     uVRegion {bIsValid=false, max=<0 0>, min=<0 0>}
    }
    colorAndOpacity=<rgba 0.9131 0.6308 0.0802 1>
    renderTransform {angle=-90, scale=<1 1>, shear=<0 0>, translation=<0 0>}
   }
   slot {class=OverlaySlot, properties {horizontalAlignment=HAlign_Center, verticalAlignment=VAlign_Center, padding=<ltrb 0 -126 0 0>}}
  }
  arrow_u {
   widgetClassPath=&Image
   bIsVariable=true
   parent=Overlay_1
   widget=arrow_u
   properties {
    brush {
     bIsDynamicallyLoaded=false
     drawAs=Image
     imageType=NoImage
     mirroring=NoMirror
     resourceName=None
     resourceObject=&T_Thumbstick_Arrow01_Sprite
     tiling=NoTile
     imageSize=<30 70>
     margin=<ltrb 0 0 0 0>
     outlineSettings {
      bUseBrushTransparency=false
      roundingType=HalfHeightRadius
      width=0
      color {colorUseRule=UseColor_Specified, specifiedColor=<rgba 0 0 0 0>}
      cornerRadii=<0 0 0 0>
     }
     tintColor {colorUseRule=UseColor_Specified, specifiedColor=<rgba 1 1 1 1>}
     uVRegion {bIsValid=false, max=<0 0>, min=<0 0>}
    }
    colorAndOpacity=<rgba 1 1 1 0.2>
    renderTransform {angle=-90, scale=<1 1>, shear=<0 0>, translation=<0 0>}
   }
   slot {class=OverlaySlot, properties {horizontalAlignment=HAlign_Center, verticalAlignment=VAlign_Center, padding=<ltrb 0 -120 0 0>}}
  }
  arrow_d {
   widgetClassPath=&Image
   bIsVariable=true
   parent=Overlay_1
   widget=arrow_d
   properties {
    brush {
     bIsDynamicallyLoaded=false
     drawAs=Image
     imageType=NoImage
     mirroring=NoMirror
     resourceName=None
     resourceObject=&T_Thumbstick_Arrow01_Sprite
     tiling=NoTile
     imageSize=<30 70>
     margin=<ltrb 0 0 0 0>
     outlineSettings {
      bUseBrushTransparency=false
      roundingType=HalfHeightRadius
      width=0
      color {colorUseRule=UseColor_Specified, specifiedColor=<rgba 0 0 0 0>}
      cornerRadii=<0 0 0 0>
     }
     tintColor {colorUseRule=UseColor_Specified, specifiedColor=<rgba 1 1 1 1>}
     uVRegion {bIsValid=false, max=<0 0>, min=<0 0>}
    }
    colorAndOpacity=<rgba 1 1 1 0.2>
    renderTransform {angle=90, scale=<1 1>, shear=<0 0>, translation=<0 0>}
   }
   slot {class=OverlaySlot, properties {horizontalAlignment=HAlign_Center, verticalAlignment=VAlign_Center, padding=<ltrb 0 120 0 0>}}
  }
  Image_Move_Thumb {
   widgetClassPath=&Image
   bInherited=true
   bIsVariable=true
   parent=MoveStickCanvas
   widget=Image_Move_Thumb
   properties {
    visibility=HitTestInvisible
    brush {
     bIsDynamicallyLoaded=false
     drawAs=Image
     imageType=FullColor
     mirroring=NoMirror
     resourceName=None
     resourceObject=/Game/Athena/HUD/Art/Materials/MI_Hud_Thumbstick_Unlocked
     tiling=NoTile
     imageSize=<150 150>
     margin=<ltrb 0 0 0 0>
     outlineSettings {
      bUseBrushTransparency=false
      roundingType=HalfHeightRadius
      width=0
      color {colorUseRule=UseColor_Specified, specifiedColor=<rgba 0 0 0 0>}
      cornerRadii=<0 0 0 0>
     }
     tintColor {colorUseRule=UseColor_Specified, specifiedColor=<rgba 1 1 1 1>}
     uVRegion {bIsValid=false, max=<0 0>, min=<0 0>}
    }
   }
   slot {
    class=CanvasPanelSlot
    properties {bAutoSize=true, layoutData {alignment=<0.5 0.5>, anchors {maximum=<0.5 0.5>, minimum=<0.5 0.5>}, offsets=<ltrb 0 0 100 30>}}
   }
  }
  Image_Autorun_Activated {
   widgetClassPath=&Image
   bIsVariable=true
   parent=MoveStickCanvas
   widget=Image_Autorun_Activated
   properties {
    visibility=Collapsed
    brush {
     bIsDynamicallyLoaded=false
     drawAs=Image
     imageType=NoImage
     mirroring=NoMirror
     resourceName=None
     resourceObject=/Game/Athena/HUD/Bacchus/Material/Autorun/M_Autorun_Button_Inst
     tiling=NoTile
     imageSize=<128 128>
     margin=<ltrb 0 0 0 0>
     outlineSettings {
      bUseBrushTransparency=false
      roundingType=HalfHeightRadius
      width=0
      color {colorUseRule=UseColor_Specified, specifiedColor=<rgba 0 0 0 0>}
      cornerRadii=<0 0 0 0>
     }
     tintColor {colorUseRule=UseColor_Specified, specifiedColor=<rgba 1 1 1 1>}
     uVRegion {bIsValid=false, max=<0 0>, min=<0 0>}
    }
    renderTransform {angle=0, scale=<1.9 1.9>, shear=<0 0>, translation=<0 0>}
   }
   slot {
    class=CanvasPanelSlot
    properties {bAutoSize=true, layoutData {alignment=<0.5 0.5>, anchors {maximum=<0.5 0.5>, minimum=<0.5 0.5>}, offsets=<ltrb 0 0 100 30>}}
   }
  }
  DpadCanvas {
   widgetClassPath=&CanvasPanel
   bInherited=true
   bIsVariable=true
   parent=Overlay
   widget=DpadCanvas
   slot {class=OverlaySlot, properties {horizontalAlignment=HAlign_Center, verticalAlignment=VAlign_Center}}
  }
  Overlay_2 {
   widgetClassPath=&Overlay
   parent=DpadCanvas
   widget=Overlay_2
   slot {
    class=CanvasPanelSlot
    properties {layoutData {alignment=<0 0>, anchors {maximum=<0.5 0.5>, minimum=<0.5 0.5>}, offsets=<ltrb -160 -160 320 320>}}
   }
  }
  Image_Dpad_Background {
   widgetClassPath=&Image
   parent=Overlay_2
   widget=Image_Dpad_Background
   properties {brush {
     bIsDynamicallyLoaded=false
     drawAs=Image
     imageType=FullColor
     mirroring=NoMirror
     resourceName=None
     resourceObject=&MI_UI_ButtonBG_Dark
     tiling=NoTile
     imageSize=<32 32>
     margin=<ltrb 0 0 0 0>
     outlineSettings {
      bUseBrushTransparency=false
      roundingType=HalfHeightRadius
      width=0
      color {colorUseRule=UseColor_Specified, specifiedColor=<rgba 0 0 0 0>}
      cornerRadii=<0 0 0 0>
     }
     tintColor {colorUseRule=UseColor_Specified, specifiedColor=<rgba 1 1 1 1>}
     uVRegion {bIsValid=false, max=<0 0>, min=<0 0>}
    }}
   slot {class=OverlaySlot, properties {horizontalAlignment=HAlign_Fill, verticalAlignment=VAlign_Fill, padding=<ltrb -30 -30 -30 -30>}}
  }
  arrow_r_1 {
   widgetClassPath=&Image
   bIsVariable=true
   parent=Overlay_2
   widget=arrow_r_1
   properties {brush {
     bIsDynamicallyLoaded=false
     drawAs=Image
     imageType=FullColor
     mirroring=NoMirror
     resourceName=None
     resourceObject=/Game/Athena/HUD/Atlas/Bacchus/T_UI_Dpad_Right_Sprite
     tiling=NoTile
     imageSize=<96 96>
     margin=<ltrb 0 0 0 0>
     outlineSettings {
      bUseBrushTransparency=false
      roundingType=HalfHeightRadius
      width=0
      color {colorUseRule=UseColor_Specified, specifiedColor=<rgba 0 0 0 0>}
      cornerRadii=<0 0 0 0>
     }
     tintColor {colorUseRule=UseColor_Specified, specifiedColor=<rgba 1 1 1 1>}
     uVRegion {bIsValid=false, max=<0 0>, min=<0 0>}
    }}
   slot {class=OverlaySlot, properties {horizontalAlignment=HAlign_Center, verticalAlignment=VAlign_Center, padding=<ltrb 120 0 0 0>}}
  }
  Image_DpadArrowRightHighlight {
   widgetClassPath=&Image
   bInherited=true
   bIsVariable=true
   parent=Overlay_2
   widget=Image_DpadArrowRightHighlight
   properties {
    visibility=Collapsed
    brush {
     bIsDynamicallyLoaded=false
     drawAs=Image
     imageType=FullColor
     mirroring=NoMirror
     resourceName=None
     resourceObject=&T_UI_Dpad_Right_Sprite
     tiling=NoTile
     imageSize=<96 96>
     margin=<ltrb 0 0 0 0>
     outlineSettings {
      bUseBrushTransparency=false
      roundingType=HalfHeightRadius
      width=0
      color {colorUseRule=UseColor_Specified, specifiedColor=<rgba 0 0 0 0>}
      cornerRadii=<0 0 0 0>
     }
     tintColor {colorUseRule=UseColor_Specified, specifiedColor=<rgba 1 1 1 1>}
     uVRegion {bIsValid=false, max=<0 0>, min=<0 0>}
    }
    colorAndOpacity=<rgba 0.9131 0.6308 0.0802 1>
   }
   slot {class=OverlaySlot, properties {horizontalAlignment=HAlign_Center, verticalAlignment=VAlign_Center, padding=<ltrb 120 0 0 0>}}
  }
  arrow_l_1 {
   widgetClassPath=&Image
   bIsVariable=true
   parent=Overlay_2
   widget=arrow_l_1
   properties {
    brush {
     bIsDynamicallyLoaded=false
     drawAs=Image
     imageType=FullColor
     mirroring=NoMirror
     resourceName=None
     resourceObject=&T_UI_Dpad_Right_Sprite
     tiling=NoTile
     imageSize=<96 96>
     margin=<ltrb 0 0 0 0>
     outlineSettings {
      bUseBrushTransparency=false
      roundingType=HalfHeightRadius
      width=0
      color {colorUseRule=UseColor_Specified, specifiedColor=<rgba 0 0 0 0>}
      cornerRadii=<0 0 0 0>
     }
     tintColor {colorUseRule=UseColor_Specified, specifiedColor=<rgba 1 1 1 1>}
     uVRegion {bIsValid=false, max=<0 0>, min=<0 0>}
    }
    renderTransform {angle=0, scale=<-1 1>, shear=<0 0>, translation=<0 0>}
   }
   slot {class=OverlaySlot, properties {horizontalAlignment=HAlign_Center, verticalAlignment=VAlign_Center, padding=<ltrb 0 0 120 0>}}
  }
  Image_DpadArrowLeftHighlight {
   widgetClassPath=&Image
   bInherited=true
   bIsVariable=true
   parent=Overlay_2
   widget=Image_DpadArrowLeftHighlight
   properties {
    visibility=Collapsed
    brush {
     bIsDynamicallyLoaded=false
     drawAs=Image
     imageType=FullColor
     mirroring=NoMirror
     resourceName=None
     resourceObject=&T_UI_Dpad_Right_Sprite
     tiling=NoTile
     imageSize=<96 96>
     margin=<ltrb 0 0 0 0>
     outlineSettings {
      bUseBrushTransparency=false
      roundingType=HalfHeightRadius
      width=0
      color {colorUseRule=UseColor_Specified, specifiedColor=<rgba 0 0 0 0>}
      cornerRadii=<0 0 0 0>
     }
     tintColor {colorUseRule=UseColor_Specified, specifiedColor=<rgba 1 1 1 1>}
     uVRegion {bIsValid=false, max=<0 0>, min=<0 0>}
    }
    colorAndOpacity=<rgba 0.9131 0.6308 0.0802 1>
    renderTransform {angle=0, scale=<-1 1>, shear=<0 0>, translation=<0 0>}
   }
   slot {class=OverlaySlot, properties {horizontalAlignment=HAlign_Center, verticalAlignment=VAlign_Center, padding=<ltrb 0 0 120 0>}}
  }
  arrow_u_1 {
   widgetClassPath=&Image
   bIsVariable=true
   parent=Overlay_2
   widget=arrow_u_1
   properties {
    brush {
     bIsDynamicallyLoaded=false
     drawAs=Image
     imageType=FullColor
     mirroring=NoMirror
     resourceName=None
     resourceObject=&T_UI_Dpad_Right_Sprite
     tiling=NoTile
     imageSize=<96 96>
     margin=<ltrb 0 0 0 0>
     outlineSettings {
      bUseBrushTransparency=false
      roundingType=HalfHeightRadius
      width=0
      color {colorUseRule=UseColor_Specified, specifiedColor=<rgba 0 0 0 0>}
      cornerRadii=<0 0 0 0>
     }
     tintColor {colorUseRule=UseColor_Specified, specifiedColor=<rgba 1 1 1 1>}
     uVRegion {bIsValid=false, max=<0 0>, min=<0 0>}
    }
    renderTransform {angle=-90, scale=<1 1>, shear=<0 0>, translation=<0 0>}
   }
   slot {class=OverlaySlot, properties {horizontalAlignment=HAlign_Center, verticalAlignment=VAlign_Center, padding=<ltrb 0 -120 0 0>}}
  }
  Image_DpadArrowUpHighlight {
   widgetClassPath=&Image
   bInherited=true
   bIsVariable=true
   parent=Overlay_2
   widget=Image_DpadArrowUpHighlight
   properties {
    visibility=Collapsed
    brush {
     bIsDynamicallyLoaded=false
     drawAs=Image
     imageType=FullColor
     mirroring=NoMirror
     resourceName=None
     resourceObject=&T_UI_Dpad_Right_Sprite
     tiling=NoTile
     imageSize=<96 96>
     margin=<ltrb 0 0 0 0>
     outlineSettings {
      bUseBrushTransparency=false
      roundingType=HalfHeightRadius
      width=0
      color {colorUseRule=UseColor_Specified, specifiedColor=<rgba 0 0 0 0>}
      cornerRadii=<0 0 0 0>
     }
     tintColor {colorUseRule=UseColor_Specified, specifiedColor=<rgba 1 1 1 1>}
     uVRegion {bIsValid=false, max=<0 0>, min=<0 0>}
    }
    colorAndOpacity=<rgba 0.9131 0.6308 0.0802 1>
    renderTransform {angle=-90, scale=<1 1>, shear=<0 0>, translation=<0 0>}
   }
   slot {class=OverlaySlot, properties {horizontalAlignment=HAlign_Center, verticalAlignment=VAlign_Center, padding=<ltrb 0 -120 0 0>}}
  }
  arrow_d_1 {
   widgetClassPath=&Image
   bIsVariable=true
   parent=Overlay_2
   widget=arrow_d_1
   properties {
    brush {
     bIsDynamicallyLoaded=false
     drawAs=Image
     imageType=FullColor
     mirroring=NoMirror
     resourceName=None
     resourceObject=&T_UI_Dpad_Right_Sprite
     tiling=NoTile
     imageSize=<96 96>
     margin=<ltrb 0 0 0 0>
     outlineSettings {
      bUseBrushTransparency=false
      roundingType=HalfHeightRadius
      width=0
      color {colorUseRule=UseColor_Specified, specifiedColor=<rgba 0 0 0 0>}
      cornerRadii=<0 0 0 0>
     }
     tintColor {colorUseRule=UseColor_Specified, specifiedColor=<rgba 1 1 1 1>}
     uVRegion {bIsValid=false, max=<0 0>, min=<0 0>}
    }
    renderTransform {angle=90, scale=<1 1>, shear=<0 0>, translation=<0 0>}
   }
   slot {class=OverlaySlot, properties {horizontalAlignment=HAlign_Center, verticalAlignment=VAlign_Center, padding=<ltrb 0 120 0 0>}}
  }
  Image_DpadArrowDownHighlight {
   widgetClassPath=&Image
   bInherited=true
   bIsVariable=true
   parent=Overlay_2
   widget=Image_DpadArrowDownHighlight
   properties {
    visibility=Collapsed
    brush {
     bIsDynamicallyLoaded=false
     drawAs=Image
     imageType=FullColor
     mirroring=NoMirror
     resourceName=None
     resourceObject=&T_UI_Dpad_Right_Sprite
     tiling=NoTile
     imageSize=<96 96>
     margin=<ltrb 0 0 0 0>
     outlineSettings {
      bUseBrushTransparency=false
      roundingType=HalfHeightRadius
      width=0
      color {colorUseRule=UseColor_Specified, specifiedColor=<rgba 0 0 0 0>}
      cornerRadii=<0 0 0 0>
     }
     tintColor {colorUseRule=UseColor_Specified, specifiedColor=<rgba 1 1 1 1>}
     uVRegion {bIsValid=false, max=<0 0>, min=<0 0>}
    }
    colorAndOpacity=<rgba 0.9131 0.6308 0.0802 1>
    renderTransform {angle=90, scale=<1 1>, shear=<0 0>, translation=<0 0>}
   }
   slot {class=OverlaySlot, properties {horizontalAlignment=HAlign_Center, verticalAlignment=VAlign_Center, padding=<ltrb 0 120 0 0>}}
  }
 ]
}
