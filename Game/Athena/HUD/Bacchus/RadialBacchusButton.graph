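# RadialBacchusButton is a HUD button widget that shows a radial progress timer for gameplay cue effects matching its listened tag.
# Cue events with positive magnitude set duration and start time, then update the effect through the C++ base class.
# UpdateRemainingTime drives the material's Percentage parameter on either the mirrored elapsed image or remaining image, per fill/deplete behavior.
# CueCount tracks overlapping cues, with a failsafe timer hiding leftover progress after expiry.
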WidgetBlueprint /Game/Athena/HUD/Bacchus/RadialBacchusButton
animations [TapAnim {binding_count=0, display_label="", length_seconds=0.1, object=&self:TapAnim, sequence {bindings []}}]
graphs [
 EventGraph {nodes [
   n1 = VerifyCueCountIsClear() {is=title K2Node_CustomEvent}
   > Branch(Condition="Not Equal (Integer)"!(A:int=Get CueCount() -> (CueCount:int=0), B:int=0) -> (ReturnValue:bool) {is=title K2Node_CallFunction})
   > Set CueCount(CueCount:int=0) -> (Output_Get:int=0)
   > HideSelf()
   HideSelf() {is=title K2Node_CustomEvent}
   > SetVisibility(
    self=[Get Image_Elapsed() -> (Image_Elapsed:Image Object Reference), Get Image_Remaining() -> (Image_Remaining:Image Object Reference)]
    InVisibility=Collapsed
   )
   n2 = Event Construct()
   n3 = OnCursorModeChanged!() -> (OutputDelegate, then, bCursorModeEnabled:bool=~, ActionName:Name, CursorModeContentWidget:User Widget Object Reference) {is=title K2Node_CustomEvent}
   n4 = Get FortHUDContext!() -> (ReturnValue:Fort HUDContext Object Reference) {is=title K2Node_GetBlueprintContextEx}
   Bind Event to On Cursor Mode Changed(execute=@n2.then, self:Fort HUDContext Object Reference=@n4, Delegate=@n3) {is=title K2Node_AddDelegate}
   > SetVisibility(self=[Get Image_Elapsed(), Get Image_Remaining()], InVisibility=Collapsed)
   n5 = Event UpdateRemainingTime!() -> (OutputDelegate, then, Remaining:float, Duration:float) {is=title K2Node_Event}
   > n6 = Switch on ERadialBarBehavior!(execute, Selection:ERadialBarBehavior Enum=Get RadialBarBehavior() -> (RadialBarBehavior:ERadialBarBehavior Enum=FillBar)) -> (FillBar:Exec, DepleteBar:Exec) {is=title K2Node_SwitchEnum}
   UpdateDepleteProgress!(execute=@n6.DepleteBar, self, Remaining:double=@n5.Remaining, Duration:double=@n5.Duration) -> (then)
   UpdateFillProgress!(execute=@n6.FillBar, self, Remaining:double=@n5.Remaining, Duration:double=@n5.Duration) -> (then)
   Event Destruct()
   > Unbind Event from On Cursor Mode Changed(self:Fort HUDContext Object Reference=@n4, Delegate=@n3) {is=title K2Node_RemoveDelegate}
   n7 = Event ReceiveRadialProgressEvent_Blueprint!() -> (
    OutputDelegate
    then
    EventName:Name=None
    EventType:EGameplayCueEvent Enum=OnActive
    Parameters:"Gameplay Cue Parameters Structure (by ref)"
   ) {is=title K2Node_Event}
   n8 = BreakGameplayCueParameters!(Parameters:"Gameplay Cue Parameters Structure (by ref)"=@n7.Parameters) -> (
    NormalizedMagnitude:float
    RawMagnitude:float
    EffectContext:Gameplay Effect Context Handle Structure
    MatchedTagName:Gameplay Tag Structure
    OriginalTag:Gameplay Tag Structure
    AggregatedSourceTags:Gameplay Tag Container Structure
    AggregatedTargetTags:Gameplay Tag Container Structure
    Location:Vector
    Normal:Vector
    Instigator:Actor Object Reference
    EffectCauser:Actor Object Reference
    SourceObject:Object Reference
    PhysicalMaterial:Physical Material Object Reference
    GameplayEffectLevel:int=0
    AbilityLevel:int=0
    TargetAttachComponent:Scene Component Object Reference
    bReplicateLocationWhenUsingMinimalRepProxy:bool
   )
   > n9 = Branch(Condition="float <= float"!(A:double=@n8.RawMagnitude, B:double=0) -> (ReturnValue:bool) {is=title K2Node_CallFunction})
   > HideSelf() {is=title K2Node_CallFunction}
   n10 = GetListenedGameplayTag!(execute=@n9.else, self:Radial Bacchus Button Base Object Reference) -> (then, ReturnValue:Gameplay Tag Structure)
   > Branch(Condition="Equal (Name)"(A=GetTagName!(GameplayTag:"Gameplay Tag Structure (by ref)"=@n10) -> (ReturnValue:Name=None), B=@n7.EventName))
   > n11 = Switch on EGameplayCueEvent!(execute, Selection:EGameplayCueEvent Enum=@n7.EventType) -> (OnActive:Exec, WhileActive:Exec, Executed:Exec, Removed:Exec) {is=title K2Node_SwitchEnum}
   Set EffectDuration(execute=[@n11.Executed, @n11.WhileActive, @n13.FillBar], EffectDuration:float=@n8.RawMagnitude) -> (Output_Get:float)
   > Set EffectStartTime(EffectStartTime:float=GetGameTimeInSeconds()) -> (Output_Get:float)
   > UpdateEffect(self:Radial Bacchus Button Base Object Reference)
   > SetVisibility(
    self=Select!(
     FillBar:Widget Object Reference=Get Image_Elapsed()
     DepleteBar:Widget Object Reference=Get Image_Remaining()
     Index:ERadialBarBehavior Enum=Get RadialBarBehavior() -> (RadialBarBehavior=FillBar)
    ) -> (ReturnValue:Widget Object Reference)
    InVisibility=HitTestInvisible
   )
   > ForceVolatile!(execute, self:Widget Object Reference, bForce:bool=true) -> (then)
   > Set CueCount(CueCount=int + int(A=Get CueCount() -> (CueCount=0), B=1) -> (ReturnValue=0) {is=title K2Node_CommutativeAssociativeBinaryOperator}) -> (Output_Get=0)
   > Clear and Invalidate Timer by Handle(Handle=Get VerifyCueCountTimerHandle() -> (VerifyCueCountTimerHandle:Timer Handle Structure))
   > n12 = Set Timer by Event(
    Delegate=@n1
    Time=float + float(A=Get EffectDuration() -> (EffectDuration:float), B=1) {is=title K2Node_CommutativeAssociativeBinaryOperator}
    bLooping=false
   )
   > Set VerifyCueCountTimerHandle(VerifyCueCountTimerHandle:Timer Handle Structure=@n12) -> (Output_Get:Timer Handle Structure)
   > StartUpdateEffect(self:Radial Bacchus Button Base Object Reference)
   n13 = Switch on ERadialBarBehavior(execute=@n11.OnActive, Selection=Get RadialBarBehavior() -> (RadialBarBehavior=FillBar))
   n14 = Set CueCount(execute=@n11.Removed, CueCount=int + int(A=Get CueCount() -> (CueCount=0), B=-1) -> (ReturnValue=0)) -> (Output_Get=0)
   > Branch(Condition="integer <= integer"!(A:int=@n14, B:int=0) -> (ReturnValue:bool) {is=title K2Node_CallFunction})
   > HideSelf()
   Event OnEffectStarted() {is=title K2Node_Event}
   n15 = Get RadialBarBehavior() -> (RadialBarBehavior=FillBar)
   > SetVisibility(self=Select(FillBar=Get Image_Elapsed(), DepleteBar=Get Image_Remaining(), Index=@n15), InVisibility=HitTestInvisible)
   > ForceVolatile(bForce=true)
   > SetVisibility(
    self=Select(FillBar:Image Object Reference=Get Image_Remaining(), DepleteBar:Image Object Reference=Get Image_Elapsed(), Index=@n15) -> (ReturnValue:Image Object Reference)
    InVisibility=Collapsed
   )
   Event OnEffectCompleted() {is=title K2Node_Event}
   > HideSelf()
   Event PreConstruct()
   > n16 = Get!(execute:Exec) -> (then:Exec, else:Exec, RadialTimerOverrideMaterial:Material Interface Object Reference)
   > SetBrushFromMaterial(self=[Get Image_Remaining(), Get Image_Elapsed()], Material=@n16)
  ]}
 IsListenedTag {nodes [
   n1 = IsListenedTag!() -> (then, MatchedTag:Gameplay Tag Structure) {is=title K2Node_FunctionEntry}
   > n2 = GetListenedGameplayTag()
   > Return Node!(
    execute
    Match:bool="Equal (GameplayTag)"!(A:Gameplay Tag Structure=@n2, B:Gameplay Tag Structure=@n1) -> (ReturnValue:bool) {is=title K2Node_CallFunction}
   )
  ]}
 UpdateDepleteProgress {nodes [
   n1 = UpdateDepleteProgress!() -> (then, Remaining:double=~, Duration:double=~) {is=title K2Node_FunctionEntry}
   > n2 = GetDynamicMaterial(self=Get Image_Remaining())
   > SetScalarParameterValue(self=@n2, ParameterName=Percentage, Value="Clamp (Float)"(Value=float / float(A=@n1, B=@n1.Duration), Min=0, Max=1))
   > SetVisibility(self=Get Image_Elapsed(), InVisibility=Collapsed)
  ]}
 UpdateFillProgress {nodes [
   n1 = UpdateFillProgress!() -> (then, Remaining:double=~, Duration:double=~) {is=title K2Node_FunctionEntry}
   > n2 = GetDynamicMaterial(self=Get Image_Elapsed())
   n3 = float - float(A=@n1.Duration, B=@n1)
   > SetScalarParameterValue(self=@n2, ParameterName=Percentage, Value="Clamp (Float)"(Value=float / float(A=@n3, B=@n1.Duration), Min=0, Max=1))
   > SetVisibility(self=Get Image_Remaining(), InVisibility=Collapsed)
  ]}
]
variables [
 CueCount {type=integer}
 DefaultIcon {type=PaperSprite, default=null}
 Pressed Color {type=SlateColor, default {colorUseRule=UseColor_Specified, specifiedColor=<rgba 0 0 0 1>}}
 RadialTimerOverrideMaterial {type=MaterialInterface, default=null}
 Unpressed Color {type=SlateColor, default {colorUseRule=UseColor_Specified, specifiedColor=<rgba 1 1 1 1>}}
 VerifyCueCountTimerHandle {type=TimerHandle}
]
widgets {
 info {inheritedWidgetCount=1, parentClass=/Script/FortniteUI.RadialBacchusButtonBase, rootWidgetClass=&Overlay, widgetCount=4}
 widgets [
  Overlay_1 {widgetClassPath=&Overlay, widget=Overlay_1}
  Image_Remaining {
   widgetClassPath=&Image
   bIsVariable=true
   parent=Overlay_1
   widget=Image_Remaining
   properties {
    visibility=SelfHitTestInvisible
    brush {
     bIsDynamicallyLoaded=false
     drawAs=Image
     imageType=NoImage
     mirroring=NoMirror
     resourceName=None
     resourceObject=/Game/Athena/HUD/M_HealProgress
     tiling=NoTile
     imageSize=<256 256>
     margin=<ltrb 0 0 0 0>
     outlineSettings {
      bUseBrushTransparency=false
      roundingType=HalfHeightRadius
      width=0
      color {colorUseRule=UseColor_Specified, specifiedColor=<rgba 0 0 0 0>}
      cornerRadii=<0 0 0 0>
     }
     tintColor {colorUseRule=UseColor_Specified, specifiedColor=<rgba 1 1 1 1>}
     uVRegion {bIsValid=false, max=<0 0>, min=<0 0>}
    }
   }
   slot {class=OverlaySlot}
  }
  Image_Elapsed {
   widgetClassPath=&Image
   bIsVariable=true
   parent=Overlay_1
   widget=Image_Elapsed
   properties {
    visibility=SelfHitTestInvisible
    brush {
     bIsDynamicallyLoaded=false
     drawAs=Image
     imageType=NoImage
     mirroring=NoMirror
     resourceName=None
     resourceObject=&M_HealProgress
     tiling=NoTile
     imageSize=<256 256>
     margin=<ltrb 0 0 0 0>
     outlineSettings {
      bUseBrushTransparency=false
      roundingType=HalfHeightRadius
      width=0
      color {colorUseRule=UseColor_Specified, specifiedColor=<rgba 0 0 0 0>}
      cornerRadii=<0 0 0 0>
     }
     tintColor {colorUseRule=UseColor_Specified, specifiedColor=<rgba 1 1 1 1>}
     uVRegion {bIsValid=false, max=<0 0>, min=<0 0>}
    }
    renderTransform {angle=0, scale=<-1 1>, shear=<0 0>, translation=<0 0>}
   }
   slot {class=OverlaySlot}
  }
  TutorialHighlight {widgetClassPath=&TutorialHighlight}
 ]
}
