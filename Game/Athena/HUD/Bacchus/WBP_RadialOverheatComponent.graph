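# WBP_RadialOverheatComponent is a RadialBacchusButton child widget that tints itself to show overheat timing on the Bacchus HUD.
# It overrides UpdateRemainingTime, calling the parent first and then setting the whole widget's color and opacity.
# The color lerps by Remaining over Duration, shifting from translucent yellow at full time to opaque red as time runs out.
# It adds no widgets of its own.
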WidgetBlueprint /Game/Athena/HUD/Bacchus/WBP_RadialOverheatComponent
graphs [EventGraph {nodes [
   n1 = Event UpdateRemainingTime!() -> (OutputDelegate, then, Remaining:float, Duration:float) {is=title K2Node_Event}
   > Parent: UpdateRemainingTime!(execute, Remaining:float=@n1.Remaining, Duration:float=@n1.Duration) -> (then) {is=title K2Node_CallParentFunction}
   > SetColorAndOpacity(
    self:User Widget Object Reference
    InColorAndOpacity:Linear Color Structure="Lerp (LinearColor)"(
     A=Get FinalOverheatColor() -> (FinalOverheatColor:Linear Color Structure)
     B=Get InitialOverheatColor() -> (InitialOverheatColor:Linear Color Structure)
     Alpha=SafeDivide(A=@n1.Remaining, B=@n1.Duration)
    )
   )
  ]}]
variables [
 FinalOverheatColor {type=LinearColor, default=<rgba 0.6174 0.0161 0.0161 1>}
 InitialOverheatColor {type=LinearColor, default=<rgba 0.7835 0.7529 0.0027 0.2>}
]
widgets {
 info {inheritedWidgetCount=0, parentClass=&/RadialBacchusButton.RadialBacchusButton_C, rootWidgetClass=null, widgetCount=0}
 widgets []
}
