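# Mobile HUD element canvas for Fortnite's touch movement stick, as a Widget Blueprint derived from the native MobileHUDElementCanvas.
# Its widget list holds a CanvasPanel wrapper, background image and HUD-scale ScaleBox marked as inherited, plus a TutorialHighlight widget.
# The event graph and variable list are empty, so the asset adds no Blueprint logic of its own.
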WidgetBlueprint /Game/Athena/HUD/Bacchus/WBP_TouchMoveStickCanvas
graphs [EventGraph {links [], nodes []}]
variables []
widgets {
 info {inheritedWidgetCount=4, parentClass=/Script/FortniteUI.MobileHUDElementCanvas, rootWidgetClass=&CanvasPanel, widgetCount=4}
 widgets [
  CanvasPanel_Wrapper {widgetClassPath=&CanvasPanel, bInherited=true, widget=CanvasPanel_Wrapper}
  Image_Background {widgetClassPath=&Image, bInherited=true, widget=null}
  ScaleBox_HUDWidgetScale {widgetClassPath=&ScaleBox, bInherited=true, widget=null}
  TutorialHighlight {widgetClassPath=&TutorialHighlight}
 ]
}
